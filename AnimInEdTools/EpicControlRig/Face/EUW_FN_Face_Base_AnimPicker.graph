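# Editor utility widget, child of EUW_FN_Face_Base, that builds a face animation picker for Control Rig.
# On widget load it sets core face variables, then creates the faceboard and type-specific picker class, placing it in Main_Layout.
# OnFace mode only shows a "not implemented" warning, and an empty face type logs an error.
# Get Primary Panel returns the canvas, while Get all Selectable Elements returns nothing.
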EditorUtilityWidgetBlueprint /AnimInEdTools/EpicControlRig/Face/EUW_FN_Face_Base_AnimPicker
graphs [
 Create Face GUI and Picker {nodes [
   n1 = Create Face GUI and Picker!() -> (then, OnFace:bool=~) {is=title K2Node_FunctionEntry}
   > n2 = Branch(Condition=@n1)
   > Show Message Dialog(
    Title=MakeLiteralText(Value="LOCTEXT(\"Cannot find Facial Rig\")")
    Message=MakeLiteralText(Value="LOCTEXT(\"Sorry, OnFace ControlRig is not implemented yet!!!\")")
    MessageType=Ok
    DefaultValue=No
    MessageCategory=Warning
   )
   n3 = Get Face Type!(execute=@n2.else, self:EUW FN Face Base Object Reference) -> (then, face_type:String) {is=title K2Node_CallFunction}
   > n4 = Branch(Condition=NOT Boolean(A=IsEmpty(InString=@n3)))
   > Set Mann Faceboard!(
    execute
    self:EUW FN Face Base Object Reference
    face type:String=Get Face Type() -> (Face Type:String) {is=title K2Node_VariableGet}
    onFace:bool="Get onFace "() -> ("onFace ":bool)
   ) -> (then) {is=title K2Node_CallFunction}
   > n5 = Get Face Picker by Type!(execute, self:EUW FN Face Base Object Reference, FaceType:String=Get Face Type(), OnFace:bool="Get onFace "()) -> (then, FacePicker:User Widget Class Reference) {is=title K2Node_CallFunction}
   > n6 = Set Current Face Picker(Current Face Picker:User Widget Class Reference=@n5) -> (Output_Get:User Widget Class Reference)
   > n7 = Set Face Picker!(
    execute
    self:EUW FN Face Base Object Reference
    EUW_FacePicker:User Widget Class Reference=@n6
    Layout:Vertical Box Object Reference=Get Main_Layout() -> (Main_Layout:Vertical Box Object Reference)
   ) -> (then, Success:bool, Out Face Picker:EUW Asset Picker Base Object Reference) {is=title K2Node_CallFunction}
   > Set Current Face Picker EUW(Current Face Picker EUW:EUW Asset Picker Base Object Reference=@n7.Out Face Picker) -> (Output_Get:EUW Asset Picker Base Object Reference)
   LogString(execute=@n4.else, InString=Append(A=Invalid Face Type:, B=@n3))
   > Return Node!(execute, Success:bool=false, Out Face Picker:EUW Asset Picker Base Object Reference)
   For Each Loop(Array:Array of Widget Object References=GetAllChildren(self=Get Main_Layout())) -> (Array Element:Widget Object Reference)
  ]}
 EventGraph {nodes [
   n1 = Event Widget has loaded!() -> (
    OutputDelegate
    then
    Control Rig Proxy Binding:Control Rig Sequencer Binding Proxy Structure
    Animatable Creature:Animatable Creature Object Reference
    Anim Picker EUW:Animation Picker Object Reference
   ) {is=title K2Node_Event}
   > Parent: Widget has loaded!(
    execute
    Control Rig Proxy Binding:Control Rig Sequencer Binding Proxy Structure=@n1.Control Rig Proxy Binding
    Animatable Creature:Animatable Creature Object Reference=@n1.Animatable Creature
    Anim Picker EUW:Animation Picker Object Reference=@n1.Anim Picker EUW
   ) -> (then) {is=title K2Node_CallParentFunction}
   > Set Face Core Variables(self:EUW FN Face Base Object Reference) {is=title K2Node_CallFunction}
   > Create Face GUI and Picker!(execute, self, OnFace:bool="Get onFace "()) -> (then, Success:bool, Out Face Picker:EUW Asset Picker Base Object Reference) {is=title K2Node_CallFunction}
  ]}
 Get Primary Panel {nodes [
   Get Primary Panel() {is=title K2Node_FunctionEntry}
   > Return Node!(execute, Panel:Panel Widget Object Reference=Get Main_CanvasPanel() -> (Main_CanvasPanel:Canvas Panel Object Reference))
  ]}
 Get all Selectable Elements {nodes [
   Get all Selectable Elements() {is=title K2Node_FunctionEntry}
   > Return Node!(execute, All Elements:Array of EUW Picker Element Object References)
  ]}
]
variables [
 Current Face Picker {type=Class, default=null}
 Current Face Picker EUW {type=EUW_Asset_Picker_Base_C, default=null}
 "onFace " {type=boolean}
]
widgets {
 info {parentClass=&/EUW_FN_Face_Base.EUW_FN_Face_Base_C, rootWidgetClass=&CanvasPanel, widgetCount=3}
 widgets [
  Main_CanvasPanel {widgetClassPath=&CanvasPanel, bIsVariable=true, widget=Main_CanvasPanel, properties {renderTransformPivot=<0 0>}}
  SizeBox_67 {
   widgetClassPath=&SizeBox
   parent=Main_CanvasPanel
   widget=SizeBox_67
   slot {
    class=CanvasPanelSlot
    properties {layoutData {alignment=<0 0>, anchors {maximum=<0.0896 0.1222>, minimum=<0.0896 0.1222>}, offsets=<ltrb -156.1559 -120.1201 1891.5364 1062.0721>}}
   }
  }
  Main_Layout {widgetClassPath=&VerticalBox, bIsVariable=true, parent=SizeBox_67, widget=Main_Layout, slot {class=SizeBoxSlot}}
 ]
}
