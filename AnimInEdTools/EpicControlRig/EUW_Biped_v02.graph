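# EUW_Biped_v02 is an editor utility picker widget for animating a biped Control Rig in Sequencer.
# It caches limb, spine and neck button groups, then wires IK/FK checkboxes to channel bools, button tints and labels.
# Buttons snap or match IK/FK, zero controls, toggle group visibility and a reduced BlockingMode layout, and open face and IK/FK matching tools.
# On load it stores the bound rig, casts to ModularRig and applies modular-rig naming to buttons.
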EditorUtilityWidgetBlueprint /AnimInEdTools/EpicControlRig/EUW_Biped_v02
graphs [
 Cache IK FK Buttons {nodes [
   Cache IK FK Buttons() {is=title K2Node_FunctionEntry}
   > n1 = Sequence!(execute) -> (then_0, then_1, then_2:Exec, then_3:Exec)
   > Set Left Arm FK(Left Arm FK:Array of EUW Picker Button Object References=Make Array!(
     "[0]":EUW Picker Button Object Reference=Get Arm-L_FK-0() -> (Arm-L_FK-0:EUW Picker Button Object Reference)
     "[1]":EUW Picker Button Object Reference=Get Arm-L_FK-1() -> (Arm-L_FK-1:EUW Picker Button Object Reference)
     "[2]":EUW Picker Button Object Reference=Get Arm-L_FK-2() -> (Arm-L_FK-2:EUW Picker Button Object Reference)
     "[3]":EUW Picker Button Object Reference=Get AllFKLArm() -> (AllFKLArm:EUW Picker Button Object Reference)
     "[4]":EUW Picker Button Object Reference=Get Arm-L_FK-0-Gimbal() -> (Arm-L_FK-0-Gimbal:EUW Picker Button Object Reference)
    ) -> (Array:Array of EUW Picker Button Object References)) -> (Output_Get:Array of EUW Picker Button Object References)
   > Set Right Arm FK(Right Arm FK:Array of EUW Picker Button Object References=Make Array(
     "[0]"=Get Arm-R_FK-0() -> (Arm-R_FK-0:EUW Picker Button Object Reference)
     "[1]"=Get Arm-R_FK-1() -> (Arm-R_FK-1:EUW Picker Button Object Reference)
     "[2]"=Get Arm-R_FK-2() -> (Arm-R_FK-2:EUW Picker Button Object Reference)
     "[3]"=Get AllFKRArm() -> (AllFKRArm:EUW Picker Button Object Reference)
     "[4]"=Get Arm-R_FK-0-Gimbal() -> (Arm-R_FK-0-Gimbal:EUW Picker Button Object Reference)
    )) -> (Output_Get:Array of EUW Picker Button Object References)
   > Set Left Leg FK(Left Leg FK:Array of EUW Picker Button Object References=Make Array(
     "[0]"=Get Leg-L_FK-0() -> (Leg-L_FK-0:EUW Picker Button Object Reference)
     "[1]"=Get Leg-L_FK-1() -> (Leg-L_FK-1:EUW Picker Button Object Reference)
     "[2]"=Get Leg-L_FK-2() -> (Leg-L_FK-2:EUW Picker Button Object Reference)
     "[3]"=Get Foot-L_Toes-FK() -> (Foot-L_Toes-FK:EUW Picker Button Object Reference)
     "[4]"=Get AllFKLeg_l() -> (AllFKLeg_l:EUW Picker Button Object Reference)
    )) -> (Output_Get:Array of EUW Picker Button Object References)
   > Set Right Leg FK(Right Leg FK:Array of EUW Picker Button Object References=Make Array(
     "[0]"=Get Leg-R_FK-0() -> (Leg-R_FK-0:EUW Picker Button Object Reference)
     "[1]"=Get Leg-R_FK-1() -> (Leg-R_FK-1:EUW Picker Button Object Reference)
     "[2]"=Get Leg-R_FK-2() -> (Leg-R_FK-2:EUW Picker Button Object Reference)
     "[3]"=Get Foot-R_Toes-FK() -> (Foot-R_Toes-FK:EUW Picker Button Object Reference)
     "[4]"=Get AllFKLeg_r() -> (AllFKLeg_r:EUW Picker Button Object Reference)
    )) -> (Output_Get:Array of EUW Picker Button Object References)
   > Set Neck IK(Neck IK:Array of EUW Picker Button Object References=Make Array!(
     "[0]"=Get Neck_Start-IK() -> (Neck_Start-IK:EUW Picker Button Object Reference)
     "[1]"=Get Neck_Mid-IK() -> (Neck_Mid-IK:EUW Picker Button Object Reference)
     "[2]"=Get Neck_End-IK() -> (Neck_End-IK:EUW Picker Button Object Reference)
    ) -> (Array)) -> (Output_Get:Array of EUW Picker Button Object References)
   > Set Neck FK(Neck FK:Array of EUW Picker Button Object References=Make Array(
     "[0]"=Get Neck_End-FK() -> (Neck_End-FK:EUW Picker Button Object Reference)
     "[1]"=Get Neck_Mid-FK() -> (Neck_Mid-FK:EUW Picker Button Object Reference)
     "[2]"=Get Neck_Start-FK() -> (Neck_Start-FK:EUW Picker Button Object Reference)
    )) -> (Output_Get:Array of EUW Picker Button Object References)
   > Set Spine FK(Spine FK:Array of EUW Picker Button Object References=Make Array(
     "[0]"=Get Spine_End-FK() -> (Spine_End-FK:EUW Picker Button Object Reference)
     "[1]"=Get Spine_Mid-FK() -> (Spine_Mid-FK:EUW Picker Button Object Reference)
     "[2]"=Get Spine_Start-FK() -> (Spine_Start-FK:EUW Picker Button Object Reference)
    )) -> (Output_Get:Array of EUW Picker Button Object References)
   > Set Spine IK(Spine IK:Array of EUW Picker Button Object References=Make Array(
     "[0]"=Get Spine_End-IK() -> (Spine_End-IK:EUW Picker Button Object Reference)
     "[1]"=Get Spine_Mid-IK() -> (Spine_Mid-IK:EUW Picker Button Object Reference)
     "[2]"=Get Spine_Start-IK() -> (Spine_Start-IK:EUW Picker Button Object Reference)
    )) -> (Output_Get:Array of EUW Picker Button Object References)
   Set Left Arm IK(
    execute=@n1.then_1
    Left Arm IK:Array of EUW Picker Button Object References=Make Array!(
     "[0]"=Get Arm-L_IK() -> (Arm-L_IK:EUW Picker Button Object Reference)
     "[1]"=Get Arm-L_PV() -> (Arm-L_PV:EUW Picker Button Object Reference)
     "[2]"=Get Arm-L_IK-Rotation() -> (Arm-L_IK-Rotation:EUW Picker Button Object Reference)
     "[3]":EUW Picker Button Object Reference=Get Arm-L_Mid() -> (Arm-L_Mid:EUW Picker Button Object Reference)
     "[4]":EUW Picker Button Object Reference=Get Arm-L_IK-Rotation()
     "[5]":EUW Picker Button Object Reference=Get AllIKLArm() -> (AllIKLArm:EUW Picker Button Object Reference)
    ) -> (Array)
   ) -> (Output_Get:Array of EUW Picker Button Object References)
   > Set Right Arm IK(Right Arm IK:Array of EUW Picker Button Object References=Make Array(
     "[0]"=Get Arm-R_IK() -> (Arm-R_IK:EUW Picker Button Object Reference)
     "[1]"=Get Arm-R_PV() -> (Arm-R_PV:EUW Picker Button Object Reference)
     "[2]"=Get Arm-R_IK-Rotation() -> (Arm-R_IK-Rotation:EUW Picker Button Object Reference)
     "[3]"=Get Arm-R_Mid() -> (Arm-R_Mid:EUW Picker Button Object Reference)
     "[4]"=Get Arm-R_IK-Rotation()
     "[5]"=Get AllIKRArm() -> (AllIKRArm:EUW Picker Button Object Reference)
    )) -> (Output_Get:Array of EUW Picker Button Object References)
   > Set Left Leg IK(Left Leg IK:Array of EUW Picker Button Object References=Make Array!(
     "[0]"=Get Leg-L_PV() -> (Leg-L_PV:EUW Picker Button Object Reference)
     "[1]"=Get Leg-L_IK() -> (Leg-L_IK:EUW Picker Button Object Reference)
     "[2]"=Get Foot-L_Foot-Rocker() -> (Foot-L_Foot-Rocker:EUW Picker Button Object Reference)
     "[3]"=Get Foot-L_Ball-IK() -> (Foot-L_Ball-IK:EUW Picker Button Object Reference)
     "[4]"=Get Foot-L_Toes-IK() -> (Foot-L_Toes-IK:EUW Picker Button Object Reference)
     "[5]"=Get Foot-L_Heel() -> (Foot-L_Heel:EUW Picker Button Object Reference)
     "[6]":EUW Picker Button Object Reference=Get Foot-L_Ball() -> (Foot-L_Ball:EUW Picker Button Object Reference)
     "[7]":EUW Picker Button Object Reference=Get Leg-L_Mid() -> (Leg-L_Mid:EUW Picker Button Object Reference)
     "[8]":EUW Picker Button Object Reference=Get Leg-L_IK-Base() -> (Leg-L_IK-Base:EUW Picker Button Object Reference)
     "[9]":EUW Picker Button Object Reference=Get AllIKLeg_l() -> (AllIKLeg_l:EUW Picker Button Object Reference)
     "[10]":EUW Picker Button Object Reference=Get AllIKLegExtra_l() -> (AllIKLegExtra_l:EUW Picker Button Object Reference)
     "[11]":EUW Picker Button Object Reference=Get Foot-L_Toe-Tip() -> (Foot-L_Toe-Tip:EUW Picker Button Object Reference)
    ) -> (Array)) -> (Output_Get:Array of EUW Picker Button Object References)
   > Set Right Leg IK(Right Leg IK:Array of EUW Picker Button Object References=Make Array(
     "[0]"=Get Leg-R_IK() -> (Leg-R_IK:EUW Picker Button Object Reference)
     "[1]"=Get Leg-R_PV() -> (Leg-R_PV:EUW Picker Button Object Reference)
     "[2]"=Get Foot-R_Foot-Rocker() -> (Foot-R_Foot-Rocker:EUW Picker Button Object Reference)
     "[3]"=Get Foot-R_Ball-IK() -> (Foot-R_Ball-IK:EUW Picker Button Object Reference)
     "[4]"=Get Foot-R_Toes-IK() -> (Foot-R_Toes-IK:EUW Picker Button Object Reference)
     "[5]"=Get Foot-R_Heel() -> (Foot-R_Heel:EUW Picker Button Object Reference)
     "[6]"=Get Foot-R_Ball() -> (Foot-R_Ball:EUW Picker Button Object Reference)
     "[7]"=Get Leg-R_Mid() -> (Leg-R_Mid:EUW Picker Button Object Reference)
     "[8]"=Get Leg-R_IK-Base() -> (Leg-R_IK-Base:EUW Picker Button Object Reference)
     "[9]"=Get AllIKLegExtras_r() -> (AllIKLegExtras_r:EUW Picker Button Object Reference)
     "[10]"=Get AllIKLeg_r() -> (AllIKLeg_r:EUW Picker Button Object Reference)
     "[11]"=Get Foot-R_Toe-Tip() -> (Foot-R_Toe-Tip:EUW Picker Button Object Reference)
    )) -> (Output_Get:Array of EUW Picker Button Object References)
   Return Node(execute=@n1.then_3)
   n2 = Get BlockingMode Buttons() -> (BlockingMode Buttons:Array of EUW Picker Button Object References)
   Append Array!(
    execute=@n1.then_2
    TargetArray:Array of EUW Picker Button Object References=@n2
    SourceArray:Array of EUW Picker Button Object References=Get Spine FK() -> (Spine FK:Array of EUW Picker Button Object References)
   ) -> (then)
   > Append Array(TargetArray=@n2, SourceArray=Get Neck FK() -> (Neck FK:Array of EUW Picker Button Object References))
   > Append Array(TargetArray=@n2, SourceArray=Get All Legs() -> (All Legs:Array of EUW Picker Button Object References))
   > Append Array(TargetArray=@n2, SourceArray=Get All Arms() -> (All Arms:Array of EUW Picker Button Object References))
   > Append Array(TargetArray=@n2, SourceArray=Get All Fingers() -> (All Fingers:Array of EUW Picker Button Object References))
   > Append Array(
    TargetArray=Get BlockingMode Buttons()
    SourceArray=Make Array!(
     "[0]"=Get BlockingMode() -> (BlockingMode:EUW Picker Button Object Reference)
     "[1]"=Get Body() -> (Body:EUW Picker Button Object Reference)
     "[2]"=Get Spine_Start-IK()
     "[3]"=Get TwistsVisibility() -> (TwistsVisibility:EUW Picker Button Object Reference)
     "[4]"=Get SpineSecVisibility() -> (SpineSecVisibility:EUW Picker Button Object Reference)
     "[5]"=Get LegIKVisibility() -> (LegIKVisibility:EUW Picker Button Object Reference)
     "[6]"=Get LegFKVisibility() -> (LegFKVisibility:EUW Picker Button Object Reference)
     "[7]"=Get ArmIKVisibility() -> (ArmIKVisibility:EUW Picker Button Object Reference)
     "[8]"=Get ArmFKVisibility() -> (ArmFKVisibility:EUW Picker Button Object Reference)
     "[9]"=Get Middle_L() -> (Middle_L:EUW Picker Button Object Reference)
     "[10]"=Get Middle_R() -> (Middle_R:EUW Picker Button Object Reference)
     "[11]"=Get Pinky_L() -> (Pinky_L:EUW Picker Button Object Reference)
     "[12]":EUW Picker Button Object Reference=Get Pinky_R() -> (Pinky_R:EUW Picker Button Object Reference)
     "[13]":EUW Picker Button Object Reference=Get Ring_L() -> (Ring_L:EUW Picker Button Object Reference)
     "[14]":EUW Picker Button Object Reference=Get Ring_R() -> (Ring_R:EUW Picker Button Object Reference)
     "[15]":EUW Picker Button Object Reference=Get Thumb_L() -> (Thumb_L:EUW Picker Button Object Reference)
     "[16]":EUW Picker Button Object Reference=Get Thumb_R() -> (Thumb_R:EUW Picker Button Object Reference)
     "[17]":EUW Picker Button Object Reference=Get Index_L() -> (Index_L:EUW Picker Button Object Reference)
     "[18]":EUW Picker Button Object Reference=Get Index_R() -> (Index_R:EUW Picker Button Object Reference)
     "[19]":EUW Picker Button Object Reference=Get Fingers_R_Start() -> (Fingers_R_Start:EUW Picker Button Object Reference)
     "[20]":EUW Picker Button Object Reference=Get Fingers_R_Mid() -> (Fingers_R_Mid:EUW Picker Button Object Reference)
     "[21]":EUW Picker Button Object Reference=Get Fingers_R_End() -> (Fingers_R_End:EUW Picker Button Object Reference)
     "[22]":EUW Picker Button Object Reference=Get Fingers_L_Start() -> (Fingers_L_Start:EUW Picker Button Object Reference)
     "[23]":EUW Picker Button Object Reference=Get Fingers_L_Mid() -> (Fingers_L_Mid:EUW Picker Button Object Reference)
     "[24]":EUW Picker Button Object Reference=Get Fingers_L_End() -> (Fingers_L_End:EUW Picker Button Object Reference)
     "[25]":EUW Picker Button Object Reference=Get ShowExtraButtons() -> (ShowExtraButtons:EUW Picker Button Object Reference)
     "[26]":EUW Picker Button Object Reference=Get ToggleCapsule_Button() -> (ToggleCapsule_Button:EUW Picker Button Object Reference)
     "[27]":EUW Picker Button Object Reference=Get BakeIKFKTool() -> (BakeIKFKTool:EUW Picker Button Object Reference)
     "[28]":EUW Picker Button Object Reference=Get Meta_R() -> (Meta_R:EUW Picker Button Object Reference)
     "[29]":EUW Picker Button Object Reference=Get Meta_L() -> (Meta_L:EUW Picker Button Object Reference)
     "[30]":EUW Picker Button Object Reference=Get PropVis_Button() -> (PropVis_Button:EUW Picker Button Object Reference)
     "[31]":EUW Picker Button Object Reference=Get ZeroOutselected() -> (ZeroOutselected:EUW Picker Button Object Reference)
     "[32]":EUW Picker Button Object Reference=Get ZeroOutbutton() -> (ZeroOutbutton:EUW Picker Button Object Reference)
     "[33]":EUW Picker Button Object Reference=Get root_Global() -> (root_Global:EUW Picker Button Object Reference)
     "[34]":EUW Picker Button Object Reference=Get root_Local() -> (root_Local:EUW Picker Button Object Reference)
     "[35]":EUW Picker Button Object Reference=Get root_Root() -> (root_Root:EUW Picker Button Object Reference)
     "[36]":EUW Picker Button Object Reference=Get Neck_End-IK()
     "[37]":EUW Picker Button Object Reference=Get Neck_Mid-IK()
     "[38]":EUW Picker Button Object Reference=Get Neck_Start-IK()
     "[39]":EUW Picker Button Object Reference=Get Spine_End-IK()
     "[40]":EUW Picker Button Object Reference=Get Spine_Mid-IK()
     "[41]":EUW Picker Button Object Reference=Get FacePickerButton() -> (FacePickerButton:EUW Picker Button Object Reference)
     "[42]":EUW Picker Button Object Reference=Get Attach_Weapon-R() -> (Attach_Weapon-R:EUW Picker Button Object Reference)
     "[43]":EUW Picker Button Object Reference=Get Attach_Weapon-L() -> (Attach_Weapon-L:EUW Picker Button Object Reference)
     "[44]":EUW Picker Button Object Reference=Get Attach_Hand-Attach-R() -> (Attach_Hand-Attach-R:EUW Picker Button Object Reference)
     "[45]":EUW Picker Button Object Reference=Get Attach_Hand-Attach-L() -> (Attach_Hand-Attach-L:EUW Picker Button Object Reference)
     "[46]":EUW Picker Button Object Reference=Get FaceBoardButton() -> (FaceBoardButton:EUW Picker Button Object Reference)
     "[47]":EUW Picker Button Object Reference=Get Match_Arm_L() -> (Match_Arm_L:EUW Picker Button Object Reference)
     "[48]":EUW Picker Button Object Reference=Get Match_Arm_R() -> (Match_Arm_R:EUW Picker Button Object Reference)
     "[49]":EUW Picker Button Object Reference=Get Match_Leg_L() -> (Match_Leg_L:EUW Picker Button Object Reference)
     "[50]":EUW Picker Button Object Reference=Get Match_Leg_R() -> (Match_Leg_R:EUW Picker Button Object Reference)
     "[51]":EUW Picker Button Object Reference=Get Snap_Leg_R() -> (Snap_Leg_R:EUW Picker Button Object Reference)
     "[52]":EUW Picker Button Object Reference=Get Snap_Leg_L() -> (Snap_Leg_L:EUW Picker Button Object Reference)
     "[53]":EUW Picker Button Object Reference=Get Snap_Arm_R() -> (Snap_Arm_R:EUW Picker Button Object Reference)
     "[54]":EUW Picker Button Object Reference=Get Snap_Arm_L() -> (Snap_Arm_L:EUW Picker Button Object Reference)
     "[55]":EUW Picker Button Object Reference=Get IKMove() -> (IKMove:EUW Picker Button Object Reference)
     "[56]":EUW Picker Button Object Reference=Get 4Fingers_r() -> (4Fingers_r:EUW Picker Button Object Reference)
     "[57]":EUW Picker Button Object Reference=Get 4Fingers_l() -> (4Fingers_l:EUW Picker Button Object Reference)
     "[58]":EUW Picker Button Object Reference=Get allFingers_r() -> (allFingers_r:EUW Picker Button Object Reference)
     "[59]":EUW Picker Button Object Reference=Get AllFingers_l() -> (AllFingers_l:EUW Picker Button Object Reference)
    ) -> (Array)
   )
  ]}
 EventGraph {nodes [
   n1 = Break Control Rig Sequencer Binding Proxy!(ControlRigSequencerBindingProxy:"Control Rig Sequencer Binding Proxy Structure (by ref)"=Get Control Rig Proxy Binding() -> (Control Rig Proxy Binding:Control Rig Sequencer Binding Proxy Structure)) -> (
    Proxy:Movie Scene Binding Proxy Structure
    ControlRig:Control Rig Object Reference
    Track:Movie Scene Control Rig Parameter Track Object Reference
   ) {is=title K2Node_BreakStruct}
   n2 = GetHierarchy!(self:Control Rig Object Reference=@n1.ControlRig) -> (ReturnValue:Rig Hierarchy Object Reference)
   n3 = GetSelectedKeys!(self:Rig Hierarchy Object Reference=@n2, InTypeFilter:ERigElementType Enum=Control) -> (ReturnValue:Array of Rig Element Key Structures)
   n4 = PrintString(InString="To String (Integer)"(InInt=Length(TargetArray:Array of Rig Element Key Structures=@n3)))
   n5 = For Each Loop(
    Exec=[@n4.then, @n51.then]
    Array:Array of Rig Element Key Structures=Get Controls!(self:Rig Hierarchy Object Reference=@n2, bTraverse:bool=true) -> (ReturnValue:Array of Rig Element Key Structures) {is=title K2Node_CallFunction}
   ) -> (Array Element:Rig Element Key Structure)
   > n6 = Branch(Condition=IsSelected!(self:Rig Hierarchy Object Reference=@n2, InKey:Rig Element Key Structure=@n5) -> (ReturnValue:bool))
   n7 = Get Current Time!(execute=[@n12.then, @n6.then], TimeUnit:EMovieSceneTimeUnit Enum=DisplayRate) -> (then, ReturnValue:Movie Scene Sequence Playback Params Structure) {is=title K2Node_CallFunction}
   > n8 = GetFocusedLevelSequence!(execute) -> (then, ReturnValue:Level Sequence Object Reference)
   n9 = Break Control Rig Sequencer Binding Proxy(ControlRigSequencerBindingProxy=Get Control Rig Proxy Binding())
   n10 = Make Frame Number!(Value:int=Break Movie Scene Sequence Playback Params!(MovieSceneSequencePlaybackParams:"Movie Scene Sequence Playback Params Structure (by ref)"=@n7) -> (
     Frame_FrameNumber_Value:int
     Frame_SubFrame:float=~
     Time:float=~
     MarkedFrame:String
     Timecode:Timecode Structure
     PositionType:EMovieScenePositionType Enum
     UpdateMethod:EUpdatePositionMethod Enum
     bHasJumped:bool=~
    ) {is=title K2Node_BreakStruct}) -> (FrameNumber:Frame Number Structure) {is=title K2Node_MakeStruct}
   n11 = Break Rig Element Key!(RigElementKey:"Rig Element Key Structure (by ref)"=@n5) -> (Type:ERigElementType Enum, Name:Name) {is=title K2Node_BreakStruct}
   > SetLocalControlRigEulerTransform!(
    execute
    LevelSequence:Level Sequence Object Reference=@n8
    ControlRig:Control Rig Object Reference=@n9.ControlRig
    ControlName:Name=@n11.Name
    Frame:Frame Number Structure=@n10
    Value_Location:Vector="0,0,0"
    Value_Rotation:Rotator="0,0,0"
    Value_Scale:Vector="1,1,1"
    TimeUnit:EMovieSceneTimeUnit Enum=DisplayRate
    bSetKey:bool=false
   ) -> (then)
   > SetLocalControlRigRotator!(
    execute
    LevelSequence:Level Sequence Object Reference=@n8
    ControlRig:Control Rig Object Reference=@n9.ControlRig
    ControlName:Name=@n11.Name
    Frame:Frame Number Structure=@n10
    Value:Rotator="0, 0, 0"
    TimeUnit:EMovieSceneTimeUnit Enum=DisplayRate
    bSetKey:bool=false
   ) -> (then)
   n12 = PrintString(InString="To String (Name)"(InName=@n11.Name))
   n13 = "On Check State Changed (Leg-L_Ik-Fk-Switch)"!() -> (OutputDelegate, then, bIsChecked:bool) {is=title K2Node_ComponentBoundEvent}
   n14 = Get Leg-L_Ik-Fk-Switch() -> (Leg-L_Ik-Fk-Switch:Editor Utility Check Box Object Reference)
   > Set Buttons Color From Checkbox State!(
    execute
    self:EUW Asset Picker Base Object Reference
    CheckBox:Editor Utility Check Box Object Reference=@n14
    IK Buttons:Array of EUW Picker Button Object References=Get Left Leg IK() -> (Left Leg IK:Array of EUW Picker Button Object References)
    FK Buttons:Array of EUW Picker Button Object References=Get Left Leg FK() -> (Left Leg FK:Array of EUW Picker Button Object References)
    Color Mult:double=1
    Opacity:double=0.5
   ) -> (then) {is=title K2Node_CallFunction}
   > Set Animation Channel Bool!(
    execute
    self:EUW Asset Picker Base Object Reference
    ControlRig:Control Rig Object Reference=Get Control Rig() -> (Control Rig:Control Rig Object Reference)
    Channel Name:Name=Convert to MR naming!(In name:String=GetDisplayName(Object=@n14)) -> (Out name:Name=None)
    Value:bool=@n13.bIsChecked
   ) -> (then) {is=title K2Node_CallFunction}
   > IK FK Switch Text Update!(execute, self, Switch CheckBox:Editor Utility Check Box Object Reference=Get Leg-L_Ik-Fk-Switch(), Prefix:String=Left Leg) -> (then)
   n15 = "On Check State Changed (Leg-R_Ik-Fk-Switch)"!() -> (OutputDelegate, then, bIsChecked:bool) {is=title K2Node_ComponentBoundEvent}
   n16 = Get Leg-R_Ik-Fk-Switch() -> (Leg-R_Ik-Fk-Switch:Editor Utility Check Box Object Reference)
   > Set Buttons Color From Checkbox State(
    CheckBox=@n16
    IK Buttons=Get Right Leg IK() -> (Right Leg IK:Array of EUW Picker Button Object References)
    FK Buttons=Get Right Leg FK() -> (Right Leg FK:Array of EUW Picker Button Object References)
    Color Mult=1
    Opacity=0.5
   )
   > Set Animation Channel Bool(
    ControlRig=Get Control Rig()
    Channel Name=Convert to MR naming(In name=GetDisplayName(Object=@n16)) -> (Out name=None)
    Value=@n15.bIsChecked
   )
   > IK FK Switch Text Update(Switch CheckBox=Get Leg-R_Ik-Fk-Switch(), Prefix=Right Leg)
   "Custom on Clicked Event (Snap_Arm_R)"() {is=title K2Node_ComponentBoundEvent}
   > Snap Event!(
    execute
    self
    Ik Fk Checkbox:Check Box Object Reference=Get Arm-R_Ik-Fk-Switch() -> (Arm-R_Ik-Fk-Switch:Editor Utility Check Box Object Reference)
    Module Name:Name=Arm R
   ) -> (then)
   "Custom on Clicked Event (ArmFKVisibility)"() {is=title K2Node_ComponentBoundEvent}
   > n17 = Flip Flop!(None:Exec) -> (A:Exec, B:Exec, IsA:bool=~) {is=title K2Node_MacroInstance}
   Visibility Buttons!(
    execute=[@n17.A, @n17.B]
    self
    Visibility Button:EUW Picker Button Object Reference=Get ArmFKVisibility()
    Buttons to Hide:Array of EUW Picker Button Object References=Get All FK Arms() -> (All FK Arms:Array of EUW Picker Button Object References)
    Bool:bool=@n17
   ) -> (then)
   "Custom on Clicked Event (ArmIKVisibility)"() {is=title K2Node_ComponentBoundEvent}
   > n18 = Flip Flop() -> (IsA=~)
   Visibility Buttons(
    execute=[@n18.A, @n18.B]
    Visibility Button=Get ArmIKVisibility()
    Buttons to Hide=Get All IK Arms() -> (All IK Arms:Array of EUW Picker Button Object References)
    Bool=@n18
   )
   "Custom on Clicked Event (SpineSecVisibility)"() {is=title K2Node_ComponentBoundEvent}
   > n19 = Flip Flop() -> (IsA=~)
   Visibility Buttons(
    execute=[@n19.A, @n19.B]
    Visibility Button=Get SpineSecVisibility()
    Buttons to Hide=Make Array!(
     "[0]"=Get Spine_Spine-01-Sec-FK() -> (Spine_Spine-01-Sec-FK:EUW Picker Button Object Reference)
     "[1]"=Get Spine_Spine-03-Sec-FK() -> (Spine_Spine-03-Sec-FK:EUW Picker Button Object Reference)
     "[2]"=Get Spine_Spine-02-Sec-FK() -> (Spine_Spine-02-Sec-FK:EUW Picker Button Object Reference)
     "[3]"=Get Spine_Spine-04-Sec-FK() -> (Spine_Spine-04-Sec-FK:EUW Picker Button Object Reference)
     "[4]"=Get Spine_Pelvis-Sec-FK() -> (Spine_Pelvis-Sec-FK:EUW Picker Button Object Reference)
     "[5]"=Get AllSecSpine() -> (AllSecSpine:EUW Picker Button Object Reference)
    ) -> (Array)
    Bool=@n19
   )
   "Custom on Clicked Event (LegFKVisibility)"() {is=title K2Node_ComponentBoundEvent}
   > n20 = Flip Flop() -> (IsA=~)
   Visibility Buttons(
    execute=[@n20.A, @n20.B]
    Visibility Button=Get LegFKVisibility()
    Buttons to Hide=Get All FK Legs() -> (All FK Legs:Array of EUW Picker Button Object References)
    Bool=@n20
   )
   "Custom on Clicked Event (ToggleCapsule_Button)"() {is=title K2Node_ComponentBoundEvent}
   > ToggleCapsule!(execute:Exec, self:Object Reference=Get As Actor() -> (As Actor:Actor Object Reference)) -> (then:Exec) {is=title K2Node_Message}
   > n21 = Flip Flop() -> (IsA=~)
   n22 = Get ToggleCapsule_Button()
   > n23 = Set!(
    execute
    Normal Tint_SpecifiedColor:Linear Color Structure=(R=0.0774,G=0.0313,B=0.0698,A=1)
    Normal Tint_ColorUseRule:ESlateColorStylingMode Enum
    self:EUW Picker Button Object Reference=@n22
   ) -> (then, Output_Get:Slate Color Structure)
   Set Visibility for Array!(
    execute=[@n24.then, @n23.then]
    self:EUW Asset Picker Base Object Reference
    " Array":Array of EUW Picker Button Object References
    bVisibility:bool=false
   ) -> (then) {is=title K2Node_CallFunction}
   > Update Selection Display!(execute, self:EUW Picker Element Object Reference=@n22, Selected:bool=false) -> (then, Out Selected:bool)
   n24 = Set(execute=@n21.B, Normal Tint_SpecifiedColor=(R=0.2423,G=0.0595,B=0.2122,A=1), self=@n22)
   "Custom on Clicked Event (FacePickerButton)"() {is=title K2Node_ComponentBoundEvent}
   > n25 = Open Face Picker!(execute, self, Error if not found:bool=true) -> (then, Success:bool, Out Face Picker:EUW Asset Picker Base Object Reference)
   > Set Out Face Picker(Out Face Picker:EUW Asset Picker Base Object Reference=@n25.Out Face Picker) -> (Output_Get:EUW Asset Picker Base Object Reference)
   "Custom on Clicked Event (LegIKVisibility)"() {is=title K2Node_ComponentBoundEvent}
   > n26 = Flip Flop() -> (IsA=~)
   Visibility Buttons(
    execute=[@n26.A, @n26.B]
    Visibility Button=Get LegIKVisibility()
    Buttons to Hide=Get All IK Legs() -> (All IK Legs:Array of EUW Picker Button Object References)
    Bool=@n26
   )
   "Custom on Clicked Event (PropVis_Button)"() {is=title K2Node_ComponentBoundEvent}
   > n27 = Flip Flop() -> (IsA=~)
   Visibility Buttons(
    execute=[@n27.A, @n27.B]
    Visibility Button=Get PropVis_Button()
    Buttons to Hide=Make Array!(
     "[0]"=Get Prop_Prop-Global() -> (Prop_Prop-Global:EUW Picker Button Object Reference)
     "[1]"=Get Prop_Aim() -> (Prop_Aim:EUW Picker Button Object Reference)
     "[2]"=Get Prop_Prop-Local() -> (Prop_Prop-Local:EUW Picker Button Object Reference)
     "[3]"=Get Prop_Prop() -> (Prop_Prop:EUW Picker Button Object Reference)
     "[4]"=Get Prop_Prop-Attach-01() -> (Prop_Prop-Attach-01:EUW Picker Button Object Reference)
     "[5]"=Get Prop_Prop-Attach-02() -> (Prop_Prop-Attach-02:EUW Picker Button Object Reference)
     "[6]":EUW Picker Button Object Reference=Get Prop_All() -> (Prop_All:EUW Picker Button Object Reference)
    ) -> (Array)
    Bool=@n27
   )
   "Custom on Clicked Event (bakeikfktool)"() {is=title K2Node_ComponentBoundEvent}
   n28 = Get DialogSlot() -> (DialogSlot:Named Slot Object Reference)
   > n29 = Branch(Condition=HasAnyChildren!(self:Panel Widget Object Reference=@n28) -> (ReturnValue:bool))
   > ClearChildren(self=@n28)
   n30 = Create Widget!(
    execute:Exec=@n29.else
    WorldContextObject:Object Reference
    Class:User Widget Class Reference=/AnimInEdTools/IkFkMatchingTool/EUW_IkFkMatching.EUW_IkFkMatching_C
    OwningPlayer:Player Controller Object Reference
    Primary Control Rig:Control Rig Object Reference=Get Control Rig()
    Animatable Creature:Animatable Creature Object Reference=Get Animatable Creature() -> (Animatable Creature:Animatable Creature Object Reference)
   ) -> (then:Exec, ReturnValue:EUW Ik Fk Matching Object Reference)
   > AddChild(self=Get DialogSlot(), Content=@n30)
   n31 = "On Check State Changed (Arm-L_Ik-Fk-Switch)"!() -> (OutputDelegate, then, bIsChecked:bool) {is=title K2Node_ComponentBoundEvent}
   n32 = Get Arm-L_Ik-Fk-Switch() -> (Arm-L_Ik-Fk-Switch:Editor Utility Check Box Object Reference)
   > Set Buttons Color From Checkbox State(
    CheckBox=@n32
    IK Buttons=Get Left Arm IK() -> (Left Arm IK:Array of EUW Picker Button Object References)
    FK Buttons=Get Left Arm FK() -> (Left Arm FK:Array of EUW Picker Button Object References)
    Color Mult=1
    Opacity=0.5
   )
   > Set Animation Channel Bool(
    ControlRig=Get Control Rig()
    Channel Name=Convert to MR naming(In name=GetDisplayName(Object=@n32)) -> (Out name=None)
    Value=@n31.bIsChecked
   )
   > IK FK Switch Text Update(Switch CheckBox=Get Arm-L_Ik-Fk-Switch(), Prefix=Left Arm)
   "Custom on Clicked Event (ZeroOutbutton)"() {is=title K2Node_ComponentBoundEvent}
   n33 = Break Control Rig Sequencer Binding Proxy(ControlRigSequencerBindingProxy=Get Control Rig Proxy Binding())
   > n34 = For Each Loop(Array=Get Controls(self=GetHierarchy(self=@n33.ControlRig), bTraverse=true))
   > n35 = Get Current Time(TimeUnit=DisplayRate)
   > n36 = GetFocusedLevelSequence()
   n37 = Break Control Rig Sequencer Binding Proxy(ControlRigSequencerBindingProxy=Get Control Rig Proxy Binding())
   n38 = Make Frame Number(Value=Break Movie Scene Sequence Playback Params(MovieSceneSequencePlaybackParams=@n35) -> (Frame_SubFrame=~, Time=~, bHasJumped=~))
   n39 = Break Rig Element Key(RigElementKey=@n34)
   > SetLocalControlRigEulerTransform(
    LevelSequence=@n36
    ControlRig=@n37.ControlRig
    ControlName=@n39.Name
    Frame=@n38
    Value_Location="0,0,0"
    Value_Rotation="0,0,0"
    Value_Scale="1,1,1"
    TimeUnit=DisplayRate
    bSetKey=false
   )
   > SetLocalControlRigRotator(LevelSequence=@n36, ControlRig=@n37.ControlRig, ControlName=@n39.Name, Frame=@n38, Value="0, 0, 0", TimeUnit=DisplayRate, bSetKey=false)
   "Custom on Clicked Event (ShowExtraButtons)"() {is=title K2Node_ComponentBoundEvent}
   > n40 = Flip Flop() -> (IsA=~)
   Visibility Buttons(
    execute=[@n40.A, @n40.B]
    Visibility Button=Get ShowExtraButtons()
    Buttons to Hide=Get All extra quick select buttons() -> (All extra quick select buttons:Array of EUW Picker Button Object References)
    Bool=@n40
   )
   "Custom on Clicked Event (BlockingMode)"() {is=title K2Node_ComponentBoundEvent}
   > n41 = Flip Flop() -> (IsA=~)
   Sequence!(execute=[@n41.A, @n41.B]) -> (then_0, then_1)
   > n42 = For Each Loop(Array:Array of Widget Object References=GetAllChildren(self=Get CanvasPanel_0() -> (CanvasPanel_0:Canvas Panel Object Reference))) -> (Array Element:Widget Object Reference)
   > n43 = Cast To EUW_Picker_Button!(execute, Object=@n42) -> (then, CastFailed, AsEUW Picker Button:EUW Picker Button Object Reference) {is=title K2Node_DynamicCast}
   > n44 = Branch(Condition=Contains Item(
     TargetArray:Array of EUW Picker Button Object References=Get BlockingMode Buttons()
     ItemToFind:"EUW Picker Button Object Reference (by ref)"=@n43
    ))
   n45 = Branch(execute=@n44.else, Condition=NOT Boolean(A=@n41))
   > SetVisibility(self=@n43, InVisibility=Visible)
   > Set Visibility for Array(
    " Array"=Make Array!("[0]"=Get AllIKLegExtras_r(), "[1]"=Get AllIKLegExtra_l(), "[2]"=Get AllIKLeg_r(), "[3]"=Get AllIKLeg_l()) -> (Array)
    bVisibility=true
   )
   n46 = Get BlockingMode()
   > n47 = Set(Normal Tint_SpecifiedColor=(R=0.2423,G=0.0595,B=0.2122,A=1), self=@n46)
   Set Visibility for Array(execute=[@n48.then, @n47.then], bVisibility=false)
   > Update Selection Display(self=@n46, Selected=false)
   SetVisibility(execute=@n45.else, self=@n43, InVisibility=Collapsed)
   > Set Visibility for Array(
    " Array"=Make Array("[0]"=Get AllIKLegExtras_r(), "[1]"=Get AllIKLegExtra_l(), "[2]"=Get AllIKLeg_r(), "[3]"=Get AllIKLeg_l())
    bVisibility=false
   )
   > n48 = Set(Normal Tint_SpecifiedColor=(R=0.0774,G=0.0313,B=0.0698,A=1), self=@n46)
   "Custom on Clicked Event (Snap_Arm_L)"() {is=title K2Node_ComponentBoundEvent}
   > Snap Event(Ik Fk Checkbox=Get Arm-L_Ik-Fk-Switch(), Module Name=Arm L)
   "Custom on Clicked Event (Match_Leg_R)"() {is=title K2Node_ComponentBoundEvent}
   > Match Event!(execute, self, Ik Fk Checkbox:Check Box Object Reference=Get Leg-R_Ik-Fk-Switch(), Module Name:Name=Leg R) -> (then)
   "Custom on Clicked Event (Match_Arm_R)"() {is=title K2Node_ComponentBoundEvent}
   > Match Event(Ik Fk Checkbox=Get Arm-R_Ik-Fk-Switch(), Module Name=Arm R)
   "Custom on Clicked Event (Match_Arm_L)"() {is=title K2Node_ComponentBoundEvent}
   > Match Event(Ik Fk Checkbox=Get Arm-L_Ik-Fk-Switch(), Module Name=Arm L)
   n49 = "On Check State Changed (Arm-R_Ik-Fk-Switch)"!() -> (OutputDelegate, then, bIsChecked:bool) {is=title K2Node_ComponentBoundEvent}
   n50 = Get Arm-R_Ik-Fk-Switch()
   > Set Buttons Color From Checkbox State(
    CheckBox=@n50
    IK Buttons=Get Right Arm IK() -> (Right Arm IK:Array of EUW Picker Button Object References)
    FK Buttons=Get Right Arm FK() -> (Right Arm FK:Array of EUW Picker Button Object References)
    Color Mult=1
    Opacity=0.5
   )
   > Set Animation Channel Bool(
    ControlRig=Get Control Rig()
    Channel Name=Convert to MR naming(In name=GetDisplayName(Object=@n50)) -> (Out name=None)
    Value=@n49.bIsChecked
   )
   > IK FK Switch Text Update(Switch CheckBox=Get Arm-R_Ik-Fk-Switch(), Prefix=Right Arm)
   "Custom on Clicked Event (FaceBoardButton)"() {is=title K2Node_ComponentBoundEvent}
   > Spawn Faceboard camera()
   "Custom on Clicked Event (Match_Leg_L)"() {is=title K2Node_ComponentBoundEvent}
   > Match Event(Ik Fk Checkbox=Get Leg-L_Ik-Fk-Switch(), Module Name=Leg L)
   "Custom on Clicked Event (Snap_Leg_L)"() {is=title K2Node_ComponentBoundEvent}
   > Snap Event(Ik Fk Checkbox=Get Leg-L_Ik-Fk-Switch(), Module Name=Leg L)
   n51 = "Custom on Clicked Event (ZeroOutselected)"() {is=title K2Node_ComponentBoundEvent}
   "Custom on Clicked Event (Snap_Leg_R)"() {is=title K2Node_ComponentBoundEvent}
   > Snap Event(Ik Fk Checkbox=Get Leg-R_Ik-Fk-Switch(), Module Name=Leg R)
   "Custom on Clicked Event (TwistsVisibility)"() {is=title K2Node_ComponentBoundEvent}
   > n52 = Flip Flop() -> (IsA=~)
   Visibility Buttons(
    execute=[@n52.A, @n52.B]
    Visibility Button=Get TwistsVisibility()
    Buttons to Hide=Make Array!(
     "[0]"=Get Leg-Upper-Twist-R_Twist-2() -> (Leg-Upper-Twist-R_Twist-2:EUW Picker Button Object Reference)
     "[1]"=Get Leg-Upper-Twist-R_Twist-1() -> (Leg-Upper-Twist-R_Twist-1:EUW Picker Button Object Reference)
     "[2]"=Get Leg-Upper-Twist-L_Twist-2() -> (Leg-Upper-Twist-L_Twist-2:EUW Picker Button Object Reference)
     "[3]"=Get Leg-Upper-Twist-L_Twist-1() -> (Leg-Upper-Twist-L_Twist-1:EUW Picker Button Object Reference)
     "[4]":EUW Picker Button Object Reference=Get Leg-Lower-Twist-R_Twist-2() -> (Leg-Lower-Twist-R_Twist-2:EUW Picker Button Object Reference)
     "[5]":EUW Picker Button Object Reference=Get Leg-Lower-Twist-R_Twist-1() -> (Leg-Lower-Twist-R_Twist-1:EUW Picker Button Object Reference)
     "[6]":EUW Picker Button Object Reference=Get Leg-Lower-Twist-L_Twist-2() -> (Leg-Lower-Twist-L_Twist-2:EUW Picker Button Object Reference)
     "[7]":EUW Picker Button Object Reference=Get Leg-Lower-Twist-L_Twist-1() -> (Leg-Lower-Twist-L_Twist-1:EUW Picker Button Object Reference)
     "[8]":EUW Picker Button Object Reference=Get Arm-Upper-Twist-R_Twist-2() -> (Arm-Upper-Twist-R_Twist-2:EUW Picker Button Object Reference)
     "[9]":EUW Picker Button Object Reference=Get Arm-Upper-Twist-R_Twist-1() -> (Arm-Upper-Twist-R_Twist-1:EUW Picker Button Object Reference)
     "[10]":EUW Picker Button Object Reference=Get Arm-Lower-Twist-R_Twist-2() -> (Arm-Lower-Twist-R_Twist-2:EUW Picker Button Object Reference)
     "[11]":EUW Picker Button Object Reference=Get Arm-Lower-Twist-R_Twist-1() -> (Arm-Lower-Twist-R_Twist-1:EUW Picker Button Object Reference)
     "[12]":EUW Picker Button Object Reference=Get Arm-Upper-Twist-L_Twist-2() -> (Arm-Upper-Twist-L_Twist-2:EUW Picker Button Object Reference)
     "[13]":EUW Picker Button Object Reference=Get Arm-Upper-Twist-L_Twist-1() -> (Arm-Upper-Twist-L_Twist-1:EUW Picker Button Object Reference)
     "[14]":EUW Picker Button Object Reference=Get Arm-Lower-Twist-L_Twist-2() -> (Arm-Lower-Twist-L_Twist-2:EUW Picker Button Object Reference)
     "[15]":EUW Picker Button Object Reference=Get Arm-Lower-Twist-L_Twist-1() -> (Arm-Lower-Twist-L_Twist-1:EUW Picker Button Object Reference)
    ) -> (Array)
    Bool=@n52
   )
   n53 = Event Construct()
   n54 = GetFocusedLevelSequence(execute=[@n57.CastFailed, @n53.then, @n58.then])
   > Set LevelSequencer(LevelSequencer:Level Sequence Object Reference=@n54) -> (Output_Get:Level Sequence Object Reference)
   > ClearChildren(self=Get FaceSlot() -> (FaceSlot:Vertical Box Object Reference))
   > n55 = Open Face Picker(Error if not found=false)
   > Set Out Face Picker(Out Face Picker=@n55.Out Face Picker)
   > Cache IK FK Buttons() {is=title K2Node_CallFunction}
   > Setup MR Naming For Buttons!(
    execute
    self:EUW Asset Picker Base Object Reference
    Main Canvas:Panel Widget Object Reference=Get CanvasPanel_0()
    Control Rig:Control Rig Object Reference=Get Control Rig()
    Excluded Buttons:Array of Widget Object References=Get All extra quick select buttons()
   ) -> (then)
   > Init Sets Buttons()
   > Set Toggle Capsule Button Vis() {is=title K2Node_CallFunction}
   > Refresh All IK FK Widgets()
   n56 = Event Widget has loaded!() -> (
    OutputDelegate
    then
    Control Rig Proxy Binding:Control Rig Sequencer Binding Proxy Structure
    Animatable Creature:Animatable Creature Object Reference
    Anim Picker EUW:Animation Picker Object Reference
   ) {is=title K2Node_Event}
   > Parent: Widget has loaded!(
    execute
    Control Rig Proxy Binding:Control Rig Sequencer Binding Proxy Structure=@n56.Control Rig Proxy Binding
    Animatable Creature:Animatable Creature Object Reference=@n56.Animatable Creature
    Anim Picker EUW:Animation Picker Object Reference=@n56.Anim Picker EUW
   ) -> (then) {is=title K2Node_CallParentFunction}
   > Set Control Rig(Control Rig:Control Rig Object Reference=Get!() -> (
     Control Rig Proxy Binding_Proxy:Movie Scene Binding Proxy Structure
     Control Rig Proxy Binding_ControlRig:Control Rig Object Reference
     Control Rig Proxy Binding_Track:Movie Scene Control Rig Parameter Track Object Reference
    ).Control Rig Proxy Binding_ControlRig) -> (Output_Get:Control Rig Object Reference)
   > n57 = Cast To ModularRig!(execute, Object=Get Control Rig()) -> (then, CastFailed, AsModular Rig:Modular Rig Object Reference) {is=title K2Node_DynamicCast}
   > n58 = Set Modular Rig(Modular Rig:Modular Rig Object Reference=@n57) -> (Output_Get:Modular Rig Object Reference)
   Bind Event to On Global Time Changed(
    self:Level Sequence Editor Subsystem Object Reference=Get LevelSequenceEditorSubsystem() -> (ReturnValue:Level Sequence Editor Subsystem Object Reference) {is=title K2Node_GetEditorSubsystem}
    Delegate=Create Event()
   ) {is=title K2Node_AddDelegate}
   IsControlSelected!(self:Control Rig Object Reference=@n1.ControlRig, InControlName:"Name (by ref)"=None) -> (ReturnValue:bool)
  ]}
 Get Primary Panel {nodes [
   Get Primary Panel() {is=title K2Node_FunctionEntry}
   > Return Node!(execute, Panel:Panel Widget Object Reference=Get CanvasPanel_0())
  ]}
 IK FK Switch Text Update {nodes [
   n1 = IK FK Switch Text Update!() -> (then, Switch CheckBox:Editor Utility Check Box Object Reference, Prefix:String) {is=title K2Node_FunctionEntry}
   > n2 = Cast To TextBlock!(execute, Object=GetChildAt!(self:Panel Widget Object Reference=@n1, Index:int=0) -> (ReturnValue:Widget Object Reference)) -> (then, CastFailed, AsText:Text Object Reference) {is=title K2Node_DynamicCast}
   n3 = SelectString(A=" IK", B=" FK", bPickA=IsChecked(self=@n1))
   > Set Text(Text="To Text (String)"(InString=Append(A=@n1.Prefix, B=@n3)), self=@n2)
  ]}
 Init Sets Buttons {nodes [
   Init Sets Buttons() {is=title K2Node_FunctionEntry}
   > n1 = Set IK Move Buttons(IK Move Buttons:Array of EUW Picker Button Object References=Make Array!("[0]"=Get Body(), "[1]"=Get Arm-R_IK(), "[2]"=Get Arm-L_IK(), "[3]"=Get Leg-R_IK(), "[4]"=Get Leg-L_IK()) -> (Array)) -> (Output_Get:Array of EUW Picker Button Object References)
   > Select Multiple Buttons!(
    execute
    self:EUW Asset Picker Base Object Reference
    Button:EUW Picker Button Object Reference=Get IKMove()
    Buttons to Select:Array of EUW Picker Button Object References=@n1
   ) -> (then)
   > n2 = Set Twists(Twists:Array of EUW Picker Button Object References=Make Array!(
     "[0]"=Get Arm-Lower-Twist-L_Twist-2()
     "[1]"=Get Arm-Lower-Twist-L_Twist-1()
     "[2]"=Get Leg-Upper-Twist-R_Twist-2()
     "[3]"=Get Leg-Upper-Twist-R_Twist-1()
     "[4]"=Get Leg-Upper-Twist-L_Twist-2()
     "[5]":EUW Picker Button Object Reference=Get Leg-Upper-Twist-L_Twist-1()
     "[6]":EUW Picker Button Object Reference=Get Leg-Lower-Twist-R_Twist-2()
     "[7]":EUW Picker Button Object Reference=Get Leg-Lower-Twist-R_Twist-1()
     "[8]":EUW Picker Button Object Reference=Get Leg-Lower-Twist-L_Twist-2()
     "[9]":EUW Picker Button Object Reference=Get Leg-Lower-Twist-L_Twist-1()
     "[10]":EUW Picker Button Object Reference=Get Arm-Upper-Twist-R_Twist-2()
     "[11]":EUW Picker Button Object Reference=Get Arm-Upper-Twist-R_Twist-1()
     "[12]":EUW Picker Button Object Reference=Get Arm-Lower-Twist-R_Twist-2()
     "[13]":EUW Picker Button Object Reference=Get Arm-Lower-Twist-R_Twist-1()
     "[14]":EUW Picker Button Object Reference=Get Arm-Upper-Twist-L_Twist-2()
     "[15]":EUW Picker Button Object Reference=Get Arm-Upper-Twist-L_Twist-1()
    ) -> (Array)) -> (Output_Get:Array of EUW Picker Button Object References)
   > Select Multiple Buttons(Button=Get AllTwist() -> (AllTwist:EUW Picker Button Object Reference), Buttons to Select=@n2)
   > n3 = Set FK Hands(FK Hands:Array of EUW Picker Button Object References=Make Array!("[0]"=Get Arm-L_FK-2(), "[1]"=Get Arm-R_FK-2()) -> (Array)) -> (Output_Get:Array of EUW Picker Button Object References)
   > Select Multiple Buttons(Button=Get FKHands() -> (FKHands:EUW Picker Button Object Reference), Buttons to Select=@n3)
   > n4 = Set 4 Fingers Right(4 Fingers Right:Array of EUW Picker Button Object References=Make Array!(
     "[0]"=Get Index-R_FK-2() -> (Index-R_FK-2:EUW Picker Button Object Reference)
     "[1]"=Get Index-R_FK-1() -> (Index-R_FK-1:EUW Picker Button Object Reference)
     "[2]":EUW Picker Button Object Reference=Get Index-R_FK-0() -> (Index-R_FK-0:EUW Picker Button Object Reference)
     "[3]":EUW Picker Button Object Reference=Get Ring-R_FK-2() -> (Ring-R_FK-2:EUW Picker Button Object Reference)
     "[4]":EUW Picker Button Object Reference=Get Ring-R_FK-1() -> (Ring-R_FK-1:EUW Picker Button Object Reference)
     "[5]":EUW Picker Button Object Reference=Get Ring-R_FK-0() -> (Ring-R_FK-0:EUW Picker Button Object Reference)
     "[6]":EUW Picker Button Object Reference=Get Middle-R_FK-2() -> (Middle-R_FK-2:EUW Picker Button Object Reference)
     "[7]":EUW Picker Button Object Reference=Get Middle-R_FK-1() -> (Middle-R_FK-1:EUW Picker Button Object Reference)
     "[8]":EUW Picker Button Object Reference=Get Middle-R_FK-0() -> (Middle-R_FK-0:EUW Picker Button Object Reference)
     "[9]":EUW Picker Button Object Reference=Get Pinky-R_FK-2() -> (Pinky-R_FK-2:EUW Picker Button Object Reference)
     "[10]":EUW Picker Button Object Reference=Get Pinky-R_FK-1() -> (Pinky-R_FK-1:EUW Picker Button Object Reference)
     "[11]":EUW Picker Button Object Reference=Get Pinky-R_FK-0() -> (Pinky-R_FK-0:EUW Picker Button Object Reference)
    ) -> (Array)) -> (Output_Get:Array of EUW Picker Button Object References)
   > Select Multiple Buttons(Button=Get 4Fingers_r(), Buttons to Select=@n4)
   > n5 = Set 4 Fingers Left(4 Fingers Left:Array of EUW Picker Button Object References=Make Array(
     "[0]"=Get Index-L_FK-2() -> (Index-L_FK-2:EUW Picker Button Object Reference)
     "[1]"=Get Index-L_FK-1() -> (Index-L_FK-1:EUW Picker Button Object Reference)
     "[2]"=Get Index-L_FK-0() -> (Index-L_FK-0:EUW Picker Button Object Reference)
     "[3]"=Get Middle-L_FK-2() -> (Middle-L_FK-2:EUW Picker Button Object Reference)
     "[4]"=Get Middle-L_FK-1() -> (Middle-L_FK-1:EUW Picker Button Object Reference)
     "[5]"=Get Middle-L_FK-0() -> (Middle-L_FK-0:EUW Picker Button Object Reference)
     "[6]"=Get Ring-L_FK-2() -> (Ring-L_FK-2:EUW Picker Button Object Reference)
     "[7]"=Get Ring-L_FK-1() -> (Ring-L_FK-1:EUW Picker Button Object Reference)
     "[8]"=Get Ring-L_FK-0() -> (Ring-L_FK-0:EUW Picker Button Object Reference)
     "[9]"=Get Pinky-L_FK-2() -> (Pinky-L_FK-2:EUW Picker Button Object Reference)
     "[10]"=Get Pinky-L_FK-1() -> (Pinky-L_FK-1:EUW Picker Button Object Reference)
     "[11]"=Get Pinky-L_FK-0() -> (Pinky-L_FK-0:EUW Picker Button Object Reference)
    )) -> (Output_Get:Array of EUW Picker Button Object References)
   > Select Multiple Buttons(Button=Get 4Fingers_l(), Buttons to Select=@n5)
   > Set All extra quick select buttons(All extra quick select buttons:Array of EUW Picker Button Object References=Make Array!(
     "[0]"=Get IKToes() -> (IKToes:EUW Picker Button Object Reference)
     "[1]"=Get IKMove()
     "[2]"=Get IKKnees() -> (IKKnees:EUW Picker Button Object Reference)
     "[3]"=Get IKHands() -> (IKHands:EUW Picker Button Object Reference)
     "[4]"=Get IKFeet() -> (IKFeet:EUW Picker Button Object Reference)
     "[5]"=Get FKToes() -> (FKToes:EUW Picker Button Object Reference)
     "[6]"=Get FKThighs() -> (FKThighs:EUW Picker Button Object Reference)
     "[7]"=Get FKKnees() -> (FKKnees:EUW Picker Button Object Reference)
     "[8]"=Get FKHands()
     "[9]"=Get FKFeets() -> (FKFeets:EUW Picker Button Object Reference)
     "[10]"=Get FKElbows() -> (FKElbows:EUW Picker Button Object Reference)
     "[11]"=Get FKBiceps() -> (FKBiceps:EUW Picker Button Object Reference)
     "[12]":EUW Picker Button Object Reference=Get FKArms() -> (FKArms:EUW Picker Button Object Reference)
     "[13]":EUW Picker Button Object Reference=Get ArmsFingers_r() -> (ArmsFingers_r:EUW Picker Button Object Reference)
     "[14]":EUW Picker Button Object Reference=Get ArmsFingers_l() -> (ArmsFingers_l:EUW Picker Button Object Reference)
     "[15]":EUW Picker Button Object Reference=Get AllWeaponControlsButton() -> (AllWeaponControlsButton:EUW Picker Button Object Reference)
     "[16]":EUW Picker Button Object Reference=Get AllTwist()
     "[17]":EUW Picker Button Object Reference=Get AllLegs() -> (AllLegs:EUW Picker Button Object Reference)
     "[18]":EUW Picker Button Object Reference=Get AllIKLegs() -> (AllIKLegs:EUW Picker Button Object Reference)
     "[19]":EUW Picker Button Object Reference=Get AllIKLegExtras_r()
     "[20]":EUW Picker Button Object Reference=Get AllIKLegExtra_l()
     "[21]":EUW Picker Button Object Reference=Get AllIKLeg_r()
     "[22]":EUW Picker Button Object Reference=Get AllIKLeg_l()
     "[23]":EUW Picker Button Object Reference=Get AllFKLegs() -> (AllFKLegs:EUW Picker Button Object Reference)
     "[24]":EUW Picker Button Object Reference=Get AllFKLeg_r()
     "[25]":EUW Picker Button Object Reference=Get AllFKLeg_l()
     "[26]":EUW Picker Button Object Reference=Get allFingers_r()
     "[27]":EUW Picker Button Object Reference=Get AllFingers_l()
     "[28]":EUW Picker Button Object Reference=Get AllFingers() -> (AllFingers:EUW Picker Button Object Reference)
     "[29]":EUW Picker Button Object Reference=Get AllClavicles() -> (AllClavicles:EUW Picker Button Object Reference)
     "[30]":EUW Picker Button Object Reference=Get AllArms() -> (AllArms:EUW Picker Button Object Reference)
     "[31]":EUW Picker Button Object Reference=Get AllArm_r() -> (AllArm_r:EUW Picker Button Object Reference)
     "[32]":EUW Picker Button Object Reference=Get AllArm_l() -> (AllArm_l:EUW Picker Button Object Reference)
     "[33]":EUW Picker Button Object Reference=Get 4Fingers_r()
     "[34]":EUW Picker Button Object Reference=Get 4Fingers_l()
     "[35]":EUW Picker Button Object Reference=Get IKArms() -> (IKArms:EUW Picker Button Object Reference)
     "[36]":EUW Picker Button Object Reference=Get IKElbows() -> (IKElbows:EUW Picker Button Object Reference)
     "[37]":EUW Picker Button Object Reference=Get IKFeetExtras() -> (IKFeetExtras:EUW Picker Button Object Reference)
     "[38]":EUW Picker Button Object Reference=Get AllLeg_l() -> (AllLeg_l:EUW Picker Button Object Reference)
     "[39]":EUW Picker Button Object Reference=Get AllSpine() -> (AllSpine:EUW Picker Button Object Reference)
     "[40]":EUW Picker Button Object Reference=Get AllSecSpine()
     "[41]":EUW Picker Button Object Reference=Get AllIKSpine() -> (AllIKSpine:EUW Picker Button Object Reference)
     "[42]":EUW Picker Button Object Reference=Get AllFKSpine() -> (AllFKSpine:EUW Picker Button Object Reference)
     "[43]":EUW Picker Button Object Reference=Get AllIKRArm()
     "[44]":EUW Picker Button Object Reference=Get AllFKRArm()
     "[45]":EUW Picker Button Object Reference=Get AllFKLArm()
     "[46]":EUW Picker Button Object Reference=Get AllIKLArm()
     "[47]":EUW Picker Button Object Reference=Get IKBases() -> (IKBases:EUW Picker Button Object Reference)
     "[48]":EUW Picker Button Object Reference=Get AllFKLeg_l()
     "[49]":EUW Picker Button Object Reference=Get AllLeg_r() -> (AllLeg_r:EUW Picker Button Object Reference)
     "[50]":EUW Picker Button Object Reference=Get AllFKHead() -> (AllFKHead:EUW Picker Button Object Reference)
     "[51]":EUW Picker Button Object Reference=Get AllIKHead() -> (AllIKHead:EUW Picker Button Object Reference)
    ) -> (Array)) -> (Output_Get:Array of EUW Picker Button Object References)
   > Select Multiple Buttons(
    Button=Get Prop_All()
    Buttons to Select=Make Array!(
     "[0]"=Get Prop_Prop-Global()
     "[1]"=Get Prop_Aim()
     "[2]"=Get Prop_Prop-Local()
     "[3]"=Get Prop_Prop()
     "[4]"=Get Prop_Prop-Attach-01()
     "[5]"=Get Prop_Prop-Attach-02()
    ) -> (Array)
   )
   > Select Multiple Buttons(
    Button=Get AllWeaponControlsButton()
    Buttons to Select=Make Array!("[0]"=Get Attach_Weapon-R(), "[1]"=Get Attach_Hand-Attach-R(), "[2]"=Get Attach_Weapon-L(), "[3]"=Get Attach_Hand-Attach-L()) -> (Array)
   )
   > n6 = Sequence!(execute) -> (then_0, then_1, then_2:Exec, then_3:Exec, then_4:Exec, then_5:Exec, then_6:Exec)
   > Select Multiple Buttons(Button=Get AllIKHead(), Buttons to Select=Get Neck IK() -> (Neck IK:Array of EUW Picker Button Object References))
   > Select Multiple Buttons(Button=Get AllFKHead(), Buttons to Select=Get Neck FK())
   > Select Multiple Buttons(
    Button=Get AllSpine()
    Buttons to Select=Make Array!(
     "[0]"=Get Spine_Start-IK()
     "[1]"=Get Spine_Spine-04-Sec-FK()
     "[2]"=Get Spine_Spine-05-Sec-FK() -> (Spine_Spine-05-Sec-FK:EUW Picker Button Object Reference)
     "[3]"=Get Spine_Spine-02-Sec-FK()
     "[4]":EUW Picker Button Object Reference=Get Spine_Spine-03-Sec-FK()
     "[5]":EUW Picker Button Object Reference=Get Spine_Spine-01-Sec-FK()
     "[6]":EUW Picker Button Object Reference=Get Spine_Pelvis-Local() -> (Spine_Pelvis-Local:EUW Picker Button Object Reference)
     "[7]":EUW Picker Button Object Reference=Get Spine_Mid-IK()
     "[8]":EUW Picker Button Object Reference=Get Spine_End-FK()
     "[9]":EUW Picker Button Object Reference=Get Spine_Mid-FK()
     "[10]":EUW Picker Button Object Reference=Get Spine_Start-FK()
     "[11]":EUW Picker Button Object Reference=Get Spine_End-IK()
     "[12]":EUW Picker Button Object Reference=Get Spine_Pelvis-Sec-FK()
    ) -> (Array)
   )
   > Select Multiple Buttons(Button=Get AllIKSpine(), Buttons to Select=Get Spine IK() -> (Spine IK:Array of EUW Picker Button Object References))
   > Select Multiple Buttons(Button=Get AllFKSpine(), Buttons to Select=Get Spine FK())
   > Select Multiple Buttons(
    Button=Get AllSecSpine()
    Buttons to Select=Make Array!(
     "[0]"=Get Spine_Pelvis-Local()
     "[1]"=Get Spine_Spine-01-Sec-FK()
     "[2]"=Get Spine_Spine-03-Sec-FK()
     "[3]"=Get Spine_Spine-02-Sec-FK()
     "[4]"=Get Spine_Spine-05-Sec-FK()
     "[5]"=Get Spine_Spine-04-Sec-FK()
     "[6]"=Get Spine_Pelvis-Sec-FK()
    ) -> (Array)
   )
   Select Multiple Buttons(
    execute=@n6.then_1
    Button=Get AllClavicles()
    Buttons to Select=Make Array!(
     "[0]"=Get Clavicle-L_FK() -> (Clavicle-L_FK:EUW Picker Button Object Reference)
     "[1]"=Get Clavicle-R_FK() -> (Clavicle-R_FK:EUW Picker Button Object Reference)
    ) -> (Array)
   )
   > n7 = Set All Arms(All Arms:Array of EUW Picker Button Object References=Get All Arm Left() -> (All Arm Left:Array of EUW Picker Button Object References)) -> (Output_Get:Array of EUW Picker Button Object References)
   > Append Array(TargetArray=@n7, SourceArray=Get All Arm Right() -> (All Arm Right:Array of EUW Picker Button Object References))
   > Select Multiple Buttons(Button=Get AllArms(), Buttons to Select=@n7)
   > n8 = Set All Fingers(All Fingers:Array of EUW Picker Button Object References=Get All Fingers Left() -> (All Fingers Left:Array of EUW Picker Button Object References)) -> (Output_Get:Array of EUW Picker Button Object References)
   > Append Array(TargetArray=@n8, SourceArray=Get All Fingers Right() -> (All Fingers Right:Array of EUW Picker Button Object References))
   > Select Multiple Buttons(Button=Get AllFingers(), Buttons to Select=@n8)
   > n9 = Set All Arm Right(All Arm Right:Array of EUW Picker Button Object References=Make Array!(
     "[0]"=Get Arm-R_FK-2()
     "[1]"=Get Arm-R_FK-1()
     "[2]":EUW Picker Button Object Reference=Get Arm-R_FK-0()
     "[3]":EUW Picker Button Object Reference=Get Arm-R_IK()
     "[4]":EUW Picker Button Object Reference=Get Arm-R_PV()
     "[5]":EUW Picker Button Object Reference=Get Arm-R_Mid()
     "[6]":EUW Picker Button Object Reference=Get Clavicle-R_FK()
     "[7]":EUW Picker Button Object Reference=Get Arm-R_IK-Rotation()
     "[8]":EUW Picker Button Object Reference=Get Arm-R_FK-0-Gimbal()
    ) -> (Array)) -> (Output_Get:Array of EUW Picker Button Object References)
   > Select Multiple Buttons(Button=Get AllArm_r(), Buttons to Select=@n9)
   > n10 = Set All Arm Fingers Right(All Arm Fingers Right:Array of EUW Picker Button Object References=Get All Arm Right()) -> (Output_Get:Array of EUW Picker Button Object References)
   > Append Array(TargetArray=@n10, SourceArray=Get All Fingers Right())
   > Select Multiple Buttons(Button=Get ArmsFingers_r(), Buttons to Select=@n10)
   > n11 = Set All Arm Left(All Arm Left:Array of EUW Picker Button Object References=Make Array(
     "[0]"=Get Arm-L_FK-2()
     "[1]"=Get Arm-L_FK-1()
     "[2]"=Get Arm-L_FK-0()
     "[3]"=Get Arm-L_IK()
     "[4]"=Get Arm-L_PV()
     "[5]"=Get Arm-L_Mid()
     "[6]"=Get Clavicle-L_FK()
     "[7]"=Get Arm-L_IK-Rotation()
     "[8]"=Get Arm-L_FK-0-Gimbal()
    )) -> (Output_Get:Array of EUW Picker Button Object References)
   > Select Multiple Buttons(Button=Get AllArm_l(), Buttons to Select=@n11)
   > n12 = Set All Arm Fingers Left(All Arm Fingers Left:Array of EUW Picker Button Object References=Get All Arm Left()) -> (Output_Get:Array of EUW Picker Button Object References)
   > Append Array(TargetArray=@n12, SourceArray=Get All Fingers Left())
   > Select Multiple Buttons(Button=Get ArmsFingers_l(), Buttons to Select=@n12)
   n13 = Set All FK Arms(
    execute=@n6.then_2
    All FK Arms:Array of EUW Picker Button Object References=Make Array!(
     "[0]"=Get Arm-L_FK-2()
     "[1]"=Get Arm-L_FK-1()
     "[2]"=Get Arm-L_FK-0()
     "[3]"=Get Arm-R_FK-2()
     "[4]"=Get Arm-R_FK-1()
     "[5]"=Get Arm-R_FK-0()
     "[6]"=Get Arm-L_FK-0-Gimbal()
     "[7]"=Get Arm-R_FK-0-Gimbal()
     "[8]"=Get AllFKRArm()
     "[9]":EUW Picker Button Object Reference=Get AllFKLArm()
     "[10]":EUW Picker Button Object Reference=Get FKArms()
     "[11]":EUW Picker Button Object Reference=Get FKBiceps()
     "[12]":EUW Picker Button Object Reference=Get FKElbows()
     "[13]":EUW Picker Button Object Reference=Get FKHands()
    ) -> (Array)
   ) -> (Output_Get:Array of EUW Picker Button Object References)
   > Select Multiple Buttons(Button=Get FKArms(), Buttons to Select=@n13)
   > Select Multiple Buttons(Button=Get FKElbows(), Buttons to Select=Make Array!("[0]"=Get Arm-L_FK-1(), "[1]"=Get Arm-R_FK-1()) -> (Array))
   > Select Multiple Buttons(Button=Get FKBiceps(), Buttons to Select=Make Array("[0]"=Get Arm-L_FK-0(), "[1]"=Get Arm-R_FK-0()))
   > Select Multiple Buttons(Button=Get FKHands(), Buttons to Select=Make Array("[0]"=Get Arm-L_FK-2(), "[1]"=Get Arm-R_FK-2()))
   > Select Multiple Buttons(
    Button=Get AllFKLArm()
    Buttons to Select=Make Array!(
     "[0]"=Get Arm-L_FK-2()
     "[1]"=Get Arm-L_FK-1()
     "[2]":EUW Picker Button Object Reference=Get Arm-L_FK-0()
     "[3]":EUW Picker Button Object Reference=Get Clavicle-L_FK()
     "[4]":EUW Picker Button Object Reference=Get Arm-L_FK-0-Gimbal()
    ) -> (Array)
   )
   > Select Multiple Buttons(
    Button=Get AllFKRArm()
    Buttons to Select=Make Array("[0]"=Get Arm-R_FK-2(), "[1]"=Get Arm-R_FK-1(), "[2]"=Get Arm-R_FK-0(), "[3]"=Get Clavicle-R_FK(), "[4]"=Get Arm-R_FK-0-Gimbal())
   )
   > n14 = Set All IK Arms(All IK Arms:Array of EUW Picker Button Object References=Make Array!(
     "[0]"=Get Arm-R_IK()
     "[1]"=Get Arm-L_IK()
     "[2]"=Get Arm-R_PV()
     "[3]"=Get Arm-L_PV()
     "[4]"=Get IKArms()
     "[5]":EUW Picker Button Object Reference=Get IKElbows()
     "[6]":EUW Picker Button Object Reference=Get IKHands()
     "[7]":EUW Picker Button Object Reference=Get AllIKRArm()
     "[8]":EUW Picker Button Object Reference=Get AllIKLArm()
    ) -> (Array)) -> (Output_Get:Array of EUW Picker Button Object References)
   > Select Multiple Buttons(Button=Get IKArms(), Buttons to Select=@n14)
   > Select Multiple Buttons(Button=Get IKElbows(), Buttons to Select=Make Array!("[0]"=Get Arm-R_PV(), "[1]"=Get Arm-L_PV()) -> (Array))
   > Select Multiple Buttons(Button=Get IKHands(), Buttons to Select=Make Array("[0]"=Get Arm-L_IK(), "[1]"=Get Arm-R_IK()))
   > Select Multiple Buttons(
    Button=Get AllIKRArm()
    Buttons to Select=Make Array!(
     "[0]"=Get Arm-R_IK()
     "[1]"=Get Arm-R_PV()
     "[2]":EUW Picker Button Object Reference=Get Arm-R_Mid()
     "[3]":EUW Picker Button Object Reference=Get Clavicle-R_FK()
     "[4]":EUW Picker Button Object Reference=Get Arm-R_IK-Rotation()
    ) -> (Array)
   )
   > Select Multiple Buttons(
    Button=Get AllIKLArm()
    Buttons to Select=Make Array("[0]"=Get Arm-L_IK(), "[1]"=Get Arm-L_PV(), "[2]"=Get Arm-L_Mid(), "[3]"=Get Clavicle-L_FK(), "[4]"=Get Arm-L_IK-Rotation())
   )
   > n15 = Set All Legs(All Legs:Array of EUW Picker Button Object References=Get All IK Leg Left() -> (All IK Leg Left:Array of EUW Picker Button Object References)) -> (Output_Get:Array of EUW Picker Button Object References)
   > Append Array(TargetArray=@n15, SourceArray=Get All IK Leg Right() -> (All IK Leg Right:Array of EUW Picker Button Object References))
   > Append Array(TargetArray=@n15, SourceArray=Get All IK Leg Extras Right() -> (All IK Leg Extras Right:Array of EUW Picker Button Object References))
   > Append Array(TargetArray=@n15, SourceArray=Get All IK Leg Left())
   > Append Array(TargetArray=@n15, SourceArray=Get All IK Leg Extras Left() -> (All IK Leg Extras Left:Array of EUW Picker Button Object References))
   > Append Array(TargetArray=@n15, SourceArray=Get All FK Leg Right() -> (All FK Leg Right:Array of EUW Picker Button Object References))
   > Append Array(TargetArray=@n15, SourceArray=Get All FK Leg Left() -> (All FK Leg Left:Array of EUW Picker Button Object References))
   > n16 = Set All Leg Left(All Leg Left:Array of EUW Picker Button Object References=Get All IK Leg Left()) -> (Output_Get:Array of EUW Picker Button Object References)
   > Append Array(TargetArray=@n16, SourceArray=Get All FK Leg Left())
   > Select Multiple Buttons(Button=Get AllLeg_l(), Buttons to Select=Get All Leg Left() -> (All Leg Left:Array of EUW Picker Button Object References))
   > n17 = Set All Leg Right(All Leg Right:Array of EUW Picker Button Object References=Get All IK Leg Right()) -> (Output_Get:Array of EUW Picker Button Object References)
   > Append Array(TargetArray=@n17, SourceArray=Get All FK Leg Right())
   > Select Multiple Buttons(Button=Get AllLeg_r(), Buttons to Select=Get All Leg Right() -> (All Leg Right:Array of EUW Picker Button Object References))
   Set Thumb Right(
    execute=@n6.then_3
    Thumb Right:Array of EUW Picker Button Object References=Make Array!(
     "[0]"=Get Thumb-R_FK-2() -> (Thumb-R_FK-2:EUW Picker Button Object Reference)
     "[1]"=Get Thumb-R_FK-1() -> (Thumb-R_FK-1:EUW Picker Button Object Reference)
     "[2]"=Get Thumb-R_FK-0() -> (Thumb-R_FK-0:EUW Picker Button Object Reference)
    ) -> (Array)
   ) -> (Output_Get:Array of EUW Picker Button Object References)
   > n18 = Set All Fingers Right(All Fingers Right:Array of EUW Picker Button Object References=Get 4 Fingers Right() -> (4 Fingers Right:Array of EUW Picker Button Object References)) -> (Output_Get:Array of EUW Picker Button Object References)
   > Append Array(TargetArray=@n18, SourceArray=Get Thumb Right() -> (Thumb Right:Array of EUW Picker Button Object References))
   > Append Array(
    TargetArray=@n18
    SourceArray=Make Array!(
     "[0]"=Get Meta-R_Pinky-Metacarpal-R() -> (Meta-R_Pinky-Metacarpal-R:EUW Picker Button Object Reference)
     "[1]"=Get Meta-R_Ring-Metacarpal-R() -> (Meta-R_Ring-Metacarpal-R:EUW Picker Button Object Reference)
     "[2]"=Get Meta-R_Middle-Metacarpal-R() -> (Meta-R_Middle-Metacarpal-R:EUW Picker Button Object Reference)
     "[3]":EUW Picker Button Object Reference=Get Meta-R_Index-Metacarpal-R() -> (Meta-R_Index-Metacarpal-R:EUW Picker Button Object Reference)
    ) -> (Array)
   )
   > Select Multiple Buttons(Button=Get allFingers_r(), Buttons to Select=@n18)
   > Set Thumb Left(Thumb Left:Array of EUW Picker Button Object References=Make Array!(
     "[0]"=Get Thumb-L_FK-2() -> (Thumb-L_FK-2:EUW Picker Button Object Reference)
     "[1]"=Get Thumb-L_FK-1() -> (Thumb-L_FK-1:EUW Picker Button Object Reference)
     "[2]"=Get Thumb-L_FK-0() -> (Thumb-L_FK-0:EUW Picker Button Object Reference)
    ) -> (Array)) -> (Output_Get:Array of EUW Picker Button Object References)
   > n19 = Set All Fingers Left(All Fingers Left:Array of EUW Picker Button Object References=Get 4 Fingers Left() -> (4 Fingers Left:Array of EUW Picker Button Object References)) -> (Output_Get:Array of EUW Picker Button Object References)
   > Append Array(TargetArray=@n19, SourceArray=Get Thumb Left() -> (Thumb Left:Array of EUW Picker Button Object References))
   > Append Array(
    TargetArray=@n19
    SourceArray=Make Array!(
     "[0]"=Get Meta-L_Pinky-Metacarpal-L() -> (Meta-L_Pinky-Metacarpal-L:EUW Picker Button Object Reference)
     "[1]"=Get Meta-L_Ring-Metacarpal-L() -> (Meta-L_Ring-Metacarpal-L:EUW Picker Button Object Reference)
     "[2]"=Get Meta-L_Middle-Metacarpal-L() -> (Meta-L_Middle-Metacarpal-L:EUW Picker Button Object Reference)
     "[3]":EUW Picker Button Object Reference=Get Meta-L_Index-Metacarpal-L() -> (Meta-L_Index-Metacarpal-L:EUW Picker Button Object Reference)
    ) -> (Array)
   )
   > Select Multiple Buttons(Button=Get AllFingers_l(), Buttons to Select=@n19)
   > Select Multiple Buttons(Button=Get Index_R(), Buttons to Select=Make Array!("[0]"=Get Index-R_FK-0(), "[1]"=Get Index-R_FK-1(), "[2]"=Get Index-R_FK-2()) -> (Array))
   > Select Multiple Buttons(Button=Get Middle_R(), Buttons to Select=Make Array("[0]"=Get Middle-R_FK-0(), "[1]"=Get Middle-R_FK-1(), "[2]"=Get Middle-R_FK-2()))
   > Select Multiple Buttons(Button=Get Ring_R(), Buttons to Select=Make Array("[0]"=Get Ring-R_FK-0(), "[1]"=Get Ring-R_FK-1(), "[2]"=Get Ring-R_FK-2()))
   > Select Multiple Buttons(Button=Get Pinky_R(), Buttons to Select=Make Array("[0]"=Get Pinky-R_FK-0(), "[1]"=Get Pinky-R_FK-1(), "[2]"=Get Pinky-R_FK-2()))
   > Select Multiple Buttons(Button=Get Thumb_R(), Buttons to Select=Make Array("[0]"=Get Thumb-R_FK-0(), "[1]"=Get Thumb-R_FK-1(), "[2]"=Get Thumb-R_FK-2()))
   > Select Multiple Buttons(
    Button=Get Fingers_R_Start()
    Buttons to Select=Make Array!("[0]"=Get Index-R_FK-0(), "[1]"=Get Middle-R_FK-0(), "[2]"=Get Ring-R_FK-0(), "[3]":EUW Picker Button Object Reference=Get Pinky-R_FK-0()) -> (Array)
   )
   > Select Multiple Buttons(
    Button=Get Fingers_R_Mid()
    Buttons to Select=Make Array("[0]"=Get Index-R_FK-1(), "[1]"=Get Middle-R_FK-1(), "[2]"=Get Ring-R_FK-1(), "[3]"=Get Pinky-R_FK-1())
   )
   > Select Multiple Buttons(
    Button=Get Fingers_R_End()
    Buttons to Select=Make Array("[0]"=Get Index-R_FK-2(), "[1]"=Get Middle-R_FK-2(), "[2]"=Get Ring-R_FK-2(), "[3]"=Get Pinky-R_FK-2())
   )
   > Select Multiple Buttons(
    Button=Get Meta_R()
    Buttons to Select=Make Array(
     "[0]"=Get Meta-R_Pinky-Metacarpal-R()
     "[1]"=Get Meta-R_Ring-Metacarpal-R()
     "[2]"=Get Meta-R_Middle-Metacarpal-R()
     "[3]"=Get Meta-R_Index-Metacarpal-R()
    )
   )
   > Select Multiple Buttons(Button=Get Index_L(), Buttons to Select=Make Array!("[0]"=Get Index-L_FK-0(), "[1]"=Get Index-L_FK-1(), "[2]"=Get Index-L_FK-2()) -> (Array))
   > Select Multiple Buttons(Button=Get Middle_L(), Buttons to Select=Make Array("[0]"=Get Middle-L_FK-0(), "[1]"=Get Middle-L_FK-1(), "[2]"=Get Middle-L_FK-2()))
   > Select Multiple Buttons(Button=Get Ring_L(), Buttons to Select=Make Array("[0]"=Get Ring-L_FK-0(), "[1]"=Get Ring-L_FK-1(), "[2]"=Get Ring-L_FK-2()))
   > Select Multiple Buttons(Button=Get Pinky_L(), Buttons to Select=Make Array("[0]"=Get Pinky-L_FK-0(), "[1]"=Get Pinky-L_FK-1(), "[2]"=Get Pinky-L_FK-2()))
   > Select Multiple Buttons(Button=Get Thumb_L(), Buttons to Select=Make Array("[0]"=Get Thumb-L_FK-0(), "[1]"=Get Thumb-L_FK-1(), "[2]"=Get Thumb-L_FK-2()))
   > Select Multiple Buttons(
    Button=Get Fingers_L_Start()
    Buttons to Select=Make Array!("[0]"=Get Index-L_FK-0(), "[1]"=Get Middle-L_FK-0(), "[2]"=Get Ring-L_FK-0(), "[3]":EUW Picker Button Object Reference=Get Pinky-L_FK-0()) -> (Array)
   )
   > Select Multiple Buttons(
    Button=Get Fingers_L_Mid()
    Buttons to Select=Make Array("[0]"=Get Index-L_FK-1(), "[1]"=Get Middle-L_FK-1(), "[2]"=Get Ring-L_FK-1(), "[3]"=Get Pinky-L_FK-1())
   )
   > Select Multiple Buttons(
    Button=Get Fingers_L_End()
    Buttons to Select=Make Array("[0]"=Get Index-L_FK-2(), "[1]"=Get Middle-L_FK-2(), "[2]"=Get Ring-L_FK-2(), "[3]"=Get Pinky-L_FK-2())
   )
   > Select Multiple Buttons(
    Button=Get Meta_L()
    Buttons to Select=Make Array(
     "[0]"=Get Meta-L_Pinky-Metacarpal-L()
     "[1]"=Get Meta-L_Ring-Metacarpal-L()
     "[2]"=Get Meta-L_Middle-Metacarpal-L()
     "[3]"=Get Meta-L_Index-Metacarpal-L()
    )
   )
   n20 = Set All IK Leg Extras Right(
    execute=@n6.then_4
    All IK Leg Extras Right:Array of EUW Picker Button Object References=Make Array!(
     "[0]"=Get Foot-R_Foot-Rocker()
     "[1]"=Get Foot-R_Ball-IK()
     "[2]"=Get Foot-R_Heel()
     "[3]"=Get Foot-R_Ball()
     "[4]":EUW Picker Button Object Reference=Get Foot-R_Toe-Tip()
    ) -> (Array)
   ) -> (Output_Get:Array of EUW Picker Button Object References)
   > Select Multiple Buttons(Button=Get AllIKLegExtras_r(), Buttons to Select=@n20)
   > n21 = Set All IK Leg Right(All IK Leg Right:Array of EUW Picker Button Object References=Get Right Leg IK()) -> (Output_Get:Array of EUW Picker Button Object References)
   > Select Multiple Buttons(Button=Get AllIKLeg_r(), Buttons to Select=@n21)
   > n22 = Set All FK Leg Right(All FK Leg Right:Array of EUW Picker Button Object References=Make Array!("[0]"=Get Leg-R_FK-2(), "[1]"=Get Leg-R_FK-1(), "[2]"=Get Leg-R_FK-0(), "[3]"=Get Foot-R_Toes-FK()) -> (Array)) -> (Output_Get:Array of EUW Picker Button Object References)
   > Select Multiple Buttons(Button=Get AllFKLeg_r(), Buttons to Select=@n22)
   > n23 = Set All IK Leg Extras Left(All IK Leg Extras Left:Array of EUW Picker Button Object References=Make Array!(
     "[0]"=Get Foot-L_Foot-Rocker()
     "[1]"=Get Foot-L_Ball-IK()
     "[2]"=Get Foot-L_Heel()
     "[3]"=Get Foot-L_Ball()
     "[4]":EUW Picker Button Object Reference=Get Foot-L_Toe-Tip()
    ) -> (Array)) -> (Output_Get:Array of EUW Picker Button Object References)
   > Select Multiple Buttons(Button=Get AllIKLegExtra_l(), Buttons to Select=@n23)
   > n24 = Set All IK Leg Left(All IK Leg Left:Array of EUW Picker Button Object References=Get Left Leg IK()) -> (Output_Get:Array of EUW Picker Button Object References)
   > Select Multiple Buttons(Button=Get AllIKLeg_l(), Buttons to Select=@n24)
   > n25 = Set All FK Leg Left(All FK Leg Left:Array of EUW Picker Button Object References=Make Array!("[0]"=Get Leg-L_FK-0(), "[1]"=Get Leg-L_FK-1(), "[2]"=Get Leg-L_FK-2(), "[3]"=Get Foot-L_Toes-FK()) -> (Array)) -> (Output_Get:Array of EUW Picker Button Object References)
   > Select Multiple Buttons(Button=Get AllFKLeg_l(), Buttons to Select=@n25)
   Select Multiple Buttons(execute=@n6.then_5, Button=Get AllLegs(), Buttons to Select=Get All Legs())
   > n26 = Set All FK Legs(All FK Legs:Array of EUW Picker Button Object References=Get All FK Leg Left()) -> (Output_Get:Array of EUW Picker Button Object References)
   > Append Array(TargetArray=@n26, SourceArray=Get All FK Leg Right())
   n27 = Make Array!(
    "[0]"=Get FKToes()
    "[1]"=Get FKKnees()
    "[2]"=Get FKThighs()
    "[3]"=Get FKFeets()
    "[4]":EUW Picker Button Object Reference=Get AllFKLegs()
    "[5]":EUW Picker Button Object Reference=Get AllFKLeg_l()
    "[6]":EUW Picker Button Object Reference=Get AllFKLeg_r()
   ) -> (Array)
   > Append Array(TargetArray=Get All FK Legs(), SourceArray=@n27)
   > Select Multiple Buttons(Button=Get AllFKLegs(), Buttons to Select=@n27)
   > Select Multiple Buttons(Button=Get FKThighs(), Buttons to Select=Make Array!("[0]"=Get Leg-R_FK-0(), "[1]"=Get Leg-L_FK-0()) -> (Array))
   > Select Multiple Buttons(Button=Get FKKnees(), Buttons to Select=Make Array("[0]"=Get Leg-R_FK-1(), "[1]"=Get Leg-L_FK-1()))
   > Select Multiple Buttons(Button=Get FKFeets(), Buttons to Select=Make Array("[0]"=Get Leg-R_FK-2(), "[1]"=Get Leg-L_FK-2()))
   > Select Multiple Buttons(Button=Get FKToes(), Buttons to Select=Make Array("[0]"=Get Foot-R_Toes-FK(), "[1]"=Get Foot-L_Toes-FK()))
   > Select Multiple Buttons(Button=Get IKKnees(), Buttons to Select=Make Array("[0]"=Get Leg-R_PV(), "[1]"=Get Leg-L_PV()))
   > Select Multiple Buttons(Button=Get IKBases(), Buttons to Select=Make Array("[0]"=Get Leg-R_IK-Base(), "[1]"=Get Leg-L_IK-Base()))
   > Select Multiple Buttons(Button=Get IKFeet(), Buttons to Select=Make Array("[0]"=Get Leg-R_IK(), "[1]"=Get Leg-L_IK()))
   > Select Multiple Buttons(Button=Get IKToes(), Buttons to Select=Make Array("[0]"=Get Foot-R_Toes-IK(), "[1]"=Get Foot-L_Toes-IK()))
   > n28 = Set All IK Leg Extras(All IK Leg Extras:Array of EUW Picker Button Object References=Get All IK Leg Extras Left()) -> (Output_Get:Array of EUW Picker Button Object References)
   > Append Array(TargetArray=@n28, SourceArray=Get All IK Leg Extras Right())
   > Select Multiple Buttons(Button=Get IKFeetExtras(), Buttons to Select=@n28)
   > n29 = Set All IK Legs(All IK Legs:Array of EUW Picker Button Object References=Get All IK Leg Left()) -> (Output_Get:Array of EUW Picker Button Object References)
   > Append Array(TargetArray=@n29, SourceArray=Get All IK Leg Right())
   > Append Array(TargetArray=@n29, SourceArray=Get All IK Leg Extras Left())
   > Append Array(TargetArray=@n29, SourceArray=Get All IK Leg Extras Right())
   > Append Array(
    TargetArray=@n29
    SourceArray=Make Array!(
     "[0]"=Get IKToes()
     "[1]"=Get IKBases()
     "[2]":EUW Picker Button Object Reference=Get Leg-R_Mid()
     "[3]":EUW Picker Button Object Reference=Get Leg-L_Mid()
     "[4]":EUW Picker Button Object Reference=Get AllIKLegs()
     "[5]":EUW Picker Button Object Reference=Get IKFeetExtras()
     "[6]":EUW Picker Button Object Reference=Get IKFeet()
     "[7]":EUW Picker Button Object Reference=Get IKKnees()
    ) -> (Array)
   )
   > Select Multiple Buttons(Button=Get AllIKLegs(), Buttons to Select=Get All IK Legs())
   Make Array!("[0]"=Get Leg-L_IK(), "[1]"=Get Leg-L_PV(), "[2]"=Get Foot-L_Ball-IK(), "[3]"=Get Leg-L_IK-Base()) -> (Array)
   Make Array("[0]"=Get Leg-R_IK(), "[1]"=Get Leg-R_PV(), "[2]"=Get Foot-R_Ball-IK(), "[3]"=Get Leg-R_IK-Base())
  ]}
 Match Event {nodes [
   n1 = Match Event!() -> (then, Ik Fk Checkbox:Check Box Object Reference, Module Name:Name) {is=title K2Node_FunctionEntry}
   > n2 = Branch(Condition=IsChecked(self=@n1))
   n3 = Get Module Name() -> (Module Name:Name=None)
   n4 = Get Modular Rig() -> (Modular Rig:Modular Rig Object Reference)
   > Execute Event On Module By Name!(execute, self:Modular Rig Object Reference=@n4, InEvent:Name=To FK, InModuleName:Name=@n3) -> (then, ReturnValue:bool)
   > Execute!(execute, self:Rig VMHost Object Reference=Get Modular Rig(), InEventName:"Name (by ref)"=MakeLiteralName(Value=Forwards Solve)) -> (then, ReturnValue:bool)
   n5 = Get Control Rig()
   n6 = Convert to MR naming(In name=GetDisplayName(Object=@n1)) -> (Out name=None)
   > Set Animation Channel Bool(ControlRig=@n5, Channel Name=@n6, Value=false)
   n7 = "To String (Name)"(InName=Get Module Name() -> (Module Name=None))
   > n8 = Select Control!(
    execute
    self
    Control Rig:Control Rig Object Reference=Get Control Rig()
    Control Name:"Name (by ref)"=String To Name(InString=Append(A=@n7, B=/FK 2))
   ) -> (then)
   Execute Event On Module By Name(execute=[@n8.then, @n9.then], self=Get Modular Rig(), InEvent=Key Controls, InModuleName=Get Module Name() -> (Module Name=None))
   > Execute(self=Get Modular Rig(), InEventName=MakeLiteralName(Value=Forwards Solve))
   > Refresh All IK FK Widgets()
   > Return Node!(execute)
   Execute Event On Module By Name(execute=@n2.else, self=@n4, InEvent=To IK, InModuleName=@n3)
   > Execute(self=Get Modular Rig(), InEventName=MakeLiteralName(Value=Forwards Solve))
   > Set Animation Channel Bool(ControlRig=@n5, Channel Name=@n6, Value=true)
   > n9 = Select Control(Control Rig=Get Control Rig(), Control Name=String To Name(InString=Append(A=@n7, B=/IK)))
  ]}
 Open Face Picker {nodes [
   Open Face Picker!() -> (then, Error if not found:bool=~) {is=title K2Node_FunctionEntry}
   > Set Found Face Rig(Found Face Rig:bool=false) -> (Output_Get:bool)
   > n1 = GetFocusedLevelSequence()
   > n2 = GetControlRigs!(execute, LevelSequence:Level Sequence Object Reference=@n1) -> (then, ReturnValue:Array of Control Rig Sequencer Binding Proxy Structures)
   > n3 = For Each Loop with Break!(Exec:Exec, Array:Array of Control Rig Sequencer Binding Proxy Structures=@n2, Break:Exec=@n13.then) -> (LoopBody:Exec, Array Element:Control Rig Sequencer Binding Proxy Structure, Array Index:int, Completed:Exec)
   n4 = Break Control Rig Sequencer Binding Proxy(ControlRigSequencerBindingProxy=@n3)
   n5 = Get Class(Object=@n4.ControlRig) -> (ReturnValue:Control Rig Class Reference)
   > n6 = Branch(Condition="Equal (Class)"(A=@n5, B=/Game/Characters/Player/Common/Fortnite_Base_Head/Facials/ControlRigs/Fortnite_Legacy_Face_CtrlRig.Fortnite_Legacy_Face_CtrlRig_C))
   > n7 = Set PickerToOpen(PickerToOpen:User Widget Class Reference=/Game/Characters/Player/Common/SharedRigs/EUW_Legacy_Face_AnimPicker.EUW_Legacy_Face_AnimPicker_C) -> (Output_Get:User Widget Class Reference)
   Set BindingProxy(execute=[@n18.then, @n7.then, @n15.then, @n17.then], BindingProxy:Movie Scene Binding Proxy Structure=@n4) -> (Output_Get:Movie Scene Binding Proxy Structure)
   n8 = Get LevelSequencer() -> (LevelSequencer:Level Sequence Object Reference)
   > n9 = GetBindingID!(
    execute
    Sequence:Movie Scene Sequence Object Reference=@n8
    InBinding:"Movie Scene Binding Proxy Structure (by ref)"=Break Control Rig Sequencer Binding Proxy(ControlRigSequencerBindingProxy=Get Control Rig Proxy Binding())
   ) -> (then, ReturnValue:Movie Scene Object Binding ID Structure)
   n10 = GetBoundObjects!(ObjectBinding:Movie Scene Object Binding ID Structure=@n9) -> (ReturnValue:Array of Object References)
   > Branch(Condition="integer > integer"(A=Length(TargetArray:Array of Object References=@n10)))
   > n11 = GetBindingID(Sequence=@n8, InBinding=Get BindingProxy() -> (BindingProxy:Movie Scene Binding Proxy Structure))
   n12 = GetBoundObjects(ObjectBinding=@n11)
   > Branch(Condition="Equal (Object)"(A="Get (a copy)"(Array:Array of Object References=@n12) -> (Output:Object Reference), B="Get (a copy)"(Array=@n10)))
   > n13 = Set Found Face Rig(Found Face Rig=true)
   n14 = Branch(
    execute=@n6.else
    Condition="Equal (Class)"(A=@n5, B=/Game/Characters/Player/Common/Fortnite_Base_Head/Facials/ControlRigs/Fortnite_3L_Face_CtrlRig.Fortnite_3L_Face_CtrlRig_C)
   )
   > n15 = Set PickerToOpen(PickerToOpen=/Game/Characters/Player/Common/SharedRigs/EUW_MainCharacte_Face_AnimPicker.EUW_MainCharacte_Face_AnimPicker_C)
   n16 = Branch(
    execute=@n14.else
    Condition="Equal (Class)"(A=/Game/Characters/Player/Common/Fortnite_Base_Head/Facials/ControlRigs/Fortnite_3L_OnFace_CtrlRig.Fortnite_3L_OnFace_CtrlRig_C, B=@n5)
   )
   > n17 = Set PickerToOpen(PickerToOpen=/Game/Characters/Player/Common/SharedRigs/EUW_MainCharacte_OnFace_AnimPicker.EUW_MainCharacte_OnFace_AnimPicker_C)
   Branch(execute=@n16.else, Condition="Equal (Class)"(A=/EpicControlRig/Templates/MR_FN_Face_V01.MR_FN_Face_V01_C, B=@n5))
   > n18 = Set PickerToOpen(PickerToOpen=/AnimInEdTools/EpicControlRig/Face/EUW_FN_Face_Base_AnimPicker.EUW_FN_Face_Base_AnimPicker_C)
   n19 = Branch(execute=@n3.Completed, Condition=Get Found Face Rig() -> (Found Face Rig:bool))
   n20 = Get FaceSlot()
   > n21 = Branch(Condition="integer > integer"(A=GetChildrenCount!(self:Panel Widget Object Reference=@n20) -> (ReturnValue:int=0)))
   n22 = Get FaceSlot()
   > ClearChildren(self=@n22)
   > Return Node!(execute, Success:bool=false, Out Face Picker:EUW Asset Picker Base Object Reference)
   ClearChildren(execute=@n21.else, self=@n22)
   > n23 = Create Widget!(execute, WorldContextObject, Class=Get PickerToOpen() -> (PickerToOpen:User Widget Class Reference), OwningPlayer) -> (then, ReturnValue:User Widget Object Reference)
   > n24 = Cast To EUW_Asset_Picker_Base!(execute, Object=@n23) -> (then, CastFailed, AsEUW Asset Picker Base:EUW Asset Picker Base Object Reference) {is=title K2Node_DynamicCast}
   > Set Face Picker(Face Picker:EUW Asset Picker Base Object Reference=@n24) -> (Output_Get:EUW Asset Picker Base Object Reference)
   n25 = Get Face Picker() -> (Face Picker:EUW Asset Picker Base Object Reference)
   > AddChild(self=Get FaceSlot(), Content=@n25)
   > SetSize!(
    execute
    self:Vertical Box Slot Object Reference=SlotAsVerticalBoxSlot(Widget=@n25)
    InSize:Slate Child Size Structure=Make Slate Child Size!(Value:float=1, SizeRule:ESlateSizeRule Enum=Fill) -> (SlateChildSize:Slate Child Size Structure) {is=title K2Node_MakeStruct}
   ) -> (then)
   > Widget has loaded!(
    execute:Exec
    self:Object Reference=Get Face Picker()
    Control Rig Proxy Binding:Control Rig Sequencer Binding Proxy Structure=Get Control Rig Proxy Binding()
    Animatable Creature:Animatable Creature Object Reference=Get Animatable Creature()
    Anim Picker EUW:Animation Picker Object Reference=Get Anim Picker EUW() -> (Anim Picker EUW:Animation Picker Object Reference)
   ) -> (then:Exec) {is=title K2Node_Message}
   n26 = Get Anim Picker EUW()
   > Update All Picker Elements Array(self:Animation Picker Object Reference=@n26)
   > SetCheckedState!(
    execute
    self:Check Box Object Reference=Get Auto_CheckBox!(self:Animation Picker Object Reference=@n26) -> (Auto_CheckBox:Editor Utility Check Box Object Reference)
    InCheckedState:ECheckBoxState Enum=Unchecked
   ) -> (then)
   > Return Node(Success=true, Out Face Picker=Get Face Picker())
   LogString(execute=@n24.CastFailed, InString=Face Rig is not a child of Asset Picker Base)
   > Return Node(Success=false)
   Branch(execute=@n19.else, Condition=Get Error if not found() -> (Error if not found:bool))
   > Show Message Dialog(
    Title=MakeLiteralText(Value="LOCTEXT(\"Cannot find Facial Rig\")")
    Message=MakeLiteralText(Value="LOCTEXT(\"Please add a Legacy or 3L face ControlRig\")")
    MessageType=Ok
    DefaultValue=No
    MessageCategory=Warning
   )
   > Return Node(Success=false)
  ]}
 Refresh All IK FK Widgets {nodes [
   Set Buttons Visibility From AnimationChannel!(
    execute
    self:EUW Asset Picker Base Object Reference
    ControlRig:Control Rig Object Reference=Get Control Rig()
    AnimationChannel:String=Arm-L_Sec-Controls-Vis
    Buttons:Array of EUW Picker Button Object References=Make Array!(
     "[0]"=Get Arm-Upper-Twist-L_Twist-2()
     "[1]"=Get Arm-Upper-Twist-L_Twist-1()
     "[2]"=Get Arm-Lower-Twist-L_Twist-2()
     "[3]"=Get Arm-Lower-Twist-L_Twist-1()
     "[4]":EUW Picker Button Object Reference=Get Arm-L_Mid()
    ) -> (Array)
   ) -> (then) {is=title K2Node_CallFunction}
   > Set Buttons Visibility From AnimationChannel(
    ControlRig=Get Control Rig()
    AnimationChannel=Arm-R_Sec-Controls-Vis
    Buttons=Make Array(
     "[0]"=Get Arm-Upper-Twist-R_Twist-2()
     "[1]"=Get Arm-Upper-Twist-R_Twist-1()
     "[2]"=Get Arm-Lower-Twist-R_Twist-2()
     "[3]"=Get Arm-Lower-Twist-R_Twist-1()
     "[4]"=Get Arm-R_Mid()
    )
   )
   > Set Buttons Visibility From AnimationChannel(
    ControlRig=Get Control Rig()
    AnimationChannel=Leg-R_Sec-Controls-Vis
    Buttons=Make Array(
     "[0]"=Get Leg-Upper-Twist-R_Twist-2()
     "[1]"=Get Leg-Upper-Twist-R_Twist-1()
     "[2]"=Get Leg-Lower-Twist-R_Twist-2()
     "[3]"=Get Leg-Lower-Twist-R_Twist-1()
     "[4]"=Get Leg-R_Mid()
    )
   )
   > Set Buttons Visibility From AnimationChannel(
    ControlRig=Get Control Rig()
    AnimationChannel=Leg-L_Sec-Controls-Vis
    Buttons=Make Array(
     "[0]"=Get Leg-Lower-Twist-L_Twist-2()
     "[1]"=Get Leg-Lower-Twist-L_Twist-1()
     "[2]"=Get Leg-Upper-Twist-L_Twist-2()
     "[3]"=Get Leg-Upper-Twist-L_Twist-1()
     "[4]"=Get Leg-L_Mid()
    )
   )
   > n1 = GetFocusedLevelSequence()
   > n2 = Sequence!(execute) -> (then_0, then_1)
   n3 = Get Leg-L_Ik-Fk-Switch()
   n4 = Get Current Local Time!(execute=@n8.then_1, TimeUnit:EMovieSceneTimeUnit Enum=DisplayRate) -> (
    then
    ReturnValue_Frame_FrameNumber:Frame Number Structure
    ReturnValue_Frame_SubFrame:float=~
    ReturnValue_Time:float=~
    ReturnValue_MarkedFrame:String
    ReturnValue_Timecode:Timecode Structure
    ReturnValue_PositionType:EMovieScenePositionType Enum
    ReturnValue_UpdateMethod:EUpdatePositionMethod Enum
    ReturnValue_bHasJumped:bool=~
   ) {is=title K2Node_CallFunction}
   n5 = GetLocalControlRigBool!(
    execute=@n2.then_0
    LevelSequence:Level Sequence Object Reference=@n1
    ControlRig:Control Rig Object Reference=Get Control Rig()
    ControlName:Name=Convert to MR naming(In name=GetObjectName(Object=@n3)) -> (Out name=None)
    Frame:Frame Number Structure=@n4
    TimeUnit:EMovieSceneTimeUnit Enum=DisplayRate
   ) -> (then, ReturnValue:bool)
   > Branch(Condition=@n5)
   > Set Buttons Visibility From AnimationChannel(
    ControlRig=Get Control Rig()
    AnimationChannel=Foot-L_Foot-Pivot-Control-Vis
    Buttons=Make Array!("[0]"=Get Foot-L_Toes-IK(), "[1]"=Get Foot-L_Heel(), "[2]"=Get Foot-L_Ball(), "[3]"=Get Foot-L_Toe-Tip()) -> (Array)
   )
   n6 = Get Leg-R_Ik-Fk-Switch()
   n7 = GetLocalControlRigBool(
    execute=@n2.then_1
    LevelSequence=@n1
    ControlRig=Get Control Rig()
    ControlName=Convert to MR naming(In name=GetObjectName(Object=@n6)) -> (Out name=None)
    Frame=@n4
    TimeUnit=DisplayRate
   )
   > Branch(Condition=@n7)
   > Set Buttons Visibility From AnimationChannel(
    ControlRig=Get Control Rig()
    AnimationChannel=Foot-R_Foot-Pivot-Control-Vis
    Buttons=Make Array("[0]"=Get Foot-R_Toes-IK(), "[1]"=Get Foot-R_Heel(), "[2]"=Get Foot-R_Ball(), "[3]"=Get Foot-R_Toe-Tip())
   )
   n8 = Sequence!(execute) -> (then_0, then_1, then_2:Exec, then_3:Exec)
   > n9 = Get Current Local Time(TimeUnit=DisplayRate) -> (ReturnValue_Frame_SubFrame=~, ReturnValue_Time=~, ReturnValue_bHasJumped=~)
   > n10 = GetFocusedLevelSequence()
   > n11 = GetLocalControlRigBool(
    LevelSequence=@n10
    ControlRig=Get Control Rig()
    ControlName=Convert to MR naming(In name=Arm-L_Hand-Align) -> (Out name=None)
    Frame=@n9
    TimeUnit=DisplayRate
   )
   n12 = Get Arm-L_Ik-Fk-Switch()
   > n13 = GetLocalControlRigBool(
    LevelSequence=@n10
    ControlRig=Get Control Rig()
    ControlName=Convert to MR naming(In name=GetObjectName(Object=@n12)) -> (Out name=None)
    Frame=@n9
    TimeUnit=DisplayRate
   )
   > Set Visibility for Array(" Array"=Make Array!("[0]"=Get Arm-L_IK-Rotation()) -> (Array), bVisibility=AND Boolean(A=@n11, B=@n13))
   Refresh All IK FK Widgets() {is=title K2Node_FunctionEntry}
   > Set Checkboxes State From Scene!(
    execute
    self:EUW Asset Picker Base Object Reference
    ControlRig:Control Rig Object Reference=Get Control Rig()
    Checkboxes:Array of Check Box Object References=Make Array!(
     "[0]":Check Box Object Reference=Get Leg-L_Ik-Fk-Switch()
     "[1]":Check Box Object Reference=Get Leg-R_Ik-Fk-Switch()
     "[2]":Check Box Object Reference=Get Arm-L_Ik-Fk-Switch()
     "[3]":Check Box Object Reference=Get Arm-R_Ik-Fk-Switch()
    ) -> (Array:Array of Check Box Object References)
   ) -> (then) {is=title K2Node_CallFunction}
   > Set Buttons Color From Checkbox State(CheckBox=Get Leg-L_Ik-Fk-Switch(), IK Buttons=Get Left Leg IK(), FK Buttons=Get Left Leg FK(), Color Mult=1, Opacity=0.5)
   > Set Buttons Color From Checkbox State(CheckBox=Get Leg-R_Ik-Fk-Switch(), IK Buttons=Get Right Leg IK(), FK Buttons=Get Right Leg FK(), Color Mult=1, Opacity=0.5)
   > Set Buttons Color From Checkbox State(CheckBox=Get Arm-L_Ik-Fk-Switch(), IK Buttons=Get Left Arm IK(), FK Buttons=Get Left Arm FK(), Color Mult=1, Opacity=0.5)
   > Set Buttons Color From Checkbox State(CheckBox=Get Arm-R_Ik-Fk-Switch(), IK Buttons=Get Right Arm IK(), FK Buttons=Get Right Arm FK(), Color Mult=1, Opacity=0.5)
   > IK FK Switch Text Update(Switch CheckBox=Get Leg-L_Ik-Fk-Switch(), Prefix=Left Leg) {is=title K2Node_CallFunction}
   > IK FK Switch Text Update(Switch CheckBox=Get Leg-R_Ik-Fk-Switch(), Prefix=Right Leg)
   > IK FK Switch Text Update(Switch CheckBox=Get Arm-L_Ik-Fk-Switch(), Prefix=Left Arm)
   > IK FK Switch Text Update(Switch CheckBox=Get Arm-R_Ik-Fk-Switch(), Prefix=Right Arm)
   n14 = Get Arm-R_Ik-Fk-Switch()
   n15 = GetFocusedLevelSequence(execute=@n4.then)
   n16 = GetLocalControlRigBool(
    execute=@n17.then
    LevelSequence=@n15
    ControlRig=Get Control Rig()
    ControlName=Convert to MR naming(In name=GetObjectName(Object=@n14)) -> (Out name=None)
    Frame=@n4
    TimeUnit=DisplayRate
   )
   n17 = GetLocalControlRigBool(
    execute=@n15.then
    LevelSequence=@n15
    ControlRig=Get Control Rig()
    ControlName=Convert to MR naming(In name=Arm-R_Hand-Align) -> (Out name=None)
    Frame=@n4
    TimeUnit=DisplayRate
   )
   Set Visibility for Array(
    execute=@n16.then
    " Array"=Make Array!("[0]":EUW Picker Button Object Reference=Get Arm-R_IK-Rotation()) -> (Array:Array of EUW Picker Button Object References)
    bVisibility=AND Boolean(A=@n17, B=@n16)
   )
  ]}
 Select Control {nodes [
   n1 = Select Control!() -> (then, Control Rig:Control Rig Object Reference, Control Name:"Name (by ref)") {is=title K2Node_FunctionEntry}
   > SelectNothing(self:Editor Actor Subsystem Object Reference=Get EditorActorSubsystem())
   > n2 = GetFocusedLevelSequence()
   > n3 = GetControlRigs(LevelSequence=@n2)
   > n4 = For Each Loop!(Exec, Array:Array of Control Rig Sequencer Binding Proxy Structures=@n3) -> (
    LoopBody
    Array Element_Proxy:Movie Scene Binding Proxy Structure
    Array Element_ControlRig:Control Rig Object Reference
    Array Element_Track:Movie Scene Control Rig Parameter Track Object Reference
    Array Index
    Completed
   )
   SelectControl!(
    execute=@n4.Completed
    self:Control Rig Object Reference=@n1
    InControlName:"Name (by ref)"=@n1.Control Name
    bSelect:bool=true
    bSetupUndo:bool=false
   ) -> (then)
   ClearControlSelection!(execute=@n4.LoopBody, self:Control Rig Object Reference=@n4.Array Element_ControlRig, bSetupUndo:bool=false) -> (then, ReturnValue:bool)
  ]}
 Set Toggle Capsule Button Vis {nodes [
   Set Toggle Capsule Button Vis() {is=title K2Node_FunctionEntry}
   > n1 = GetBindingID(
    Sequence=Get LevelSequencer()
    InBinding=Break Control Rig Sequencer Binding Proxy(ControlRigSequencerBindingProxy=Get Control Rig Proxy Binding())
   )
   n2 = GetBoundObjects(ObjectBinding=@n1)
   > Branch(Condition="integer > integer"(A=Length(TargetArray=@n2)))
   > n3 = Cast To Actor!(execute, Object="Get (a copy)"(Array=@n2)) -> (then, CastFailed, AsActor:Actor Object Reference) {is=title K2Node_DynamicCast}
   > n4 = Set As Actor(As Actor:Actor Object Reference=@n3) -> (Output_Get:Actor Object Reference)
   > n5 = GetAllActorsWithInterface!(
    execute
    WorldContextObject:Object Reference=@n4
    Interface:Interface Interface=/AnimInEdTools/EpicControlRig/Interfaces/BPI_ToggleCapsule.BPI_ToggleCapsule_C
   ) -> (then, OutActors:Array of Actor Object References)
   > n6 = Branch(Condition="integer >= integer"(A=Find Item!(TargetArray:Array of Actor Object References=@n5, ItemToFind:"Actor Object Reference (by ref)"=@n4) -> (ReturnValue)))
   n7 = Get ToggleCapsule_Button()
   > Set Visibility(Visibility=Visible, self=@n7)
   Set Visibility(execute=@n6.else, Visibility=Collapsed, self=@n7)
  ]}
 Setup Button Arrays {nodes [Setup Button Arrays() {is=title K2Node_FunctionEntry}]}
 Snap Event {nodes [
   n1 = Snap Event!() -> (then, Ik Fk Checkbox:Check Box Object Reference, Module Name:Name) {is=title K2Node_FunctionEntry}
   > n2 = Branch(Condition=IsChecked(self=@n1))
   n3 = Convert to MR naming(In name=GetDisplayName(Object=@n1)) -> (Out name=None)
   > Set Animation Channel Bool(ControlRig=Get Control Rig(), Channel Name=@n3, Value=false)
   > Execute Event On Module By Name(self=Get Modular Rig(), InEvent=To IK, InModuleName=Get Module Name() -> (Module Name=None))
   > Execute(self=Get Modular Rig(), InEventName=MakeLiteralName(Value=Forwards Solve))
   n4 = Get Control Rig()
   > n5 = Set Animation Channel Bool(ControlRig=@n4, Channel Name=@n3, Value=true)
   Execute Event On Module By Name(execute=[@n6.then, @n5.then], self=Get Modular Rig(), InEvent=Key Controls, InModuleName=Get Module Name() -> (Module Name=None))
   > Execute(self=Get Modular Rig(), InEventName=MakeLiteralName(Value=Forwards Solve))
   > Return Node!(execute)
   Set Animation Channel Bool(execute=@n2.else, ControlRig=Get Control Rig(), Channel Name=@n3, Value=true)
   > Execute Event On Module By Name(self=Get Modular Rig(), InEvent=To FK, InModuleName=Get Module Name() -> (Module Name=None))
   > Execute(self=Get Modular Rig(), InEventName=MakeLiteralName(Value=Forwards Solve))
   > n6 = Set Animation Channel Bool(ControlRig=@n4, Channel Name=@n3, Value=false)
  ]}
 Spawn Faceboard camera {nodes [
   n1 = GetFocusedLevelSequence()
   > Set Current LS(Current LS:Level Sequence Object Reference=@n1) -> (Output_Get:Level Sequence Object Reference)
   > n2 = GetSpawnables!(execute, Sequence:Movie Scene Sequence Object Reference=Get Current LS() -> (Current LS:Level Sequence Object Reference)) -> (then, ReturnValue:Array of Movie Scene Binding Proxy Structures)
   > n3 = For Each Loop!(Exec, Array:Array of Movie Scene Binding Proxy Structures=@n2) -> (LoopBody, Array Element:Movie Scene Binding Proxy Structure, Array Index, Completed)
   n4 = AddSpawnableFromClass!(
    execute=@n3.Completed
    self:Level Sequence Editor Subsystem Object Reference=Get LevelSequenceEditorSubsystem()
    Sequence:Movie Scene Sequence Object Reference=Get Current LS()
    ClassToSpawn:Object Class Reference=/Game/Characters/Player/Common/SharedRigs/PostProcess/BP_FaceboardCamera.BP_FaceboardCamera_C
   ) -> (then, ReturnValue:Movie Scene Binding Proxy Structure)
   > n5 = Set New Camera Binding(New Camera Binding:Movie Scene Binding Proxy Structure=@n4) -> (Output_Get:Movie Scene Binding Proxy Structure)
   > n6 = GetBindingID(Sequence=Get Current LS(), InBinding=@n5)
   > n7 = Cast To BP_FaceboardCamera!(execute, Object="Get (a copy)"(Array=GetBoundObjects(ObjectBinding=@n6))) -> (then, CastFailed, AsBP Faceboard Camera:BP Faceboard Camera Object Reference) {is=title K2Node_DynamicCast}
   > Set Faceboard Camera(Faceboard Camera:BP Faceboard Camera Object Reference=@n7) -> (Output_Get:BP Faceboard Camera Object Reference)
   > n8 = Sequence!(execute) -> (then_0, then_1, then_2, then_3, then_4:Exec)
   > n9 = GetControlRigs(LevelSequence=Get Current LS())
   > n10 = For Each Loop(Array:Array of Control Rig Sequencer Binding Proxy Structures=@n9) -> (Array Element:Control Rig Sequencer Binding Proxy Structure)
   n11 = Break Control Rig Sequencer Binding Proxy(ControlRigSequencerBindingProxy=@n10)
   n12 = Get Class(Object=@n11.ControlRig)
   > Branch(Condition=OR Boolean!(
     A="Equal (Class)"(A=@n12, B=&Fortnite_3L_Face_CtrlRig_C)
     B="Equal (Class)"(A=@n12, B=&Fortnite_Legacy_Face_CtrlRig_C)
     C:bool="Equal (Class)"(A=&Fortnite_3L_OnFace_CtrlRig_C, B=@n12)
    ) -> (ReturnValue))
   > n13 = Set Found Face Board Class(Found Face Board Class:String=Get Class Display Name(Class=@n12)) -> (Output_Get:String)
   > PrintString(InString=@n13)
   n14 = Get Current LS()
   > n15 = GetBindingID(Sequence=@n14, InBinding=Break Control Rig Sequencer Binding Proxy(ControlRigSequencerBindingProxy=Get Control Rig Proxy Binding()))
   > n16 = GetBindingID(Sequence=@n14, InBinding=@n11)
   n17 = GetBoundObjects(ObjectBinding=@n16)
   n18 = GetBoundObjects(ObjectBinding=@n15)
   > Branch(Condition="Equal (Object)"(A="Get (a copy)"(Array=@n17), B="Get (a copy)"(Array=@n18)))
   > Set Face CR(Face CR:Control Rig Object Reference=@n11.ControlRig) -> (Output_Get:Control Rig Object Reference)
   n19 = GetPossessables!(execute=@n8.then_1, Sequence:Movie Scene Sequence Object Reference=Get Current LS()) -> (then, ReturnValue:Array of Movie Scene Binding Proxy Structures)
   > n20 = For Each Loop(Array:Array of Movie Scene Binding Proxy Structures=@n19) -> (Array Element:Movie Scene Binding Proxy Structure)
   > n21 = GetBindingID(Sequence=Get Current LS(), InBinding=@n20)
   > Cast To CameraComponent!(execute, Object="Get (a copy)"(Array=GetBoundObjects(ObjectBinding=@n21))) -> (then, CastFailed, AsCamera Component:Camera Component Object Reference) {is=title K2Node_DynamicCast}
   > Remove!(execute, InBinding:"Movie Scene Binding Proxy Structure (by ref)"=@n20) -> (then)
   n22 = GetEditorWorld(execute=@n8.then_2, self=Get UnrealEditorSubsystem())
   > n23 = CreateTransformableHandle!(
    execute
    InWorld:World Object Reference=@n22
    InObject:Object Reference=Get Faceboard Camera() -> (Faceboard Camera:BP Faceboard Camera Object Reference)
    InAttachmentName:"Name (by ref)"=MakeLiteralName()
   ) -> (then, ReturnValue:Transformable Handle Object Reference)
   > n24 = CreateFromType!(execute, InWorld:World Object Reference=@n22, InType:ETransformConstraintType Enum=Parent) -> (then, ReturnValue:Tickable Transform Constraint Object Reference)
   > n25 = Switch on String!(execute:Exec, Selection:String=Get Found Face Board Class() -> (Found Face Board Class:String)) -> (Default:Exec, Fortnite_3L_Face_CtrlRig_C:Exec, Fortnite_3L_OnFace_CtrlRig_C:Exec, Fortnite_Legacy_Face_CtrlRig_C:Exec) {pinNames [Fortnite_3L_Face_CtrlRig_C, Fortnite_3L_OnFace_CtrlRig_C, Fortnite_Legacy_Face_CtrlRig_C]}
   AddConstraint!(
    execute=@n25.Fortnite_3L_Face_CtrlRig_C
    InWorld:World Object Reference=@n22
    InParentHandle:Transformable Handle Object Reference=CreateTransformableControlHandle!(
     self:Control Rig Object Reference=Get Face CR() -> (Face CR:Control Rig Object Reference)
     ControlName:"Name (by ref)"=MakeLiteralName(Value=CTRL_Master)
    ) -> (ReturnValue:Transformable Control Handle Object Reference)
    InChildHandle:Transformable Handle Object Reference=@n23
    InConstraint:Tickable Transform Constraint Object Reference=@n24
    bMaintainOffset:bool=false
   ) -> (then, ReturnValue:bool)
   > Set Camera Offset!(execute, Camera Offset:Vector="7,130,15", self:BP Faceboard Camera Object Reference=Get Faceboard Camera()) -> (then, Output_Get:Vector)
   AddConstraint(
    execute=@n25.Fortnite_3L_OnFace_CtrlRig_C
    InWorld=@n22
    InParentHandle=CreateTransformableControlHandle(self=Get Face CR(), ControlName=MakeLiteralName(Value=Face_Ctrl_Master))
    InChildHandle=@n23
    InConstraint=@n24
    bMaintainOffset=false
   )
   > Set Camera Offset(Camera Offset="0,130,15", self=Get Faceboard Camera())
   AddConstraint(
    execute=@n25.Fortnite_Legacy_Face_CtrlRig_C
    InWorld=@n22
    InParentHandle=CreateTransformableControlHandle(self=Get Face CR(), ControlName=MakeLiteralName(Value=Face_Ctrl_Master))
    InChildHandle=@n23
    InConstraint=@n24
    bMaintainOffset=false
   )
   > Set Camera Offset(Camera Offset="0,130,10", self=Get Faceboard Camera())
   n26 = GetBindingID(execute=@n3.LoopBody, Sequence=Get Current LS(), InBinding=@n3)
   n27 = "Get (a copy)"(Array=GetBoundObjects(ObjectBinding=@n26))
   > n28 = Cast To SkeletalMeshActor!(execute, Object=@n27) -> (then, CastFailed, AsSkeletal Mesh Actor:Skeletal Mesh Actor Object Reference) {is=title K2Node_DynamicCast}
   > Branch(Condition=Contains(SearchIn=GetDisplayName(Object=@n28), Substring=Anim))
   > n29 = GetBindingID(Sequence=Get Current LS(), InBinding=@n3)
   > Set BP Anim Binding ID(BP Anim Binding ID:Movie Scene Object Binding ID Structure=@n29) -> (Output_Get:Movie Scene Object Binding ID Structure)
   n30 = Cast To CameraActor!(execute=@n28.CastFailed, Object=@n27) -> (then, CastFailed, AsCamera Actor:Camera Actor Object Reference) {is=title K2Node_DynamicCast}
   > Branch(Condition=ActorHasTag!(self:Actor Object Reference=@n30, Tag:Name=Get Camera Tag() -> (Camera Tag:Name=None)) -> (ReturnValue:bool))
   > Remove(InBinding=@n3)
   Spawn Faceboard camera() {is=title K2Node_FunctionEntry}
  ]}
 Visibility Buttons {nodes [
   n1 = Visibility Buttons!() -> (then, Visibility Button:EUW Picker Button Object Reference, Buttons to Hide:Array of EUW Picker Button Object References, Bool:bool=~) {is=title K2Node_FunctionEntry}
   > n2 = Branch(Condition=@n1.Bool)
   > n3 = Set(Normal Tint_SpecifiedColor=(R=0.0774,G=0.0313,B=0.0698,A=1), self=@n1)
   Set Visibility for Array(execute=[@n4.then, @n3.then], " Array"=@n1.Buttons to Hide, bVisibility=NOT Boolean(A=@n1.Bool))
   > Update Selection Display(self=@n1, Selected=false)
   n4 = Set(execute=@n2.else, Normal Tint_SpecifiedColor=(R=0.2423,G=0.0595,B=0.2122,A=1), self=@n1)
  ]}
]
variables [
 4 Fingers Left {type="array<EUW_Picker_Button_C>", default []}
 4 Fingers Right {type="array<EUW_Picker_Button_C>", default []}
 All Arm Fingers Left {type="array<EUW_Picker_Button_C>", default []}
 All Arm Fingers Right {type="array<EUW_Picker_Button_C>", default []}
 All Arm Left {type="array<EUW_Picker_Button_C>", default []}
 All Arm Right {type="array<EUW_Picker_Button_C>", default []}
 All Arms {type="array<EUW_Picker_Button_C>", default []}
 All extra quick select buttons {type="array<EUW_Picker_Button_C>", default []}
 All Fingers {type="array<EUW_Picker_Button_C>", default []}
 All Fingers Left {type="array<EUW_Picker_Button_C>", default []}
 All Fingers Right {type="array<EUW_Picker_Button_C>", default []}
 All FK Arms {type="array<EUW_Picker_Button_C>", default []}
 All FK Leg Left {type="array<EUW_Picker_Button_C>", default []}
 All FK Leg Right {type="array<EUW_Picker_Button_C>", default []}
 All FK Legs {type="array<EUW_Picker_Button_C>", default []}
 All IK Arms {type="array<EUW_Picker_Button_C>", default []}
 All IK Leg Extras {type="array<EUW_Picker_Button_C>", default []}
 All IK Leg Extras Left {type="array<EUW_Picker_Button_C>", default []}
 All IK Leg Extras Right {type="array<EUW_Picker_Button_C>", default []}
 All IK Leg Left {type="array<EUW_Picker_Button_C>", default []}
 All IK Leg Right {type="array<EUW_Picker_Button_C>", default []}
 All IK Legs {type="array<EUW_Picker_Button_C>", default []}
 All Legs {type="array<EUW_Picker_Button_C>", default []}
 Animatable Creature_0 {type=AnimatableCreature_C, default=null}
 As Actor {type=Actor, default=null}
 BlockingMode Buttons {type="array<EUW_Picker_Button_C>", default []}
 Checkboxes {type="array<EditorUtilityCheckBox>", default []}
 Control Rig {type=ControlRig, default=null}
 Control Rig Proxy Binding_0 {
  type=ControlRigSequencerBindingProxy
  default {controlRig=null, track=null, proxy {bindingId=#1, sequence=null}}
 }
 Faceboard Camera {type=BP_FaceboardCamera_C, default=null}
 FK Hands {type="array<EUW_Picker_Button_C>", default []}
 Frame Number {type=FrameNumber, default {value=0}}
 IK Move Buttons {type="array<EUW_Picker_Button_C>", default []}
 Left Arm FK {type="array<EUW_Picker_Button_C>", default []}
 Left Arm IK {type="array<EUW_Picker_Button_C>", default []}
 Left Leg FK {type="array<EUW_Picker_Button_C>", default []}
 Left Leg IK {type="array<EUW_Picker_Button_C>", default []}
 LevelSequencer {type=LevelSequence, default=null}
 Modular Rig {type=ModularRig, default=null}
 Neck FK {type="array<EUW_Picker_Button_C>", default []}
 Neck IK {type="array<EUW_Picker_Button_C>", default []}
 Out Face Picker {type=EUW_Asset_Picker_Base_C, default=null}
 Right Arm FK {type="array<EUW_Picker_Button_C>", default []}
 Right Arm IK {type="array<EUW_Picker_Button_C>", default []}
 Right Leg FK {type="array<EUW_Picker_Button_C>", default []}
 Right Leg IK {type="array<EUW_Picker_Button_C>", default []}
 Spine FK {type="array<EUW_Picker_Button_C>", default []}
 Spine IK {type="array<EUW_Picker_Button_C>", default []}
 Thumb Left {type="array<EUW_Picker_Button_C>", default []}
 Thumb Right {type="array<EUW_Picker_Button_C>", default []}
 Twists {type="array<EUW_Picker_Button_C>", default []}
]
widgets {
 info {
  parentClass=/AnimInEdTools/AnimationPicker/EUW_Asset_Picker_Base.EUW_Asset_Picker_Base_C
  rootWidgetClass=&CanvasPanel
  widgetCount=247
 }
 widgets [
  CanvasPanel_0 {widgetClassPath=&CanvasPanel, bIsVariable=true, widget=CanvasPanel_0}
  DialogSlot {
   widgetClassPath=&NamedSlot
   bIsVariable=true
   parent=CanvasPanel_0
   widget=DialogSlot
   slot {
    class=CanvasPanelSlot
    properties {layoutData {alignment=<0 0>, anchors {maximum=<0.5 0.5>, minimum=<0.5 0.5>}, offsets=<ltrb 555 15 470 295>}}
   }
  }
  SizeBox_67 {
   widgetClassPath=&SizeBox
   parent=CanvasPanel_0
   widget=SizeBox_67
   slot {
    class=CanvasPanelSlot
    properties {layoutData {alignment=<0 0>, anchors {maximum=<0.5 0.5>, minimum=<0.5 0.5>}, offsets=<ltrb 685 -500 1355 990>}}
   }
  }
  FaceSlot {widgetClassPath=&VerticalBox, bIsVariable=true, parent=SizeBox_67, widget=FaceSlot, slot {class=SizeBoxSlot}}
  IKswitches {
   widgetClassPath=&CanvasPanel
   parent=CanvasPanel_0
   widget=IKswitches
   slot {
    class=CanvasPanelSlot
    properties {layoutData {alignment=<0 0>, anchors {maximum=<0.5 0.5>, minimum=<0.5 0.5>}, offsets=<ltrb -452.9609 -356.5405 930.3303 675.1351>}}
   }
  }
  Leg-R_Ik-Fk-Switch {
   widgetClassPath=&EditorUtilityCheckBox
   bIsVariable=true
   parent=IKswitches
   widget=Leg-R_Ik-Fk-Switch
   properties {checkedState=Checked}
   slot {
    class=CanvasPanelSlot
    properties {bAutoSize=true, layoutData {alignment=<0 0>, anchors {maximum=<0.5 0.5>, minimum=<0.5 0.5>}, offsets=<ltrb -221.1652 215 100 30>}}
   }
  }
  TextBlock {
   widgetClassPath=&TextBlock
   parent=Leg-R_Ik-Fk-Switch
   widget=TextBlock
   properties {
    text=Right Leg FK
    font {
     bForceMonospaced=false
     bMaterialIsStencil=false
     fontMaterial=null
     fontObject=&Roboto
     letterSpacing=0
     monospacedWidth=1
     size=8.25
     skewAmount=0
     typefaceFontName=Bold
     outlineSettings {
      bApplyOutlineToDropShadows=false
      bMiteredCorners=false
      bSeparateFillAlpha=false
      outlineMaterial=null
      outlineSize=0
      outlineColor=<rgba 0 0 0 1>
     }
    }
   }
   slot {class=PanelSlot}
  }
  Leg-L_Ik-Fk-Switch {
   widgetClassPath=&EditorUtilityCheckBox
   bIsVariable=true
   parent=IKswitches
   widget=Leg-L_Ik-Fk-Switch
   properties {checkedState=Checked}
   slot {
    class=CanvasPanelSlot
    properties {bAutoSize=true, layoutData {alignment=<0 0>, anchors {maximum=<0.5 0.5>, minimum=<0.5 0.5>}, offsets=<ltrb 147 215 100 30>}}
   }
  }
  TextBlock_26 {
   widgetClassPath=&TextBlock
   parent=Leg-L_Ik-Fk-Switch
   widget=TextBlock_26
   properties {
    text=Left Leg FK
    font {
     bForceMonospaced=false
     bMaterialIsStencil=false
     fontMaterial=null
     fontObject=&Roboto
     letterSpacing=0
     monospacedWidth=1
     size=8.25
     skewAmount=0
     typefaceFontName=Bold
     outlineSettings {
      bApplyOutlineToDropShadows=false
      bMiteredCorners=false
      bSeparateFillAlpha=false
      outlineMaterial=null
      outlineSize=0
      outlineColor=<rgba 0 0 0 1>
     }
    }
   }
   slot {class=PanelSlot}
  }
  Arm-R_Ik-Fk-Switch {
   widgetClassPath=&EditorUtilityCheckBox
   bIsVariable=true
   parent=IKswitches
   widget=Arm-R_Ik-Fk-Switch
   slot {
    class=CanvasPanelSlot
    properties {bAutoSize=true, layoutData {alignment=<0.5 0.5>, anchors {maximum=<0.5 0.5>, minimum=<0.5 0.5>}, offsets=<ltrb -171.2155 -292.5676 100 30>}}
   }
  }
  TextBlock_2 {
   widgetClassPath=&TextBlock
   parent=Arm-R_Ik-Fk-Switch
   widget=TextBlock_2
   properties {
    text=Right Arm IK
    font {
     bForceMonospaced=false
     bMaterialIsStencil=false
     fontMaterial=null
     fontObject=&Roboto
     letterSpacing=0
     monospacedWidth=1
     size=8.25
     skewAmount=0
     typefaceFontName=Bold
     outlineSettings {
      bApplyOutlineToDropShadows=false
      bMiteredCorners=false
      bSeparateFillAlpha=false
      outlineMaterial=null
      outlineSize=0
      outlineColor=<rgba 0 0 0 1>
     }
    }
   }
   slot {class=PanelSlot}
  }
  Arm-L_Ik-Fk-Switch {
   widgetClassPath=&EditorUtilityCheckBox
   bIsVariable=true
   parent=IKswitches
   widget=Arm-L_Ik-Fk-Switch
   slot {
    class=CanvasPanelSlot
    properties {bAutoSize=true, layoutData {alignment=<0.5 0.5>, anchors {maximum=<0.5 0.5>, minimum=<0.5 0.5>}, offsets=<ltrb 145.4489 -292.5676 100 30>}}
   }
  }
  TextBlock_1 {
   widgetClassPath=&TextBlock
   parent=Arm-L_Ik-Fk-Switch
   widget=TextBlock_1
   properties {
    text=Left Arm IK
    font {
     bForceMonospaced=false
     bMaterialIsStencil=false
     fontMaterial=null
     fontObject=&Roboto
     letterSpacing=0
     monospacedWidth=1
     size=8.25
     skewAmount=0
     typefaceFontName=Bold
     outlineSettings {
      bApplyOutlineToDropShadows=false
      bMiteredCorners=false
      bSeparateFillAlpha=false
      outlineMaterial=null
      outlineSize=0
      outlineColor=<rgba 0 0 0 1>
     }
    }
   }
   slot {class=PanelSlot}
  }
  Arm-L_FK-0-Gimbal {
   widgetClassPath=/AnimInEdTools/AnimationPicker/EUW_Picker_Button
   bIsVariable=true
   parent=CanvasPanel_0
   widget=Arm-L_FK-0-Gimbal
   properties {
    draw Style=Square
    label Font Size=6
    label Text=GIM
    label Color {colorUseRule=UseColor_Specified, specifiedColor=<rgba 0 0 0 1>}
    normal Tint {colorUseRule=UseColor_Specified, specifiedColor=<rgba 0.0272 0.067 0.7826 1>}
   }
   slot {
    class=CanvasPanelSlot
    properties {layoutData {alignment=<0.5 0.5>, anchors {maximum=<0.5 0.5>, minimum=<0.5 0.5>}, offsets=<ltrb 95 -264 25 25>}}
   }
  }
  Arm-R_FK-0-Gimbal {
   widgetClassPath=&EUW_Picker_Button
   bIsVariable=true
   parent=CanvasPanel_0
   widget=Arm-R_FK-0-Gimbal
   properties {
    draw Style=Square
    label Font Size=6
    label Text=GIM
    label Color {colorUseRule=UseColor_Specified, specifiedColor=<rgba 0 0 0 1>}
    normal Tint {colorUseRule=UseColor_Specified, specifiedColor=<rgba 1 0.02 0.02 1>}
   }
   slot {
    class=CanvasPanelSlot
    properties {layoutData {alignment=<0.5 0.5>, anchors {maximum=<0.5 0.5>, minimum=<0.5 0.5>}, offsets=<ltrb -95 -264 25 25>}}
   }
  }
  Spine_Start-FK {
   widgetClassPath=&EUW_Picker_Button
   bIsVariable=true
   parent=CanvasPanel_0
   widget=Spine_Start-FK
   properties {draw Style=Square, label Font Size=5, normal Tint {colorUseRule=UseColor_Specified, specifiedColor=<rgba 0.9823 0.7758 0 1>}}
   slot {
    class=CanvasPanelSlot
    properties {layoutData {alignment=<0.5 0.5>, anchors {maximum=<0.5 0.5>, minimum=<0.5 0.5>}, offsets=<ltrb 0 -120 25 25>}}
   }
  }
  Spine_Mid-FK {
   widgetClassPath=&EUW_Picker_Button
   bIsVariable=true
   parent=CanvasPanel_0
   widget=Spine_Mid-FK
   properties {draw Style=Square, label Font Size=5, normal Tint {colorUseRule=UseColor_Specified, specifiedColor=<rgba 0.9823 0.7758 0 1>}}
   slot {
    class=CanvasPanelSlot
    properties {layoutData {alignment=<0.5 0.5>, anchors {maximum=<0.5 0.5>, minimum=<0.5 0.5>}, offsets=<ltrb 0 -160 25 25>}}
   }
  }
  Spine_End-FK {
   widgetClassPath=&EUW_Picker_Button
   bIsVariable=true
   parent=CanvasPanel_0
   widget=Spine_End-FK
   properties {
    draw Style=Square
    label Color {colorUseRule=UseColor_Specified, specifiedColor=<rgba 0 0 0 1>}
    normal Tint {colorUseRule=UseColor_Specified, specifiedColor=<rgba 0.9823 0.7758 0 1>}
   }
   slot {
    class=CanvasPanelSlot
    properties {layoutData {alignment=<0.5 0.5>, anchors {maximum=<0.5 0.5>, minimum=<0.5 0.5>}, offsets=<ltrb 0 -200 25 25>}}
   }
  }
  Spine_Pelvis-Sec-FK {
   widgetClassPath=&EUW_Picker_Button
   bIsVariable=true
   parent=CanvasPanel_0
   widget=Spine_Pelvis-Sec-FK
   properties {
    draw Style=Square
    label Color {colorUseRule=UseColor_Specified, specifiedColor=<rgba 0 0 0 1>}
    normal Tint {colorUseRule=UseColor_Specified, specifiedColor=<rgba 1 0.9911 0.0999 1>}
   }
   slot {
    class=CanvasPanelSlot
    properties {layoutData {alignment=<0.5 0.5>, anchors {maximum=<0.5 0.5>, minimum=<0.5 0.5>}, offsets=<ltrb 30.5391 -109.0405 15 15>}}
   }
  }
  Spine_Spine-01-Sec-FK {
   widgetClassPath=&EUW_Picker_Button
   bIsVariable=true
   parent=CanvasPanel_0
   widget=Spine_Spine-01-Sec-FK
   properties {
    draw Style=Square
    label Color {colorUseRule=UseColor_Specified, specifiedColor=<rgba 0 0 0 1>}
    normal Tint {colorUseRule=UseColor_Specified, specifiedColor=<rgba 1 0.9911 0.0999 1>}
   }
   slot {
    class=CanvasPanelSlot
    properties {layoutData {alignment=<0.5 0.5>, anchors {maximum=<0.5 0.5>, minimum=<0.5 0.5>}, offsets=<ltrb 30.5391 -129.0405 15 15>}}
   }
  }
  Spine_Spine-02-Sec-FK {
   widgetClassPath=&EUW_Picker_Button
   bIsVariable=true
   parent=CanvasPanel_0
   widget=Spine_Spine-02-Sec-FK
   properties {
    draw Style=Square
    label Color {colorUseRule=UseColor_Specified, specifiedColor=<rgba 0 0 0 1>}
    normal Tint {colorUseRule=UseColor_Specified, specifiedColor=<rgba 1 0.9911 0.0999 1>}
   }
   slot {
    class=CanvasPanelSlot
    properties {layoutData {alignment=<0.5 0.5>, anchors {maximum=<0.5 0.5>, minimum=<0.5 0.5>}, offsets=<ltrb 30.5391 -149.0405 15 15>}}
   }
  }
  Spine_Spine-03-Sec-FK {
   widgetClassPath=&EUW_Picker_Button
   bIsVariable=true
   parent=CanvasPanel_0
   widget=Spine_Spine-03-Sec-FK
   properties {
    draw Style=Square
    label Color {colorUseRule=UseColor_Specified, specifiedColor=<rgba 0 0 0 1>}
    normal Tint {colorUseRule=UseColor_Specified, specifiedColor=<rgba 1 0.9911 0.0999 1>}
   }
   slot {
    class=CanvasPanelSlot
    properties {layoutData {alignment=<0.5 0.5>, anchors {maximum=<0.5 0.5>, minimum=<0.5 0.5>}, offsets=<ltrb 30.5391 -169.0405 15 15>}}
   }
  }
  Spine_Spine-04-Sec-FK {
   widgetClassPath=&EUW_Picker_Button
   bIsVariable=true
   parent=CanvasPanel_0
   widget=Spine_Spine-04-Sec-FK
   properties {
    draw Style=Square
    label Color {colorUseRule=UseColor_Specified, specifiedColor=<rgba 0 0 0 1>}
    normal Tint {colorUseRule=UseColor_Specified, specifiedColor=<rgba 1 0.9911 0.0999 1>}
   }
   slot {
    class=CanvasPanelSlot
    properties {layoutData {alignment=<0.5 0.5>, anchors {maximum=<0.5 0.5>, minimum=<0.5 0.5>}, offsets=<ltrb 30.5391 -189.0405 15 15>}}
   }
  }
  Spine_Spine-05-Sec-FK {
   widgetClassPath=&EUW_Picker_Button
   bIsVariable=true
   parent=CanvasPanel_0
   widget=Spine_Spine-05-Sec-FK
   properties {
    draw Style=Square
    label Color {colorUseRule=UseColor_Specified, specifiedColor=<rgba 0 0 0 1>}
    normal Tint {colorUseRule=UseColor_Specified, specifiedColor=<rgba 0.6174 0.145 0.6174 1>}
   }
   slot {
    class=CanvasPanelSlot
    properties {layoutData {alignment=<0.5 0.5>, anchors {maximum=<0.5 0.5>, minimum=<0.5 0.5>}, offsets=<ltrb 30.5391 -209.0405 15 15>}}
   }
  }
  Neck_Start-FK {
   widgetClassPath=&EUW_Picker_Button
   bIsVariable=true
   parent=CanvasPanel_0
   widget=Neck_Start-FK
   properties {draw Style=Square, label Font Size=5, normal Tint {colorUseRule=UseColor_Specified, specifiedColor=<rgba 0.9823 0.7758 0 1>}}
   slot {
    class=CanvasPanelSlot
    properties {layoutData {alignment=<0.5 0.5>, anchors {maximum=<0.5 0.5>, minimum=<0.5 0.5>}, offsets=<ltrb 0 -255 20 20>}}
   }
  }
  Neck_Mid-FK {
   widgetClassPath=&EUW_Picker_Button
   bIsVariable=true
   parent=CanvasPanel_0
   widget=Neck_Mid-FK
   properties {draw Style=Square, label Font Size=5, normal Tint {colorUseRule=UseColor_Specified, specifiedColor=<rgba 0.9823 0.7758 0 1>}}
   slot {
    class=CanvasPanelSlot
    properties {layoutData {alignment=<0.5 0.5>, anchors {maximum=<0.5 0.5>, minimum=<0.5 0.5>}, offsets=<ltrb 0 -285 20 20>}}
   }
  }
  Neck_End-FK {
   widgetClassPath=&EUW_Picker_Button
   bIsVariable=true
   parent=CanvasPanel_0
   widget=Neck_End-FK
   properties {
    draw Style=Square
    label Font Size=12
    label Text=FK
    label Color {colorUseRule=UseColor_Specified, specifiedColor=<rgba 0 0 0 1>}
    normal Tint {colorUseRule=UseColor_Specified, specifiedColor=<rgba 0.9823 0.7758 0 1>}
   }
   slot {
    class=CanvasPanelSlot
    properties {layoutData {alignment=<0.5 0.5>, anchors {maximum=<0.5 0.5>, minimum=<0.5 0.5>}, offsets=<ltrb 0 -325 40 40>}}
   }
  }
  Neck_End-IK {
   widgetClassPath=&EUW_Picker_Button
   bIsVariable=true
   parent=CanvasPanel_0
   widget=Neck_End-IK
   properties {
    draw Style=Square
    label Font Size=12
    label Text=IK
    label Color {colorUseRule=UseColor_Specified, specifiedColor=<rgba 0 0 0 1>}
    normal Tint {colorUseRule=UseColor_Specified, specifiedColor=<rgba 1 0.9911 0.0999 1>}
   }
   slot {
    class=CanvasPanelSlot
    properties {layoutData {alignment=<0.5 0.5>, anchors {maximum=<0.5 0.5>, minimum=<0.5 0.5>}, offsets=<ltrb -40 -310 30 30>}}
   }
  }
  Neck_Mid-IK {
   widgetClassPath=&EUW_Picker_Button
   bIsVariable=true
   parent=CanvasPanel_0
   widget=Neck_Mid-IK
   properties {draw Style=Square, label Font Size=5, normal Tint {colorUseRule=UseColor_Specified, specifiedColor=<rgba 1 0.9911 0.0999 1>}}
   slot {
    class=CanvasPanelSlot
    properties {layoutData {alignment=<0.5 0.5>, anchors {maximum=<0.5 0.5>, minimum=<0.5 0.5>}, offsets=<ltrb -40 -282 15 15>}}
   }
  }
  Neck_Start-IK {
   widgetClassPath=&EUW_Picker_Button
   bIsVariable=true
   parent=CanvasPanel_0
   widget=Neck_Start-IK
   properties {draw Style=Square, label Font Size=5, normal Tint {colorUseRule=UseColor_Specified, specifiedColor=<rgba 1 0.9911 0.0999 1>}}
   slot {
    class=CanvasPanelSlot
    properties {layoutData {alignment=<0.5 0.5>, anchors {maximum=<0.5 0.5>, minimum=<0.5 0.5>}, offsets=<ltrb -40 -263 15 15>}}
   }
  }
  Spine_End-IK {
   widgetClassPath=&EUW_Picker_Button
   bIsVariable=true
   parent=CanvasPanel_0
   widget=Spine_End-IK
   properties {
    draw Style=Square
    label Text=IK
    label Color {colorUseRule=UseColor_Specified, specifiedColor=<rgba 0 0 0 1>}
    normal Tint {colorUseRule=UseColor_Specified, specifiedColor=<rgba 1 0.9911 0.0999 1>}
   }
   slot {
    class=CanvasPanelSlot
    properties {layoutData {alignment=<0.5 0.5>, anchors {maximum=<0.5 0.5>, minimum=<0.5 0.5>}, offsets=<ltrb -32.4609 -200.0405 25 25>}}
   }
  }
  Spine_Mid-IK {
   widgetClassPath=&EUW_Picker_Button
   bIsVariable=true
   parent=CanvasPanel_0
   widget=Spine_Mid-IK
   properties {
    draw Style=Square
    label Text=IK
    label Color {colorUseRule=UseColor_Specified, specifiedColor=<rgba 0 0 0 1>}
    normal Tint {colorUseRule=UseColor_Specified, specifiedColor=<rgba 1 0.9911 0.0999 1>}
   }
   slot {
    class=CanvasPanelSlot
    properties {layoutData {alignment=<0.5 0.5>, anchors {maximum=<0.5 0.5>, minimum=<0.5 0.5>}, offsets=<ltrb -32.4609 -160.0405 25 25>}}
   }
  }
  Spine_End-Movable-Pivot {
   widgetClassPath=&EUW_Picker_Button
   bIsVariable=true
   parent=CanvasPanel_0
   widget=Spine_End-Movable-Pivot
   properties {
    draw Style=Square
    label Font Size=5
    label Text=PIV
    normal Tint {colorUseRule=UseColor_Specified, specifiedColor=<rgba 0.6174 0.0418 0.3047 1>}
   }
   slot {
    class=CanvasPanelSlot
    properties {layoutData {alignment=<0.5 0.5>, anchors {maximum=<0.5 0.5>, minimum=<0.5 0.5>}, offsets=<ltrb -62.9609 -198.5405 20 20>}}
   }
  }
  Spine_Start-IK {
   widgetClassPath=&EUW_Picker_Button
   bIsVariable=true
   parent=CanvasPanel_0
   widget=Spine_Start-IK
   properties {
    draw Style=Square
    label Text=PELVIS
    label Color {colorUseRule=UseColor_Specified, specifiedColor=<rgba 0 0 0 1>}
    normal Tint {colorUseRule=UseColor_Specified, specifiedColor=<rgba 0.85 0.6543 0 1>}
   }
   slot {
    class=CanvasPanelSlot
    properties {layoutData {alignment=<0.5 0.5>, anchors {maximum=<0.5 0.5>, minimum=<0.5 0.5>}, offsets=<ltrb 0 -37 80 20>}}
   }
  }
  Body_Body-Movable-Pivot {
   widgetClassPath=&EUW_Picker_Button
   bIsVariable=true
   parent=CanvasPanel_0
   widget=Body_Body-Movable-Pivot
   properties {
    draw Style=Square
    label Font Size=5
    label Text=PIV
    normal Tint {colorUseRule=UseColor_Specified, specifiedColor=<rgba 0.6174 0.0418 0.3047 1>}
   }
   slot {
    class=CanvasPanelSlot
    properties {layoutData {alignment=<0.5 0.5>, anchors {maximum=<0.5 0.5>, minimum=<0.5 0.5>}, offsets=<ltrb -70 -61 20 20>}}
   }
  }
  Body_Body-Aim {
   widgetClassPath=&EUW_Picker_Button
   bIsVariable=true
   parent=CanvasPanel_0
   widget=Body_Body-Aim
   properties {
    draw Style=Square
    label Font Size=5
    label Text=AIM
    normal Tint {colorUseRule=UseColor_Specified, specifiedColor=<rgba 0.6174 0.0418 0.3047 1>}
   }
   slot {
    class=CanvasPanelSlot
    properties {layoutData {alignment=<0.5 0.5>, anchors {maximum=<0.5 0.5>, minimum=<0.5 0.5>}, offsets=<ltrb -62.9609 -174.5405 20 20>}}
   }
  }
  Spine_Pelvis-Local {
   widgetClassPath=&EUW_Picker_Button
   bIsVariable=true
   parent=CanvasPanel_0
   widget=Spine_Pelvis-Local
   properties {
    draw Style=Square
    label Font Size=5
    label Text=LOC
    normal Tint {colorUseRule=UseColor_Specified, specifiedColor=<rgba 0.6174 0.145 0.6174 1>}
   }
   slot {
    class=CanvasPanelSlot
    properties {layoutData {alignment=<0.5 0.5>, anchors {maximum=<0.5 0.5>, minimum=<0.5 0.5>}, offsets=<ltrb -70 -37 20 20>}}
   }
  }
  Body {
   widgetClassPath=&EUW_Picker_Button
   bIsVariable=true
   parent=CanvasPanel_0
   widget=Body
   properties {
    draw Style=Square
    label Font Size=12
    label Text=BODY
    label Color {colorUseRule=UseColor_Specified, specifiedColor=<rgba 0 0 0 1>}
    normal Tint {colorUseRule=UseColor_Specified, specifiedColor=<rgba 0.85 0.6543 0 1>}
   }
   slot {
    class=CanvasPanelSlot
    properties {layoutData {alignment=<0.5 0.5>, anchors {maximum=<0.5 0.5>, minimum=<0.5 0.5>}, offsets=<ltrb 0 -61 100 20>}}
   }
  }
  Clavicle-R_FK {
   widgetClassPath=&EUW_Picker_Button
   bIsVariable=true
   parent=CanvasPanel_0
   widget=Clavicle-R_FK
   properties {
    draw Style=Square
    label Text=CLAV
    label Color {colorUseRule=UseColor_Specified, specifiedColor=<rgba 0 0 0 1>}
    normal Tint {colorUseRule=UseColor_Specified, specifiedColor=<rgba 1 0.02 0.02 1>}
   }
   slot {
    class=CanvasPanelSlot
    properties {layoutData {alignment=<0.5 0.5>, anchors {maximum=<0.5 0.5>, minimum=<0.5 0.5>}, offsets=<ltrb -46 -235 60 25>}}
   }
  }
  Arm-R_FK-0 {
   widgetClassPath=&EUW_Picker_Button
   bIsVariable=true
   parent=CanvasPanel_0
   widget=Arm-R_FK-0
   properties {draw Style=Square, label Font Size=5, normal Tint {colorUseRule=UseColor_Specified, specifiedColor=<rgba 1 0.02 0.02 1>}}
   slot {
    class=CanvasPanelSlot
    properties {layoutData {alignment=<0.5 0.5>, anchors {maximum=<0.5 0.5>, minimum=<0.5 0.5>}, offsets=<ltrb -120 -235 75 25>}}
   }
  }
  Arm-R_Mid {
   widgetClassPath=&EUW_Picker_Button
   bIsVariable=true
   parent=CanvasPanel_0
   widget=Arm-R_Mid
   properties {
    draw Style=Square
    label Font Size=5
    label Text=PIN
    label Color {colorUseRule=UseColor_Specified, specifiedColor=<rgba 0 0 0 1>}
    normal Tint {colorUseRule=UseColor_Specified, specifiedColor=<rgba 1 0.2448 0.2448 1>}
   }
   slot {
    class=CanvasPanelSlot
    properties {layoutData {alignment=<0.5 0.5>, anchors {maximum=<0.5 0.5>, minimum=<0.5 0.5>}, offsets=<ltrb -160 -177 20 20>}}
   }
  }
  Arm-R_PV {
   widgetClassPath=&EUW_Picker_Button
   bIsVariable=true
   parent=CanvasPanel_0
   widget=Arm-R_PV
   properties {
    draw Style=Square
    label Font Size=8
    label Text=PV
    label Color {colorUseRule=UseColor_Specified, specifiedColor=<rgba 0 0 0 1>}
    normal Tint {colorUseRule=UseColor_Specified, specifiedColor=<rgba 1 0.2462 0.2462 1>}
   }
   slot {
    class=CanvasPanelSlot
    properties {layoutData {alignment=<0.5 0.5>, anchors {maximum=<0.5 0.5>, minimum=<0.5 0.5>}, offsets=<ltrb -160 -205 25 25>}}
   }
  }
  Arm-R_FK-1 {
   widgetClassPath=&EUW_Picker_Button
   bIsVariable=true
   parent=CanvasPanel_0
   widget=Arm-R_FK-1
   properties {draw Style=Square, label Font Size=5, normal Tint {colorUseRule=UseColor_Specified, specifiedColor=<rgba 1 0.02 0.02 1>}}
   slot {
    class=CanvasPanelSlot
    properties {layoutData {alignment=<0.5 0.5>, anchors {maximum=<0.5 0.5>, minimum=<0.5 0.5>}, offsets=<ltrb -200 -235 75 25>}}
   }
  }
  Arm-R_FK-2 {
   widgetClassPath=&EUW_Picker_Button
   bIsVariable=true
   parent=CanvasPanel_0
   widget=Arm-R_FK-2
   properties {
    draw Style=Square
    label Font Size=8
    label Text=FK
    label Color {colorUseRule=UseColor_Specified, specifiedColor=<rgba 0 0 0 1>}
    normal Tint {colorUseRule=UseColor_Specified, specifiedColor=<rgba 1 0.02 0.02 1>}
   }
   slot {
    class=CanvasPanelSlot
    properties {layoutData {alignment=<0.5 0.5>, anchors {maximum=<0.5 0.5>, minimum=<0.5 0.5>}, offsets=<ltrb -255 -235 25 25>}}
   }
  }
  Arm-R_IK {
   widgetClassPath=&EUW_Picker_Button
   bIsVariable=true
   parent=CanvasPanel_0
   widget=Arm-R_IK
   properties {
    draw Style=Square
    label Font Size=8
    label Text=IK
    label Color {colorUseRule=UseColor_Specified, specifiedColor=<rgba 0 0 0 1>}
    normal Tint {colorUseRule=UseColor_Specified, specifiedColor=<rgba 1 0.2462 0.2462 1>}
   }
   slot {
    class=CanvasPanelSlot
    properties {layoutData {alignment=<0.5 0.5>, anchors {maximum=<0.5 0.5>, minimum=<0.5 0.5>}, offsets=<ltrb -255 -205 25 25>}}
   }
  }
  Arm-R_IK-Rotation {
   widgetClassPath=&EUW_Picker_Button
   bIsVariable=true
   parent=CanvasPanel_0
   widget=Arm-R_IK-Rotation
   properties {
    draw Style=Square
    label Font Size=7
    label Text=ROT
    label Color {colorUseRule=UseColor_Specified, specifiedColor=<rgba 0 0 0 1>}
    normal Tint {colorUseRule=UseColor_Specified, specifiedColor=<rgba 1 0.2448 0.2448 1>}
   }
   slot {
    class=CanvasPanelSlot
    properties {layoutData {alignment=<0.5 0.5>, anchors {maximum=<0.5 0.5>, minimum=<0.5 0.5>}, offsets=<ltrb -255 -177 20 20>}}
   }
  }
  Arm-Upper-Twist-R_Twist-1 {
   widgetClassPath=&EUW_Picker_Button
   bIsVariable=true
   parent=CanvasPanel_0
   widget=Arm-Upper-Twist-R_Twist-1
   properties {draw Style=Circle, label Font Size=5, normal Tint {colorUseRule=UseColor_Specified, specifiedColor=<rgba 0.286 0.3905 0.5304 1>}}
   slot {
    class=CanvasPanelSlot
    properties {layoutData {alignment=<0.5 0.5>, anchors {maximum=<0.5 0.5>, minimum=<0.5 0.5>}, offsets=<ltrb -100 -210 15 15>}}
   }
  }
  Arm-Upper-Twist-R_Twist-2 {
   widgetClassPath=&EUW_Picker_Button
   bIsVariable=true
   parent=CanvasPanel_0
   widget=Arm-Upper-Twist-R_Twist-2
   properties {draw Style=Circle, label Font Size=5, normal Tint {colorUseRule=UseColor_Specified, specifiedColor=<rgba 0.286 0.3905 0.5304 1>}}
   slot {
    class=CanvasPanelSlot
    properties {layoutData {alignment=<0.5 0.5>, anchors {maximum=<0.5 0.5>, minimum=<0.5 0.5>}, offsets=<ltrb -130 -210 15 15>}}
   }
  }
  Arm-Lower-Twist-R_Twist-1 {
   widgetClassPath=&EUW_Picker_Button
   bIsVariable=true
   parent=CanvasPanel_0
   widget=Arm-Lower-Twist-R_Twist-1
   properties {draw Style=Circle, label Font Size=5, normal Tint {colorUseRule=UseColor_Specified, specifiedColor=<rgba 0.286 0.3905 0.5304 1>}}
   slot {
    class=CanvasPanelSlot
    properties {layoutData {alignment=<0.5 0.5>, anchors {maximum=<0.5 0.5>, minimum=<0.5 0.5>}, offsets=<ltrb -190 -210 15 15>}}
   }
  }
  Arm-Lower-Twist-R_Twist-2 {
   widgetClassPath=&EUW_Picker_Button
   bIsVariable=true
   parent=CanvasPanel_0
   widget=Arm-Lower-Twist-R_Twist-2
   properties {draw Style=Circle, label Font Size=5, normal Tint {colorUseRule=UseColor_Specified, specifiedColor=<rgba 0.286 0.3905 0.5304 1>}}
   slot {
    class=CanvasPanelSlot
    properties {layoutData {alignment=<0.5 0.5>, anchors {maximum=<0.5 0.5>, minimum=<0.5 0.5>}, offsets=<ltrb -220 -210 15 15>}}
   }
  }
  Leg-L_FK-2 {
   widgetClassPath=&EUW_Picker_Button
   bIsVariable=true
   parent=CanvasPanel_0
   widget=Leg-L_FK-2
   properties {
    draw Style=Square
    label Text=FK
    label Color {colorUseRule=UseColor_Specified, specifiedColor=<rgba 0 0 0 1>}
    normal Tint {colorUseRule=UseColor_Specified, specifiedColor=<rgba 0.0272 0.067 0.7826 1>}
   }
   slot {
    class=CanvasPanelSlot
    properties {layoutData {alignment=<0.5 0.5>, anchors {maximum=<0.5 0.5>, minimum=<0.5 0.5>}, offsets=<ltrb 60 271 65 25>}}
   }
  }
  Leg-L_FK-1 {
   widgetClassPath=&EUW_Picker_Button
   bIsVariable=true
   parent=CanvasPanel_0
   widget=Leg-L_FK-1
   properties {draw Style=Square, label Font Size=5, normal Tint {colorUseRule=UseColor_Specified, specifiedColor=<rgba 0.0272 0.067 0.7826 1>}}
   slot {
    class=CanvasPanelSlot
    properties {layoutData {alignment=<0.5 0.5>, anchors {maximum=<0.5 0.5>, minimum=<0.5 0.5>}, offsets=<ltrb 40 187 25 135>}}
   }
  }
  Leg-L_Mid {
   widgetClassPath=&EUW_Picker_Button
   bIsVariable=true
   parent=CanvasPanel_0
   widget=Leg-L_Mid
   properties {
    draw Style=Square
    label Font Size=5
    label Text=PIN
    label Color {colorUseRule=UseColor_Specified, specifiedColor=<rgba 0 0 0 1>}
    normal Tint {colorUseRule=UseColor_Specified, specifiedColor=<rgba 0.2242 0.2536 0.7826 1>}
   }
   slot {
    class=CanvasPanelSlot
    properties {layoutData {alignment=<0.5 0.5>, anchors {maximum=<0.5 0.5>, minimum=<0.5 0.5>}, offsets=<ltrb 102 120 20 20>}}
   }
  }
  Leg-R_Mid {
   widgetClassPath=&EUW_Picker_Button
   bIsVariable=true
   parent=CanvasPanel_0
   widget=Leg-R_Mid
   properties {
    draw Style=Square
    label Font Size=5
    label Text=PIN
    label Color {colorUseRule=UseColor_Specified, specifiedColor=<rgba 0 0 0 1>}
    normal Tint {colorUseRule=UseColor_Specified, specifiedColor=<rgba 1 0.2448 0.2448 1>}
   }
   slot {
    class=CanvasPanelSlot
    properties {layoutData {alignment=<0.5 0.5>, anchors {maximum=<0.5 0.5>, minimum=<0.5 0.5>}, offsets=<ltrb -102 120 20 20>}}
   }
  }
  Leg-L_FK-0 {
   widgetClassPath=&EUW_Picker_Button
   bIsVariable=true
   parent=CanvasPanel_0
   widget=Leg-L_FK-0
   properties {draw Style=Square, label Font Size=5, normal Tint {colorUseRule=UseColor_Specified, specifiedColor=<rgba 0.0272 0.067 0.7826 1>}}
   slot {
    class=CanvasPanelSlot
    properties {layoutData {alignment=<0.5 0.5>, anchors {maximum=<0.5 0.5>, minimum=<0.5 0.5>}, offsets=<ltrb 40 46.3555 25 135>}}
   }
  }
  Leg-Upper-Twist-L_Twist-1 {
   widgetClassPath=&EUW_Picker_Button
   bIsVariable=true
   parent=CanvasPanel_0
   widget=Leg-Upper-Twist-L_Twist-1
   properties {draw Style=Circle, label Font Size=5, normal Tint {colorUseRule=UseColor_Specified, specifiedColor=<rgba 0.286 0.3905 0.5304 1>}}
   slot {
    class=CanvasPanelSlot
    properties {layoutData {alignment=<0.5 0.5>, anchors {maximum=<0.5 0.5>, minimum=<0.5 0.5>}, offsets=<ltrb 70.5391 26.9595 15 15>}}
   }
  }
  Leg-L_PV {
   widgetClassPath=&EUW_Picker_Button
   bIsVariable=true
   parent=CanvasPanel_0
   widget=Leg-L_PV
   properties {
    draw Style=Square
    label Font Size=8
    label Text=PV
    label Color {colorUseRule=UseColor_Specified, specifiedColor=<rgba 0 0 0 1>}
    normal Tint {colorUseRule=UseColor_Specified, specifiedColor=<rgba 0.2232 0.2542 0.7835 1>}
   }
   slot {
    class=CanvasPanelSlot
    properties {layoutData {alignment=<0.5 0.5>, anchors {maximum=<0.5 0.5>, minimum=<0.5 0.5>}, offsets=<ltrb 75 120 25 25>}}
   }
  }
  Leg-L_IK {
   widgetClassPath=&EUW_Picker_Button
   bIsVariable=true
   parent=CanvasPanel_0
   widget=Leg-L_IK
   properties {
    draw Style=Square
    label Text=IK
    label Color {colorUseRule=UseColor_Specified, specifiedColor=<rgba 0 0 0 1>}
    normal Tint {colorUseRule=UseColor_Specified, specifiedColor=<rgba 0.2232 0.2542 0.7835 1>}
   }
   slot {
    class=CanvasPanelSlot
    properties {layoutData {alignment=<0.5 0.5>, anchors {maximum=<0.5 0.5>, minimum=<0.5 0.5>}, offsets=<ltrb 60 300 65 25>}}
   }
  }
  Foot-L_Toes-FK {
   widgetClassPath=&EUW_Picker_Button
   bIsVariable=true
   parent=CanvasPanel_0
   widget=Foot-L_Toes-FK
   properties {
    draw Style=Square
    label Font Size=8
    label Text=TOE
    label Color {colorUseRule=UseColor_Specified, specifiedColor=<rgba 0 0 0 1>}
    normal Tint {colorUseRule=UseColor_Specified, specifiedColor=<rgba 0.0272 0.067 0.7826 1>}
   }
   slot {
    class=CanvasPanelSlot
    properties {layoutData {alignment=<0.5 0.5>, anchors {maximum=<0.5 0.5>, minimum=<0.5 0.5>}, offsets=<ltrb 110 271 25 25>}}
   }
  }
  Foot-L_Toes-IK {
   widgetClassPath=&EUW_Picker_Button
   bIsVariable=true
   parent=CanvasPanel_0
   widget=Foot-L_Toes-IK
   properties {
    draw Style=Square
    label Font Size=8
    label Text=TOE
    label Color {colorUseRule=UseColor_Specified, specifiedColor=<rgba 0 0 0 1>}
    normal Tint {colorUseRule=UseColor_Specified, specifiedColor=<rgba 0.2232 0.2542 0.7835 1>}
   }
   slot {
    class=CanvasPanelSlot
    properties {layoutData {alignment=<0.5 0.5>, anchors {maximum=<0.5 0.5>, minimum=<0.5 0.5>}, offsets=<ltrb 110 300 25 25>}}
   }
  }
  Foot-L_Heel {
   widgetClassPath=&EUW_Picker_Button
   bIsVariable=true
   parent=CanvasPanel_0
   widget=Foot-L_Heel
   properties {
    draw Style=Square
    label Font Size=6
    label Text=HEEL
    label Color {colorUseRule=UseColor_Specified, specifiedColor=<rgba 0 0 0 1>}
    normal Tint {colorUseRule=UseColor_Specified, specifiedColor=<rgba 0.2242 0.2536 0.7826 1>}
   }
   slot {
    class=CanvasPanelSlot
    properties {layoutData {alignment=<0.5 0.5>, anchors {maximum=<0.5 0.5>, minimum=<0.5 0.5>}, offsets=<ltrb 39 337.9595 23 13>}}
   }
  }
  Foot-R_Heel {
   widgetClassPath=&EUW_Picker_Button
   bIsVariable=true
   parent=CanvasPanel_0
   widget=Foot-R_Heel
   properties {
    draw Style=Square
    label Font Size=6
    label Text=HEEL
    label Color {colorUseRule=UseColor_Specified, specifiedColor=<rgba 0 0 0 1>}
    normal Tint {colorUseRule=UseColor_Specified, specifiedColor=<rgba 1 0.2448 0.2448 1>}
   }
   slot {
    class=CanvasPanelSlot
    properties {layoutData {alignment=<0.5 0.5>, anchors {maximum=<0.5 0.5>, minimum=<0.5 0.5>}, offsets=<ltrb -39 337.9595 23 13>}}
   }
  }
  Foot-L_Foot-Rocker {
   widgetClassPath=&EUW_Picker_Button
   bIsVariable=true
   parent=CanvasPanel_0
   widget=Foot-L_Foot-Rocker
   properties {
    draw Style=Square
    label Font Size=6
    label Text=ROCK
    label Color {colorUseRule=UseColor_Specified, specifiedColor=<rgba 0 0 0 1>}
    normal Tint {colorUseRule=UseColor_Specified, specifiedColor=<rgba 0.2242 0.2536 0.7826 1>}
   }
   slot {
    class=CanvasPanelSlot
    properties {layoutData {alignment=<0.5 0.5>, anchors {maximum=<0.5 0.5>, minimum=<0.5 0.5>}, offsets=<ltrb 39 321.9595 23 13>}}
   }
  }
  Leg-Upper-Twist-L_Twist-2 {
   widgetClassPath=&EUW_Picker_Button
   bIsVariable=true
   parent=CanvasPanel_0
   widget=Leg-Upper-Twist-L_Twist-2
   properties {draw Style=Circle, label Font Size=5, normal Tint {colorUseRule=UseColor_Specified, specifiedColor=<rgba 0.286 0.3905 0.5304 1>}}
   slot {
    class=CanvasPanelSlot
    properties {layoutData {alignment=<0.5 0.5>, anchors {maximum=<0.5 0.5>, minimum=<0.5 0.5>}, offsets=<ltrb 70.5391 70.9595 15 15>}}
   }
  }
  Leg-Lower-Twist-L_Twist-1 {
   widgetClassPath=&EUW_Picker_Button
   bIsVariable=true
   parent=CanvasPanel_0
   widget=Leg-Lower-Twist-L_Twist-1
   properties {draw Style=Circle, label Font Size=5, normal Tint {colorUseRule=UseColor_Specified, specifiedColor=<rgba 0.286 0.3905 0.5304 1>}}
   slot {
    class=CanvasPanelSlot
    properties {layoutData {alignment=<0.5 0.5>, anchors {maximum=<0.5 0.5>, minimum=<0.5 0.5>}, offsets=<ltrb 70 160 15 15>}}
   }
  }
  Leg-Lower-Twist-L_Twist-2 {
   widgetClassPath=&EUW_Picker_Button
   bIsVariable=true
   parent=CanvasPanel_0
   widget=Leg-Lower-Twist-L_Twist-2
   properties {draw Style=Circle, label Font Size=5, normal Tint {colorUseRule=UseColor_Specified, specifiedColor=<rgba 0.286 0.3905 0.5304 1>}}
   slot {
    class=CanvasPanelSlot
    properties {layoutData {alignment=<0.5 0.5>, anchors {maximum=<0.5 0.5>, minimum=<0.5 0.5>}, offsets=<ltrb 70 200 15 15>}}
   }
  }
  Leg-R_FK-0 {
   widgetClassPath=&EUW_Picker_Button
   bIsVariable=true
   parent=CanvasPanel_0
   widget=Leg-R_FK-0
   properties {draw Style=Square, label Font Size=5, normal Tint {colorUseRule=UseColor_Specified, specifiedColor=<rgba 1 0.02 0.02 1>}}
   slot {
    class=CanvasPanelSlot
    properties {layoutData {alignment=<0.5 0.5>, anchors {maximum=<0.5 0.5>, minimum=<0.5 0.5>}, offsets=<ltrb -40 46.3555 25 135>}}
   }
  }
  Leg-R_FK-1 {
   widgetClassPath=&EUW_Picker_Button
   bIsVariable=true
   parent=CanvasPanel_0
   widget=Leg-R_FK-1
   properties {draw Style=Square, label Font Size=5, normal Tint {colorUseRule=UseColor_Specified, specifiedColor=<rgba 1 0.02 0.02 1>}}
   slot {
    class=CanvasPanelSlot
    properties {layoutData {alignment=<0.5 0.5>, anchors {maximum=<0.5 0.5>, minimum=<0.5 0.5>}, offsets=<ltrb -40 187 25 135>}}
   }
  }
  Leg-R_FK-2 {
   widgetClassPath=&EUW_Picker_Button
   bIsVariable=true
   parent=CanvasPanel_0
   widget=Leg-R_FK-2
   properties {
    draw Style=Square
    label Text=FK
    label Color {colorUseRule=UseColor_Specified, specifiedColor=<rgba 0 0 0 1>}
    normal Tint {colorUseRule=UseColor_Specified, specifiedColor=<rgba 1 0.02 0.02 1>}
   }
   slot {
    class=CanvasPanelSlot
    properties {layoutData {alignment=<0.5 0.5>, anchors {maximum=<0.5 0.5>, minimum=<0.5 0.5>}, offsets=<ltrb -60 271 65 25>}}
   }
  }
  Leg-R_IK {
   widgetClassPath=&EUW_Picker_Button
   bIsVariable=true
   parent=CanvasPanel_0
   widget=Leg-R_IK
   properties {
    draw Style=Square
    label Text=IK
    label Color {colorUseRule=UseColor_Specified, specifiedColor=<rgba 0 0 0 1>}
    normal Tint {colorUseRule=UseColor_Specified, specifiedColor=<rgba 1 0.2462 0.2462 1>}
   }
   slot {
    class=CanvasPanelSlot
    properties {layoutData {alignment=<0.5 0.5>, anchors {maximum=<0.5 0.5>, minimum=<0.5 0.5>}, offsets=<ltrb -60 300 65 25>}}
   }
  }
  Leg-R_IK-Base {
   widgetClassPath=&EUW_Picker_Button
   bIsVariable=true
   parent=CanvasPanel_0
   widget=Leg-R_IK-Base
   properties {
    draw Style=Square
    label Font Size=6
    label Text=BASE
    label Color {colorUseRule=UseColor_Specified, specifiedColor=<rgba 0 0 0 1>}
    normal Tint {colorUseRule=UseColor_Specified, specifiedColor=<rgba 1 0.2462 0.2462 1>}
   }
   slot {
    class=CanvasPanelSlot
    properties {layoutData {alignment=<0.5 0.5>, anchors {maximum=<0.5 0.5>, minimum=<0.5 0.5>}, offsets=<ltrb -75 -8 25 25>}}
   }
  }
  Leg-L_IK-Base {
   widgetClassPath=&EUW_Picker_Button
   bIsVariable=true
   parent=CanvasPanel_0
   widget=Leg-L_IK-Base
   properties {
    draw Style=Square
    label Font Size=6
    label Text=BASE
    label Color {colorUseRule=UseColor_Specified, specifiedColor=<rgba 0 0 0 1>}
    normal Tint {colorUseRule=UseColor_Specified, specifiedColor=<rgba 0.2232 0.2542 0.7835 1>}
   }
   slot {
    class=CanvasPanelSlot
    properties {layoutData {alignment=<0.5 0.5>, anchors {maximum=<0.5 0.5>, minimum=<0.5 0.5>}, offsets=<ltrb 75 -8 25 25>}}
   }
  }
  Foot-R_Toes-FK {
   widgetClassPath=&EUW_Picker_Button
   bIsVariable=true
   parent=CanvasPanel_0
   widget=Foot-R_Toes-FK
   properties {
    draw Style=Square
    label Font Size=8
    label Text=TOE
    label Color {colorUseRule=UseColor_Specified, specifiedColor=<rgba 0 0 0 1>}
    normal Tint {colorUseRule=UseColor_Specified, specifiedColor=<rgba 1 0.02 0.02 1>}
   }
   slot {
    class=CanvasPanelSlot
    properties {layoutData {alignment=<0.5 0.5>, anchors {maximum=<0.5 0.5>, minimum=<0.5 0.5>}, offsets=<ltrb -110 271 25 25>}}
   }
  }
  Foot-R_Foot-Rocker {
   widgetClassPath=&EUW_Picker_Button
   bIsVariable=true
   parent=CanvasPanel_0
   widget=Foot-R_Foot-Rocker
   properties {
    draw Style=Square
    label Font Size=6
    label Text=ROCK
    label Color {colorUseRule=UseColor_Specified, specifiedColor=<rgba 0 0 0 1>}
    normal Tint {colorUseRule=UseColor_Specified, specifiedColor=<rgba 1 0.2448 0.2448 1>}
   }
   slot {
    class=CanvasPanelSlot
    properties {layoutData {alignment=<0.5 0.5>, anchors {maximum=<0.5 0.5>, minimum=<0.5 0.5>}, offsets=<ltrb -39 321.9595 23 13>}}
   }
  }
  Leg-R_PV {
   widgetClassPath=&EUW_Picker_Button
   bIsVariable=true
   parent=CanvasPanel_0
   widget=Leg-R_PV
   properties {
    draw Style=Square
    label Font Size=8
    label Text=PV
    label Color {colorUseRule=UseColor_Specified, specifiedColor=<rgba 0 0 0 1>}
    normal Tint {colorUseRule=UseColor_Specified, specifiedColor=<rgba 1 0.2462 0.2462 1>}
   }
   slot {
    class=CanvasPanelSlot
    properties {layoutData {alignment=<0.5 0.5>, anchors {maximum=<0.5 0.5>, minimum=<0.5 0.5>}, offsets=<ltrb -75 120 25 25>}}
   }
  }
  Leg-Upper-Twist-R_Twist-1 {
   widgetClassPath=&EUW_Picker_Button
   bIsVariable=true
   parent=CanvasPanel_0
   widget=Leg-Upper-Twist-R_Twist-1
   properties {draw Style=Circle, label Font Size=5, normal Tint {colorUseRule=UseColor_Specified, specifiedColor=<rgba 0.286 0.3905 0.5304 1>}}
   slot {
    class=CanvasPanelSlot
    properties {layoutData {alignment=<0.5 0.5>, anchors {maximum=<0.5 0.5>, minimum=<0.5 0.5>}, offsets=<ltrb -69.4609 26.9595 15 15>}}
   }
  }
  Leg-Upper-Twist-R_Twist-2 {
   widgetClassPath=&EUW_Picker_Button
   bIsVariable=true
   parent=CanvasPanel_0
   widget=Leg-Upper-Twist-R_Twist-2
   properties {draw Style=Circle, label Font Size=5, normal Tint {colorUseRule=UseColor_Specified, specifiedColor=<rgba 0.286 0.3905 0.5304 1>}}
   slot {
    class=CanvasPanelSlot
    properties {layoutData {alignment=<0.5 0.5>, anchors {maximum=<0.5 0.5>, minimum=<0.5 0.5>}, offsets=<ltrb -73.4609 70.9595 15 15>}}
   }
  }
  Leg-Lower-Twist-R_Twist-2 {
   widgetClassPath=&EUW_Picker_Button
   bIsVariable=true
   parent=CanvasPanel_0
   widget=Leg-Lower-Twist-R_Twist-2
   properties {draw Style=Circle, label Font Size=5, normal Tint {colorUseRule=UseColor_Specified, specifiedColor=<rgba 0.286 0.3905 0.5304 1>}}
   slot {
    class=CanvasPanelSlot
    properties {layoutData {alignment=<0.5 0.5>, anchors {maximum=<0.5 0.5>, minimum=<0.5 0.5>}, offsets=<ltrb -70 200 15 15>}}
   }
  }
  Leg-Lower-Twist-R_Twist-1 {
   widgetClassPath=&EUW_Picker_Button
   bIsVariable=true
   parent=CanvasPanel_0
   widget=Leg-Lower-Twist-R_Twist-1
   properties {draw Style=Circle, label Font Size=5, normal Tint {colorUseRule=UseColor_Specified, specifiedColor=<rgba 0.286 0.3905 0.5304 1>}}
   slot {
    class=CanvasPanelSlot
    properties {layoutData {alignment=<0.5 0.5>, anchors {maximum=<0.5 0.5>, minimum=<0.5 0.5>}, offsets=<ltrb -70 160 15 15>}}
   }
  }
  Clavicle-L_FK {
   widgetClassPath=&EUW_Picker_Button
   bIsVariable=true
   parent=CanvasPanel_0
   widget=Clavicle-L_FK
   properties {
    draw Style=Square
    label Text=CLAV
    label Color {colorUseRule=UseColor_Specified, specifiedColor=<rgba 0 0 0 1>}
    normal Tint {colorUseRule=UseColor_Specified, specifiedColor=<rgba 0.0272 0.067 0.7826 1>}
   }
   slot {
    class=CanvasPanelSlot
    properties {layoutData {alignment=<0.5 0.5>, anchors {maximum=<0.5 0.5>, minimum=<0.5 0.5>}, offsets=<ltrb 46 -235 60 25>}}
   }
  }
  Arm-L_FK-0 {
   widgetClassPath=&EUW_Picker_Button
   bIsVariable=true
   parent=CanvasPanel_0
   widget=Arm-L_FK-0
   properties {draw Style=Square, label Font Size=5, normal Tint {colorUseRule=UseColor_Specified, specifiedColor=<rgba 0.0272 0.067 0.7826 1>}}
   slot {
    class=CanvasPanelSlot
    properties {layoutData {alignment=<0.5 0.5>, anchors {maximum=<0.5 0.5>, minimum=<0.5 0.5>}, offsets=<ltrb 120 -235 75 25>}}
   }
  }
  Arm-L_FK-1 {
   widgetClassPath=&EUW_Picker_Button
   bIsVariable=true
   parent=CanvasPanel_0
   widget=Arm-L_FK-1
   properties {draw Style=Square, label Font Size=5, normal Tint {colorUseRule=UseColor_Specified, specifiedColor=<rgba 0.0272 0.067 0.7826 1>}}
   slot {
    class=CanvasPanelSlot
    properties {layoutData {alignment=<0.5 0.5>, anchors {maximum=<0.5 0.5>, minimum=<0.5 0.5>}, offsets=<ltrb 200 -235 75 25>}}
   }
  }
  Arm-L_Mid {
   widgetClassPath=&EUW_Picker_Button
   bIsVariable=true
   parent=CanvasPanel_0
   widget=Arm-L_Mid
   properties {
    draw Style=Square
    label Font Size=5
    label Text=PIN
    label Color {colorUseRule=UseColor_Specified, specifiedColor=<rgba 0 0 0 1>}
    normal Tint {colorUseRule=UseColor_Specified, specifiedColor=<rgba 0.2242 0.2536 0.7826 1>}
   }
   slot {
    class=CanvasPanelSlot
    properties {layoutData {alignment=<0.5 0.5>, anchors {maximum=<0.5 0.5>, minimum=<0.5 0.5>}, offsets=<ltrb 160 -177 20 20>}}
   }
  }
  Arm-L_FK-2 {
   widgetClassPath=&EUW_Picker_Button
   bIsVariable=true
   parent=CanvasPanel_0
   widget=Arm-L_FK-2
   properties {
    draw Style=Square
    label Font Size=8
    label Text=FK
    label Color {colorUseRule=UseColor_Specified, specifiedColor=<rgba 0 0 0 1>}
    normal Tint {colorUseRule=UseColor_Specified, specifiedColor=<rgba 0.0272 0.067 0.7826 1>}
   }
   slot {
    class=CanvasPanelSlot
    properties {layoutData {alignment=<0.5 0.5>, anchors {maximum=<0.5 0.5>, minimum=<0.5 0.5>}, offsets=<ltrb 255 -235 25 25>}}
   }
  }
  Arm-L_IK {
   widgetClassPath=&EUW_Picker_Button
   bIsVariable=true
   parent=CanvasPanel_0
   widget=Arm-L_IK
   properties {
    draw Style=Square
    label Font Size=8
    label Text=IK
    label Color {colorUseRule=UseColor_Specified, specifiedColor=<rgba 0 0 0 1>}
    normal Tint {colorUseRule=UseColor_Specified, specifiedColor=<rgba 0.2232 0.2542 0.7835 1>}
   }
   slot {
    class=CanvasPanelSlot
    properties {layoutData {alignment=<0.5 0.5>, anchors {maximum=<0.5 0.5>, minimum=<0.5 0.5>}, offsets=<ltrb 255 -205 25 25>}}
   }
  }
  Arm-L_IK-Rotation {
   widgetClassPath=&EUW_Picker_Button
   bIsVariable=true
   parent=CanvasPanel_0
   widget=Arm-L_IK-Rotation
   properties {
    draw Style=Square
    label Font Size=7
    label Text=ROT
    label Color {colorUseRule=UseColor_Specified, specifiedColor=<rgba 0 0 0 1>}
    normal Tint {colorUseRule=UseColor_Specified, specifiedColor=<rgba 0.2242 0.2536 0.7826 1>}
   }
   slot {
    class=CanvasPanelSlot
    properties {layoutData {alignment=<0.5 0.5>, anchors {maximum=<0.5 0.5>, minimum=<0.5 0.5>}, offsets=<ltrb 255 -177 20 20>}}
   }
  }
  Arm-L_PV {
   widgetClassPath=&EUW_Picker_Button
   bIsVariable=true
   parent=CanvasPanel_0
   widget=Arm-L_PV
   properties {
    draw Style=Square
    label Font Size=8
    label Text=PV
    label Color {colorUseRule=UseColor_Specified, specifiedColor=<rgba 0 0 0 1>}
    normal Tint {colorUseRule=UseColor_Specified, specifiedColor=<rgba 0.2232 0.2542 0.7835 1>}
   }
   slot {
    class=CanvasPanelSlot
    properties {layoutData {alignment=<0.5 0.5>, anchors {maximum=<0.5 0.5>, minimum=<0.5 0.5>}, offsets=<ltrb 160 -205 25 25>}}
   }
  }
  Arm-Upper-Twist-L_Twist-1 {
   widgetClassPath=&EUW_Picker_Button
   bIsVariable=true
   parent=CanvasPanel_0
   widget=Arm-Upper-Twist-L_Twist-1
   properties {draw Style=Circle, label Font Size=5, normal Tint {colorUseRule=UseColor_Specified, specifiedColor=<rgba 0.286 0.3905 0.5304 1>}}
   slot {
    class=CanvasPanelSlot
    properties {layoutData {alignment=<0.5 0.5>, anchors {maximum=<0.5 0.5>, minimum=<0.5 0.5>}, offsets=<ltrb 100 -210 15 15>}}
   }
  }
  Arm-Upper-Twist-L_Twist-2 {
   widgetClassPath=&EUW_Picker_Button
   bIsVariable=true
   parent=CanvasPanel_0
   widget=Arm-Upper-Twist-L_Twist-2
   properties {draw Style=Circle, label Font Size=5, normal Tint {colorUseRule=UseColor_Specified, specifiedColor=<rgba 0.286 0.3905 0.5304 1>}}
   slot {
    class=CanvasPanelSlot
    properties {layoutData {alignment=<0.5 0.5>, anchors {maximum=<0.5 0.5>, minimum=<0.5 0.5>}, offsets=<ltrb 130 -210 15 15>}}
   }
  }
  Arm-Lower-Twist-L_Twist-1 {
   widgetClassPath=&EUW_Picker_Button
   bIsVariable=true
   parent=CanvasPanel_0
   widget=Arm-Lower-Twist-L_Twist-1
   properties {draw Style=Circle, label Font Size=5, normal Tint {colorUseRule=UseColor_Specified, specifiedColor=<rgba 0.286 0.3905 0.5304 1>}}
   slot {
    class=CanvasPanelSlot
    properties {layoutData {alignment=<0.5 0.5>, anchors {maximum=<0.5 0.5>, minimum=<0.5 0.5>}, offsets=<ltrb 190 -210 15 15>}}
   }
  }
  Arm-Lower-Twist-L_Twist-2 {
   widgetClassPath=&EUW_Picker_Button
   bIsVariable=true
   parent=CanvasPanel_0
   widget=Arm-Lower-Twist-L_Twist-2
   properties {draw Style=Circle, label Font Size=5, normal Tint {colorUseRule=UseColor_Specified, specifiedColor=<rgba 0.286 0.3905 0.5304 1>}}
   slot {
    class=CanvasPanelSlot
    properties {layoutData {alignment=<0.5 0.5>, anchors {maximum=<0.5 0.5>, minimum=<0.5 0.5>}, offsets=<ltrb 220 -210 15 15>}}
   }
  }
  Thumb-R_FK-2 {
   widgetClassPath=&EUW_Picker_Button
   bIsVariable=true
   parent=CanvasPanel_0
   widget=Thumb-R_FK-2
   properties {draw Style=Square, label Font Size=5, normal Tint {colorUseRule=UseColor_Specified, specifiedColor=<rgba 1 0.02 0.02 1>}}
   slot {
    class=CanvasPanelSlot
    properties {layoutData {alignment=<0.5 0.5>, anchors {maximum=<0.5 0.5>, minimum=<0.5 0.5>}, offsets=<ltrb -150 10 15 15>}}
   }
  }
  Thumb-R_FK-1 {
   widgetClassPath=&EUW_Picker_Button
   bIsVariable=true
   parent=CanvasPanel_0
   widget=Thumb-R_FK-1
   properties {draw Style=Square, label Font Size=5, normal Tint {colorUseRule=UseColor_Specified, specifiedColor=<rgba 1 0.02 0.02 1>}}
   slot {
    class=CanvasPanelSlot
    properties {layoutData {alignment=<0.5 0.5>, anchors {maximum=<0.5 0.5>, minimum=<0.5 0.5>}, offsets=<ltrb -156 -10 15 15>}}
   }
  }
  Thumb-R_FK-0 {
   widgetClassPath=&EUW_Picker_Button
   bIsVariable=true
   parent=CanvasPanel_0
   widget=Thumb-R_FK-0
   properties {draw Style=Square, label Font Size=5, normal Tint {colorUseRule=UseColor_Specified, specifiedColor=<rgba 1 0.02 0.02 1>}}
   slot {
    class=CanvasPanelSlot
    properties {layoutData {alignment=<0.5 0.5>, anchors {maximum=<0.5 0.5>, minimum=<0.5 0.5>}, offsets=<ltrb -162 -30 15 15>}}
   }
  }
  Meta-R_Index-Metacarpal-R {
   widgetClassPath=&EUW_Picker_Button
   bIsVariable=true
   parent=CanvasPanel_0
   widget=Meta-R_Index-Metacarpal-R
   properties {draw Style=Square, label Font Size=5, normal Tint {colorUseRule=UseColor_Specified, specifiedColor=<rgba 1 0.2462 0.2462 1>}}
   slot {
    class=CanvasPanelSlot
    properties {layoutData {alignment=<0.5 0.5>, anchors {maximum=<0.5 0.5>, minimum=<0.5 0.5>}, offsets=<ltrb -190 -25 15 15>}}
   }
  }
  Meta-R_Middle-Metacarpal-R {
   widgetClassPath=&EUW_Picker_Button
   bIsVariable=true
   parent=CanvasPanel_0
   widget=Meta-R_Middle-Metacarpal-R
   properties {draw Style=Square, label Font Size=5, normal Tint {colorUseRule=UseColor_Specified, specifiedColor=<rgba 1 0.2462 0.2462 1>}}
   slot {
    class=CanvasPanelSlot
    properties {layoutData {alignment=<0.5 0.5>, anchors {maximum=<0.5 0.5>, minimum=<0.5 0.5>}, offsets=<ltrb -212 -25 15 15>}}
   }
  }
  Meta-R_Ring-Metacarpal-R {
   widgetClassPath=&EUW_Picker_Button
   bIsVariable=true
   parent=CanvasPanel_0
   widget=Meta-R_Ring-Metacarpal-R
   properties {draw Style=Square, label Font Size=5, normal Tint {colorUseRule=UseColor_Specified, specifiedColor=<rgba 1 0.2462 0.2462 1>}}
   slot {
    class=CanvasPanelSlot
    properties {layoutData {alignment=<0.5 0.5>, anchors {maximum=<0.5 0.5>, minimum=<0.5 0.5>}, offsets=<ltrb -234 -25 15 15>}}
   }
  }
  Meta-R_Pinky-Metacarpal-R {
   widgetClassPath=&EUW_Picker_Button
   bIsVariable=true
   parent=CanvasPanel_0
   widget=Meta-R_Pinky-Metacarpal-R
   properties {draw Style=Square, label Font Size=5, normal Tint {colorUseRule=UseColor_Specified, specifiedColor=<rgba 1 0.2462 0.2462 1>}}
   slot {
    class=CanvasPanelSlot
    properties {layoutData {alignment=<0.5 0.5>, anchors {maximum=<0.5 0.5>, minimum=<0.5 0.5>}, offsets=<ltrb -256 -25 15 15>}}
   }
  }
  Index-R_FK-0 {
   widgetClassPath=&EUW_Picker_Button
   bIsVariable=true
   parent=CanvasPanel_0
   widget=Index-R_FK-0
   properties {draw Style=Square, label Font Size=5, normal Tint {colorUseRule=UseColor_Specified, specifiedColor=<rgba 1 0.02 0.02 1>}}
   slot {
    class=CanvasPanelSlot
    properties {layoutData {alignment=<0.5 0.5>, anchors {maximum=<0.5 0.5>, minimum=<0.5 0.5>}, offsets=<ltrb -190 -5 15 15>}}
   }
  }
  Index-R_FK-1 {
   widgetClassPath=&EUW_Picker_Button
   bIsVariable=true
   parent=CanvasPanel_0
   widget=Index-R_FK-1
   properties {draw Style=Square, label Font Size=5, normal Tint {colorUseRule=UseColor_Specified, specifiedColor=<rgba 1 0.02 0.02 1>}}
   slot {
    class=CanvasPanelSlot
    properties {layoutData {alignment=<0.5 0.5>, anchors {maximum=<0.5 0.5>, minimum=<0.5 0.5>}, offsets=<ltrb -190 15 15 15>}}
   }
  }
  Index-R_FK-2 {
   widgetClassPath=&EUW_Picker_Button
   bIsVariable=true
   parent=CanvasPanel_0
   widget=Index-R_FK-2
   properties {draw Style=Square, label Font Size=5, normal Tint {colorUseRule=UseColor_Specified, specifiedColor=<rgba 1 0.02 0.02 1>}}
   slot {
    class=CanvasPanelSlot
    properties {layoutData {alignment=<0.5 0.5>, anchors {maximum=<0.5 0.5>, minimum=<0.5 0.5>}, offsets=<ltrb -190 35 15 15>}}
   }
  }
  Middle-R_FK-0 {
   widgetClassPath=&EUW_Picker_Button
   bIsVariable=true
   parent=CanvasPanel_0
   widget=Middle-R_FK-0
   properties {draw Style=Square, label Font Size=5, normal Tint {colorUseRule=UseColor_Specified, specifiedColor=<rgba 1 0.02 0.02 1>}}
   slot {
    class=CanvasPanelSlot
    properties {layoutData {alignment=<0.5 0.5>, anchors {maximum=<0.5 0.5>, minimum=<0.5 0.5>}, offsets=<ltrb -212 -5 15 15>}}
   }
  }
  Middle-R_FK-1 {
   widgetClassPath=&EUW_Picker_Button
   bIsVariable=true
   parent=CanvasPanel_0
   widget=Middle-R_FK-1
   properties {draw Style=Square, label Font Size=5, normal Tint {colorUseRule=UseColor_Specified, specifiedColor=<rgba 1 0.02 0.02 1>}}
   slot {
    class=CanvasPanelSlot
    properties {layoutData {alignment=<0.5 0.5>, anchors {maximum=<0.5 0.5>, minimum=<0.5 0.5>}, offsets=<ltrb -212 15 15 15>}}
   }
  }
  Middle-R_FK-2 {
   widgetClassPath=&EUW_Picker_Button
   bIsVariable=true
   parent=CanvasPanel_0
   widget=Middle-R_FK-2
   properties {draw Style=Square, label Font Size=5, normal Tint {colorUseRule=UseColor_Specified, specifiedColor=<rgba 1 0.02 0.02 1>}}
   slot {
    class=CanvasPanelSlot
    properties {layoutData {alignment=<0.5 0.5>, anchors {maximum=<0.5 0.5>, minimum=<0.5 0.5>}, offsets=<ltrb -212 35 15 15>}}
   }
  }
  Ring-R_FK-0 {
   widgetClassPath=&EUW_Picker_Button
   bIsVariable=true
   parent=CanvasPanel_0
   widget=Ring-R_FK-0
   properties {draw Style=Square, label Font Size=5, normal Tint {colorUseRule=UseColor_Specified, specifiedColor=<rgba 1 0.02 0.02 1>}}
   slot {
    class=CanvasPanelSlot
    properties {layoutData {alignment=<0.5 0.5>, anchors {maximum=<0.5 0.5>, minimum=<0.5 0.5>}, offsets=<ltrb -234 -5 15 15>}}
   }
  }
  Ring-R_FK-1 {
   widgetClassPath=&EUW_Picker_Button
   bIsVariable=true
   parent=CanvasPanel_0
   widget=Ring-R_FK-1
   properties {draw Style=Square, label Font Size=5, normal Tint {colorUseRule=UseColor_Specified, specifiedColor=<rgba 1 0.02 0.02 1>}}
   slot {
    class=CanvasPanelSlot
    properties {layoutData {alignment=<0.5 0.5>, anchors {maximum=<0.5 0.5>, minimum=<0.5 0.5>}, offsets=<ltrb -234 15 15 15>}}
   }
  }
  Ring-R_FK-2 {
   widgetClassPath=&EUW_Picker_Button
   bIsVariable=true
   parent=CanvasPanel_0
   widget=Ring-R_FK-2
   properties {draw Style=Square, label Font Size=5, normal Tint {colorUseRule=UseColor_Specified, specifiedColor=<rgba 1 0.02 0.02 1>}}
   slot {
    class=CanvasPanelSlot
    properties {layoutData {alignment=<0.5 0.5>, anchors {maximum=<0.5 0.5>, minimum=<0.5 0.5>}, offsets=<ltrb -234 35 15 15>}}
   }
  }
  Pinky-R_FK-0 {
   widgetClassPath=&EUW_Picker_Button
   bIsVariable=true
   parent=CanvasPanel_0
   widget=Pinky-R_FK-0
   properties {draw Style=Square, label Font Size=5, normal Tint {colorUseRule=UseColor_Specified, specifiedColor=<rgba 1 0.02 0.02 1>}}
   slot {
    class=CanvasPanelSlot
    properties {layoutData {alignment=<0.5 0.5>, anchors {maximum=<0.5 0.5>, minimum=<0.5 0.5>}, offsets=<ltrb -256 -5 15 15>}}
   }
  }
  Pinky-R_FK-1 {
   widgetClassPath=&EUW_Picker_Button
   bIsVariable=true
   parent=CanvasPanel_0
   widget=Pinky-R_FK-1
   properties {draw Style=Square, label Font Size=5, normal Tint {colorUseRule=UseColor_Specified, specifiedColor=<rgba 1 0.02 0.02 1>}}
   slot {
    class=CanvasPanelSlot
    properties {layoutData {alignment=<0.5 0.5>, anchors {maximum=<0.5 0.5>, minimum=<0.5 0.5>}, offsets=<ltrb -256 15 15 15>}}
   }
  }
  Pinky-R_FK-2 {
   widgetClassPath=&EUW_Picker_Button
   bIsVariable=true
   parent=CanvasPanel_0
   widget=Pinky-R_FK-2
   properties {draw Style=Square, label Font Size=5, normal Tint {colorUseRule=UseColor_Specified, specifiedColor=<rgba 1 0.02 0.02 1>}}
   slot {
    class=CanvasPanelSlot
    properties {layoutData {alignment=<0.5 0.5>, anchors {maximum=<0.5 0.5>, minimum=<0.5 0.5>}, offsets=<ltrb -256 35 15 15>}}
   }
  }
  Index-L_FK-0 {
   widgetClassPath=&EUW_Picker_Button
   bIsVariable=true
   parent=CanvasPanel_0
   widget=Index-L_FK-0
   properties {draw Style=Square, label Font Size=5, normal Tint {colorUseRule=UseColor_Specified, specifiedColor=<rgba 0.0272 0.067 0.7826 1>}}
   slot {
    class=CanvasPanelSlot
    properties {layoutData {alignment=<0.5 0.5>, anchors {maximum=<0.5 0.5>, minimum=<0.5 0.5>}, offsets=<ltrb 196 -5 15 15>}}
   }
  }
  Index-L_FK-1 {
   widgetClassPath=&EUW_Picker_Button
   bIsVariable=true
   parent=CanvasPanel_0
   widget=Index-L_FK-1
   properties {draw Style=Square, label Font Size=5, normal Tint {colorUseRule=UseColor_Specified, specifiedColor=<rgba 0.0272 0.067 0.7826 1>}}
   slot {
    class=CanvasPanelSlot
    properties {layoutData {alignment=<0.5 0.5>, anchors {maximum=<0.5 0.5>, minimum=<0.5 0.5>}, offsets=<ltrb 196 15 15 15>}}
   }
  }
  Index-L_FK-2 {
   widgetClassPath=&EUW_Picker_Button
   bIsVariable=true
   parent=CanvasPanel_0
   widget=Index-L_FK-2
   properties {draw Style=Square, label Font Size=5, normal Tint {colorUseRule=UseColor_Specified, specifiedColor=<rgba 0.0272 0.067 0.7826 1>}}
   slot {
    class=CanvasPanelSlot
    properties {layoutData {alignment=<0.5 0.5>, anchors {maximum=<0.5 0.5>, minimum=<0.5 0.5>}, offsets=<ltrb 196 35 15 15>}}
   }
  }
  Middle-L_FK-0 {
   widgetClassPath=&EUW_Picker_Button
   bIsVariable=true
   parent=CanvasPanel_0
   widget=Middle-L_FK-0
   properties {draw Style=Square, label Font Size=5, normal Tint {colorUseRule=UseColor_Specified, specifiedColor=<rgba 0.0272 0.067 0.7826 1>}}
   slot {
    class=CanvasPanelSlot
    properties {layoutData {alignment=<0.5 0.5>, anchors {maximum=<0.5 0.5>, minimum=<0.5 0.5>}, offsets=<ltrb 218 -5 15 15>}}
   }
  }
  Middle-L_FK-1 {
   widgetClassPath=&EUW_Picker_Button
   bIsVariable=true
   parent=CanvasPanel_0
   widget=Middle-L_FK-1
   properties {draw Style=Square, label Font Size=5, normal Tint {colorUseRule=UseColor_Specified, specifiedColor=<rgba 0.0272 0.067 0.7826 1>}}
   slot {
    class=CanvasPanelSlot
    properties {layoutData {alignment=<0.5 0.5>, anchors {maximum=<0.5 0.5>, minimum=<0.5 0.5>}, offsets=<ltrb 218 15 15 15>}}
   }
  }
  Middle-L_FK-2 {
   widgetClassPath=&EUW_Picker_Button
   bIsVariable=true
   parent=CanvasPanel_0
   widget=Middle-L_FK-2
   properties {draw Style=Square, label Font Size=5, normal Tint {colorUseRule=UseColor_Specified, specifiedColor=<rgba 0.0272 0.067 0.7826 1>}}
   slot {
    class=CanvasPanelSlot
    properties {layoutData {alignment=<0.5 0.5>, anchors {maximum=<0.5 0.5>, minimum=<0.5 0.5>}, offsets=<ltrb 218 35 15 15>}}
   }
  }
  Ring-L_FK-0 {
   widgetClassPath=&EUW_Picker_Button
   bIsVariable=true
   parent=CanvasPanel_0
   widget=Ring-L_FK-0
   properties {draw Style=Square, label Font Size=5, normal Tint {colorUseRule=UseColor_Specified, specifiedColor=<rgba 0.0272 0.067 0.7826 1>}}
   slot {
    class=CanvasPanelSlot
    properties {layoutData {alignment=<0.5 0.5>, anchors {maximum=<0.5 0.5>, minimum=<0.5 0.5>}, offsets=<ltrb 240 -5 15 15>}}
   }
  }
  Ring-L_FK-1 {
   widgetClassPath=&EUW_Picker_Button
   bIsVariable=true
   parent=CanvasPanel_0
   widget=Ring-L_FK-1
   properties {draw Style=Square, label Font Size=5, normal Tint {colorUseRule=UseColor_Specified, specifiedColor=<rgba 0.0272 0.067 0.7826 1>}}
   slot {
    class=CanvasPanelSlot
    properties {layoutData {alignment=<0.5 0.5>, anchors {maximum=<0.5 0.5>, minimum=<0.5 0.5>}, offsets=<ltrb 240 15 15 15>}}
   }
  }
  Ring-L_FK-2 {
   widgetClassPath=&EUW_Picker_Button
   bIsVariable=true
   parent=CanvasPanel_0
   widget=Ring-L_FK-2
   properties {draw Style=Square, label Font Size=5, normal Tint {colorUseRule=UseColor_Specified, specifiedColor=<rgba 0.0272 0.067 0.7826 1>}}
   slot {
    class=CanvasPanelSlot
    properties {layoutData {alignment=<0.5 0.5>, anchors {maximum=<0.5 0.5>, minimum=<0.5 0.5>}, offsets=<ltrb 240 35 15 15>}}
   }
  }
  Pinky-L_FK-0 {
   widgetClassPath=&EUW_Picker_Button
   bIsVariable=true
   parent=CanvasPanel_0
   widget=Pinky-L_FK-0
   properties {draw Style=Square, label Font Size=5, normal Tint {colorUseRule=UseColor_Specified, specifiedColor=<rgba 0.0272 0.067 0.7826 1>}}
   slot {
    class=CanvasPanelSlot
    properties {layoutData {alignment=<0.5 0.5>, anchors {maximum=<0.5 0.5>, minimum=<0.5 0.5>}, offsets=<ltrb 262 -5 15 15>}}
   }
  }
  Pinky-L_FK-1 {
   widgetClassPath=&EUW_Picker_Button
   bIsVariable=true
   parent=CanvasPanel_0
   widget=Pinky-L_FK-1
   properties {draw Style=Square, label Font Size=5, normal Tint {colorUseRule=UseColor_Specified, specifiedColor=<rgba 0.0272 0.067 0.7826 1>}}
   slot {
    class=CanvasPanelSlot
    properties {layoutData {alignment=<0.5 0.5>, anchors {maximum=<0.5 0.5>, minimum=<0.5 0.5>}, offsets=<ltrb 262 15 15 15>}}
   }
  }
  Pinky-L_FK-2 {
   widgetClassPath=&EUW_Picker_Button
   bIsVariable=true
   parent=CanvasPanel_0
   widget=Pinky-L_FK-2
   properties {draw Style=Square, label Font Size=5, normal Tint {colorUseRule=UseColor_Specified, specifiedColor=<rgba 0.0272 0.067 0.7826 1>}}
   slot {
    class=CanvasPanelSlot
    properties {layoutData {alignment=<0.5 0.5>, anchors {maximum=<0.5 0.5>, minimum=<0.5 0.5>}, offsets=<ltrb 262 35 15 15>}}
   }
  }
  Thumb-L_FK-0 {
   widgetClassPath=&EUW_Picker_Button
   bIsVariable=true
   parent=CanvasPanel_0
   widget=Thumb-L_FK-0
   properties {draw Style=Square, label Font Size=5, normal Tint {colorUseRule=UseColor_Specified, specifiedColor=<rgba 0.0272 0.067 0.7826 1>}}
   slot {
    class=CanvasPanelSlot
    properties {layoutData {alignment=<0.5 0.5>, anchors {maximum=<0.5 0.5>, minimum=<0.5 0.5>}, offsets=<ltrb 168 -30 15 15>}}
   }
  }
  Thumb-L_FK-1 {
   widgetClassPath=&EUW_Picker_Button
   bIsVariable=true
   parent=CanvasPanel_0
   widget=Thumb-L_FK-1
   properties {draw Style=Square, label Font Size=5, normal Tint {colorUseRule=UseColor_Specified, specifiedColor=<rgba 0.0272 0.067 0.7826 1>}}
   slot {
    class=CanvasPanelSlot
    properties {layoutData {alignment=<0.5 0.5>, anchors {maximum=<0.5 0.5>, minimum=<0.5 0.5>}, offsets=<ltrb 162 -10 15 15>}}
   }
  }
  Thumb-L_FK-2 {
   widgetClassPath=&EUW_Picker_Button
   bIsVariable=true
   parent=CanvasPanel_0
   widget=Thumb-L_FK-2
   properties {draw Style=Square, label Font Size=5, normal Tint {colorUseRule=UseColor_Specified, specifiedColor=<rgba 0.0272 0.067 0.7826 1>}}
   slot {
    class=CanvasPanelSlot
    properties {layoutData {alignment=<0.5 0.5>, anchors {maximum=<0.5 0.5>, minimum=<0.5 0.5>}, offsets=<ltrb 156 10 15 15>}}
   }
  }
  Meta-L_Pinky-Metacarpal-L {
   widgetClassPath=&EUW_Picker_Button
   bIsVariable=true
   parent=CanvasPanel_0
   widget=Meta-L_Pinky-Metacarpal-L
   properties {draw Style=Square, label Font Size=5, normal Tint {colorUseRule=UseColor_Specified, specifiedColor=<rgba 0.2232 0.2542 0.7835 1>}}
   slot {
    class=CanvasPanelSlot
    properties {layoutData {alignment=<0.5 0.5>, anchors {maximum=<0.5 0.5>, minimum=<0.5 0.5>}, offsets=<ltrb 262 -25 15 15>}}
   }
  }
  Meta-L_Ring-Metacarpal-L {
   widgetClassPath=&EUW_Picker_Button
   bIsVariable=true
   parent=CanvasPanel_0
   widget=Meta-L_Ring-Metacarpal-L
   properties {draw Style=Square, label Font Size=5, normal Tint {colorUseRule=UseColor_Specified, specifiedColor=<rgba 0.2232 0.2542 0.7835 1>}}
   slot {
    class=CanvasPanelSlot
    properties {layoutData {alignment=<0.5 0.5>, anchors {maximum=<0.5 0.5>, minimum=<0.5 0.5>}, offsets=<ltrb 240 -25 15 15>}}
   }
  }
  Meta-L_Middle-Metacarpal-L {
   widgetClassPath=&EUW_Picker_Button
   bIsVariable=true
   parent=CanvasPanel_0
   widget=Meta-L_Middle-Metacarpal-L
   properties {draw Style=Square, label Font Size=5, normal Tint {colorUseRule=UseColor_Specified, specifiedColor=<rgba 0.2232 0.2542 0.7835 1>}}
   slot {
    class=CanvasPanelSlot
    properties {layoutData {alignment=<0.5 0.5>, anchors {maximum=<0.5 0.5>, minimum=<0.5 0.5>}, offsets=<ltrb 218 -25 15 15>}}
   }
  }
  Meta-L_Index-Metacarpal-L {
   widgetClassPath=&EUW_Picker_Button
   bIsVariable=true
   parent=CanvasPanel_0
   widget=Meta-L_Index-Metacarpal-L
   properties {draw Style=Square, label Font Size=5, normal Tint {colorUseRule=UseColor_Specified, specifiedColor=<rgba 0.2232 0.2542 0.7835 1>}}
   slot {
    class=CanvasPanelSlot
    properties {layoutData {alignment=<0.5 0.5>, anchors {maximum=<0.5 0.5>, minimum=<0.5 0.5>}, offsets=<ltrb 196 -25 15 15>}}
   }
  }
  Foot-R_Ball-IK {
   widgetClassPath=&EUW_Picker_Button
   bIsVariable=true
   parent=CanvasPanel_0
   widget=Foot-R_Ball-IK
   properties {
    draw Style=Square
    label Font Size=6
    label Text=ROLL
    label Color {colorUseRule=UseColor_Specified, specifiedColor=<rgba 0 0 0 1>}
    normal Tint {colorUseRule=UseColor_Specified, specifiedColor=<rgba 1 0.2448 0.2448 1>}
   }
   slot {
    class=CanvasPanelSlot
    properties {layoutData {alignment=<0.5 0.5>, anchors {maximum=<0.5 0.5>, minimum=<0.5 0.5>}, offsets=<ltrb -76 321.9595 25 13>}}
   }
  }
  Foot-R_Toe-Tip {
   widgetClassPath=&EUW_Picker_Button
   bIsVariable=true
   parent=CanvasPanel_0
   widget=Foot-R_Toe-Tip
   properties {
    draw Style=Square
    label Font Size=6
    label Text=TIP
    label Color {colorUseRule=UseColor_Specified, specifiedColor=<rgba 0 0 0 1>}
    normal Tint {colorUseRule=UseColor_Specified, specifiedColor=<rgba 1 0.2448 0.2448 1>}
   }
   slot {
    class=CanvasPanelSlot
    properties {layoutData {alignment=<0.5 0.5>, anchors {maximum=<0.5 0.5>, minimum=<0.5 0.5>}, offsets=<ltrb -108.4609 329.9595 25 13>}}
   }
  }
  Foot-R_Ball {
   widgetClassPath=&EUW_Picker_Button
   bIsVariable=true
   parent=CanvasPanel_0
   widget=Foot-R_Ball
   properties {
    draw Style=Square
    label Font Size=6
    label Text=BALL
    label Color {colorUseRule=UseColor_Specified, specifiedColor=<rgba 0 0 0 1>}
    normal Tint {colorUseRule=UseColor_Specified, specifiedColor=<rgba 1 0.2448 0.2448 1>}
   }
   slot {
    class=CanvasPanelSlot
    properties {layoutData {alignment=<0.5 0.5>, anchors {maximum=<0.5 0.5>, minimum=<0.5 0.5>}, offsets=<ltrb -76 337.9595 25 13>}}
   }
  }
  Foot-L_Ball {
   widgetClassPath=&EUW_Picker_Button
   bIsVariable=true
   parent=CanvasPanel_0
   widget=Foot-L_Ball
   properties {
    draw Style=Square
    label Font Size=6
    label Text=BALL
    label Color {colorUseRule=UseColor_Specified, specifiedColor=<rgba 0 0 0 1>}
    normal Tint {colorUseRule=UseColor_Specified, specifiedColor=<rgba 0.2242 0.2536 0.7826 1>}
   }
   slot {
    class=CanvasPanelSlot
    properties {layoutData {alignment=<0.5 0.5>, anchors {maximum=<0.5 0.5>, minimum=<0.5 0.5>}, offsets=<ltrb 80 337.9595 25 13>}}
   }
  }
  Foot-L_Ball-IK {
   widgetClassPath=&EUW_Picker_Button
   bIsVariable=true
   parent=CanvasPanel_0
   widget=Foot-L_Ball-IK
   properties {
    draw Style=Square
    label Font Size=6
    label Text=ROLL
    label Color {colorUseRule=UseColor_Specified, specifiedColor=<rgba 0 0 0 1>}
    normal Tint {colorUseRule=UseColor_Specified, specifiedColor=<rgba 0.2242 0.2536 0.7826 1>}
   }
   slot {
    class=CanvasPanelSlot
    properties {layoutData {alignment=<0.5 0.5>, anchors {maximum=<0.5 0.5>, minimum=<0.5 0.5>}, offsets=<ltrb 80 321.9595 25 13>}}
   }
  }
  Foot-L_Toe-Tip {
   widgetClassPath=&EUW_Picker_Button
   bIsVariable=true
   parent=CanvasPanel_0
   widget=Foot-L_Toe-Tip
   properties {
    draw Style=Square
    label Font Size=6
    label Text=TIP
    label Color {colorUseRule=UseColor_Specified, specifiedColor=<rgba 0 0 0 1>}
    normal Tint {colorUseRule=UseColor_Specified, specifiedColor=<rgba 0.2242 0.2536 0.7826 1>}
   }
   slot {
    class=CanvasPanelSlot
    properties {layoutData {alignment=<0.5 0.5>, anchors {maximum=<0.5 0.5>, minimum=<0.5 0.5>}, offsets=<ltrb 111.5391 329.9595 25 13>}}
   }
  }
  Foot-R_Toes-IK {
   widgetClassPath=&EUW_Picker_Button
   bIsVariable=true
   parent=CanvasPanel_0
   widget=Foot-R_Toes-IK
   properties {
    draw Style=Square
    label Font Size=8
    label Text=TOE
    label Color {colorUseRule=UseColor_Specified, specifiedColor=<rgba 0 0 0 1>}
    normal Tint {colorUseRule=UseColor_Specified, specifiedColor=<rgba 1 0.2462 0.2462 1>}
   }
   slot {
    class=CanvasPanelSlot
    properties {layoutData {alignment=<0.5 0.5>, anchors {maximum=<0.5 0.5>, minimum=<0.5 0.5>}, offsets=<ltrb -110 300 25 25>}}
   }
  }
  Prop_All {
   widgetClassPath=&EUW_Picker_Button
   bIsVariable=true
   parent=CanvasPanel_0
   widget=Prop_All
   properties {
    label Text=All Prop
    label Color {colorUseRule=UseColor_Specified, specifiedColor=<rgba 0 0 0 1>}
    normal Tint {colorUseRule=UseColor_Specified, specifiedColor=<rgba 0.1144 0.2623 0.1119 1>}
   }
   slot {
    class=CanvasPanelSlot
    properties {layoutData {alignment=<0 0>, anchors {maximum=<0.5 0.5>, minimum=<0.5 0.5>}, offsets=<ltrb -465 -380 70 25>}}
   }
  }
  Prop_Prop-Global {
   widgetClassPath=&EUW_Picker_Button
   bIsVariable=true
   parent=CanvasPanel_0
   widget=Prop_Prop-Global
   properties {
    draw Style=Square
    label Font Size=9
    label Text=Prop Global
    label Color {colorUseRule=UseColor_Specified, specifiedColor=<rgba 0 0 0 1>}
    normal Tint {colorUseRule=UseColor_Specified, specifiedColor=<rgba 0.9911 0.4852 0.0452 1>}
   }
   slot {
    class=CanvasPanelSlot
    properties {layoutData {alignment=<0 0>, anchors {maximum=<0.5 0.5>, minimum=<0.5 0.5>}, offsets=<ltrb -500 -350 70 25>}}
   }
  }
  Prop_Aim {
   widgetClassPath=&EUW_Picker_Button
   bIsVariable=true
   parent=CanvasPanel_0
   widget=Prop_Aim
   properties {
    draw Style=Square
    label Font Size=9
    label Text=Prop Aim
    label Color {colorUseRule=UseColor_Specified, specifiedColor=<rgba 0 0 0 1>}
    normal Tint {colorUseRule=UseColor_Specified, specifiedColor=<rgba 0.9911 0.4852 0.0452 1>}
   }
   slot {
    class=CanvasPanelSlot
    properties {layoutData {alignment=<0 0>, anchors {maximum=<0.5 0.5>, minimum=<0.5 0.5>}, offsets=<ltrb -425 -350 70 25>}}
   }
  }
  Prop_Prop-Local {
   widgetClassPath=&EUW_Picker_Button
   bIsVariable=true
   parent=CanvasPanel_0
   widget=Prop_Prop-Local
   properties {
    draw Style=Square
    label Font Size=9
    label Text=Prop Local
    label Color {colorUseRule=UseColor_Specified, specifiedColor=<rgba 0 0 0 1>}
    normal Tint {colorUseRule=UseColor_Specified, specifiedColor=<rgba 0.9217 0.4973 0.1282 1>}
   }
   slot {
    class=CanvasPanelSlot
    properties {layoutData {alignment=<0 0>, anchors {maximum=<0.5 0.5>, minimum=<0.5 0.5>}, offsets=<ltrb -500 -320 70 25>}}
   }
  }
  Prop_Prop {
   widgetClassPath=&EUW_Picker_Button
   bIsVariable=true
   parent=CanvasPanel_0
   widget=Prop_Prop
   properties {
    draw Style=Square
    label Font Size=9
    label Text=Prop
    label Color {colorUseRule=UseColor_Specified, specifiedColor=<rgba 0 0 0 1>}
    normal Tint {colorUseRule=UseColor_Specified, specifiedColor=<rgba 0.9565 0.5517 0.1996 1>}
   }
   slot {
    class=CanvasPanelSlot
    properties {layoutData {alignment=<0 0>, anchors {maximum=<0.5 0.5>, minimum=<0.5 0.5>}, offsets=<ltrb -425 -320 70 25>}}
   }
  }
  Prop_Prop-Attach-01 {
   widgetClassPath=&EUW_Picker_Button
   bIsVariable=true
   parent=CanvasPanel_0
   widget=Prop_Prop-Attach-01
   properties {
    draw Style=Square
    label Font Size=8
    label Text=Prop Attach
    label Color {colorUseRule=UseColor_Specified, specifiedColor=<rgba 0 0 0 1>}
    normal Tint {colorUseRule=UseColor_Specified, specifiedColor=<rgba 0.9911 0.65 0.3533 1>}
   }
   slot {
    class=CanvasPanelSlot
    properties {layoutData {alignment=<0 0>, anchors {maximum=<0.5 0.5>, minimum=<0.5 0.5>}, offsets=<ltrb -500 -290 70 25>}}
   }
  }
  Prop_Prop-Attach-02 {
   widgetClassPath=&EUW_Picker_Button
   bIsVariable=true
   parent=CanvasPanel_0
   widget=Prop_Prop-Attach-02
   properties {
    draw Style=Square
    label Font Size=8
    label Text=Prop Attach 2
    label Color {colorUseRule=UseColor_Specified, specifiedColor=<rgba 0 0 0 1>}
    normal Tint {colorUseRule=UseColor_Specified, specifiedColor=<rgba 0.9911 0.65 0.3533 1>}
   }
   slot {
    class=CanvasPanelSlot
    properties {layoutData {alignment=<0 0>, anchors {maximum=<0.5 0.5>, minimum=<0.5 0.5>}, offsets=<ltrb -425 -290 70 25>}}
   }
  }
  Attach_Attach-Cape {
   widgetClassPath=&EUW_Picker_Button
   bIsVariable=true
   parent=CanvasPanel_0
   widget=Attach_Attach-Cape
   properties {
    draw Style=Square
    label Text=Cape
    label Color {colorUseRule=UseColor_Specified, specifiedColor=<rgba 0 0 0 1>}
    normal Tint {colorUseRule=UseColor_Specified, specifiedColor=<rgba 0.9911 0.4852 0.0452 1>}
   }
   slot {
    class=CanvasPanelSlot
    properties {layoutData {alignment=<0.5 0.5>, anchors {maximum=<0.5 0.5>, minimum=<0.5 0.5>}, offsets=<ltrb 305 -390 70 20>}}
   }
  }
  Attach_Attach-Backpack {
   widgetClassPath=&EUW_Picker_Button
   bIsVariable=true
   parent=CanvasPanel_0
   widget=Attach_Attach-Backpack
   properties {
    draw Style=Square
    label Text=Backpack
    label Color {colorUseRule=UseColor_Specified, specifiedColor=<rgba 0 0 0 1>}
    normal Tint {colorUseRule=UseColor_Specified, specifiedColor=<rgba 0.9911 0.4852 0.0452 1>}
   }
   slot {
    class=CanvasPanelSlot
    properties {layoutData {alignment=<0.5 0.5>, anchors {maximum=<0.5 0.5>, minimum=<0.5 0.5>}, offsets=<ltrb 305 -365 70 20>}}
   }
  }
  FPS-Controls_Camera {
   widgetClassPath=&EUW_Picker_Button
   bIsVariable=true
   parent=CanvasPanel_0
   widget=FPS-Controls_Camera
   properties {
    draw Style=Square
    label Text=Camera
    label Color {colorUseRule=UseColor_Specified, specifiedColor=<rgba 0 0 0 1>}
    normal Tint {colorUseRule=UseColor_Specified, specifiedColor=<rgba 0.9911 0.4852 0.0452 1>}
   }
   slot {
    class=CanvasPanelSlot
    properties {layoutData {alignment=<0.5 0.5>, anchors {maximum=<0.5 0.5>, minimum=<0.5 0.5>}, offsets=<ltrb 305 -340 70 20>}}
   }
  }
  Attach_Attach {
   widgetClassPath=&EUW_Picker_Button
   bIsVariable=true
   parent=CanvasPanel_0
   widget=Attach_Attach
   properties {
    draw Style=Square
    label Text=Attach
    label Color {colorUseRule=UseColor_Specified, specifiedColor=<rgba 0 0 0 1>}
    normal Tint {colorUseRule=UseColor_Specified, specifiedColor=<rgba 0.9911 0.4852 0.0452 1>}
   }
   slot {
    class=CanvasPanelSlot
    properties {layoutData {alignment=<0.5 0.5>, anchors {maximum=<0.5 0.5>, minimum=<0.5 0.5>}, offsets=<ltrb 306.0391 -314.5405 70 20>}}
   }
  }
  Attach_Weapon-L {
   widgetClassPath=&EUW_Picker_Button
   bIsVariable=true
   parent=CanvasPanel_0
   widget=Attach_Weapon-L
   properties {
    draw Style=Square
    label Font Size=8
    label Text=WA
    label Color {colorUseRule=UseColor_Specified, specifiedColor=<rgba 0 0 0 1>}
    normal Tint {colorUseRule=UseColor_Specified, specifiedColor=<rgba 0.0272 0.067 0.7826 1>}
   }
   slot {
    class=CanvasPanelSlot
    properties {layoutData {alignment=<0 0>, anchors {maximum=<0.5 0.5>, minimum=<0.5 0.5>}, offsets=<ltrb 224 -277 22 22>}}
   }
  }
  Attach_Weapon-R {
   widgetClassPath=&EUW_Picker_Button
   bIsVariable=true
   parent=CanvasPanel_0
   widget=Attach_Weapon-R
   properties {
    draw Style=Square
    label Font Size=8
    label Text=WA
    label Color {colorUseRule=UseColor_Specified, specifiedColor=<rgba 0 0 0 1>}
    normal Tint {colorUseRule=UseColor_Specified, specifiedColor=<rgba 1 0.02 0.02 1>}
   }
   slot {
    class=CanvasPanelSlot
    properties {layoutData {alignment=<0 0>, anchors {maximum=<0.5 0.5>, minimum=<0.5 0.5>}, offsets=<ltrb -252.9609 -276.5405 22 22>}}
   }
  }
  Attach_Hand-Attach-R {
   widgetClassPath=&EUW_Picker_Button
   bIsVariable=true
   parent=CanvasPanel_0
   widget=Attach_Hand-Attach-R
   properties {
    draw Style=Square
    label Font Size=8
    label Text=HA
    label Color {colorUseRule=UseColor_Specified, specifiedColor=<rgba 0 0 0 1>}
    normal Tint {colorUseRule=UseColor_Specified, specifiedColor=<rgba 1 0.02 0.02 1>}
   }
   slot {
    class=CanvasPanelSlot
    properties {layoutData {alignment=<0 0>, anchors {maximum=<0.5 0.5>, minimum=<0.5 0.5>}, offsets=<ltrb -280.9609 -276.5405 22 22>}}
   }
  }
  Attach_Hand-Attach-L {
   widgetClassPath=&EUW_Picker_Button
   bIsVariable=true
   parent=CanvasPanel_0
   widget=Attach_Hand-Attach-L
   properties {
    draw Style=Square
    label Font Size=8
    label Text=HA
    label Color {colorUseRule=UseColor_Specified, specifiedColor=<rgba 0 0 0 1>}
    normal Tint {colorUseRule=UseColor_Specified, specifiedColor=<rgba 0.0272 0.067 0.7826 1>}
   }
   slot {
    class=CanvasPanelSlot
    properties {layoutData {alignment=<0 0>, anchors {maximum=<0.5 0.5>, minimum=<0.5 0.5>}, offsets=<ltrb 255 -277 22 22>}}
   }
  }
  AllTwist {
   widgetClassPath=&EUW_Picker_Button
   bIsVariable=true
   parent=CanvasPanel_0
   widget=AllTwist
   properties {
    label Font Size=8
    label Text=All Twists
    label Color {colorUseRule=UseColor_Specified, specifiedColor=<rgba 0 0 0 1>}
    normal Tint {colorUseRule=UseColor_Specified, specifiedColor=<rgba 0.1144 0.2623 0.1119 1>}
   }
   slot {
    class=CanvasPanelSlot
    properties {layoutData {alignment=<0.5 0.5>, anchors {maximum=<0.5 0.5>, minimum=<0.5 0.5>}, offsets=<ltrb 190 -365 70 20>}}
   }
  }
  AllWeaponControlsButton {
   widgetClassPath=&EUW_Picker_Button
   bIsVariable=true
   parent=CanvasPanel_0
   widget=AllWeaponControlsButton
   properties {
    label Font Size=8
    label Text=All Wep
    label Color {colorUseRule=UseColor_Specified, specifiedColor=<rgba 0 0 0 1>}
    normal Tint {colorUseRule=UseColor_Specified, specifiedColor=<rgba 0.1144 0.2623 0.1119 1>}
   }
   slot {
    class=CanvasPanelSlot
    properties {layoutData {alignment=<0.5 0.5>, anchors {maximum=<0.5 0.5>, minimum=<0.5 0.5>}, offsets=<ltrb 190 -390 70 20>}}
   }
  }
  AllSpine {
   widgetClassPath=&EUW_Picker_Button
   bIsVariable=true
   parent=CanvasPanel_0
   widget=AllSpine
   properties {
    label Font Size=8
    label Text=Spine
    label Color {colorUseRule=UseColor_Specified, specifiedColor=<rgba 0 0 0 1>}
    normal Tint {colorUseRule=UseColor_Specified, specifiedColor=<rgba 0.1144 0.2623 0.1119 1>}
   }
   slot {
    class=CanvasPanelSlot
    properties {layoutData {alignment=<0.5 0.5>, anchors {maximum=<0.5 0.5>, minimum=<0.5 0.5>}, offsets=<ltrb 75 -185 40 20>}}
   }
  }
  AllSecSpine {
   widgetClassPath=&EUW_Picker_Button
   bIsVariable=true
   parent=CanvasPanel_0
   widget=AllSecSpine
   properties {
    label Font Size=6
    label Text=Spine Sec
    label Color {colorUseRule=UseColor_Specified, specifiedColor=<rgba 0 0 0 1>}
    normal Tint {colorUseRule=UseColor_Specified, specifiedColor=<rgba 0.1144 0.2623 0.1119 1>}
   }
   slot {
    class=CanvasPanelSlot
    properties {layoutData {alignment=<0.5 0.5>, anchors {maximum=<0.5 0.5>, minimum=<0.5 0.5>}, offsets=<ltrb 75 -139 40 20>}}
   }
  }
  IKMove {
   widgetClassPath=&EUW_Picker_Button
   bIsVariable=true
   parent=CanvasPanel_0
   widget=IKMove
   properties {
    label Text=IK Move
    label Color {colorUseRule=UseColor_Specified, specifiedColor=<rgba 0 0 0 1>}
    normal Tint {colorUseRule=UseColor_Specified, specifiedColor=<rgba 0.1248 0.5647 0.1144 1>}
   }
   slot {
    class=CanvasPanelSlot
    properties {layoutData {alignment=<0.5 0.5>, anchors {maximum=<0.5 0.5>, minimum=<0.5 0.5>}, offsets=<ltrb 0 -86 70 20>}}
   }
  }
  4Fingers_l {
   widgetClassPath=&EUW_Picker_Button
   bIsVariable=true
   parent=CanvasPanel_0
   widget=4Fingers_l
   properties {
    label Font Size=8
    label Text=4
    label Color {colorUseRule=UseColor_Specified, specifiedColor=<rgba 0 0 0 1>}
    normal Tint {colorUseRule=UseColor_Specified, specifiedColor=<rgba 0.1144 0.2623 0.1119 1>}
   }
   slot {
    class=CanvasPanelSlot
    properties {layoutData {alignment=<0.5 0.5>, anchors {maximum=<0.5 0.5>, minimum=<0.5 0.5>}, offsets=<ltrb 260 -85 25 25>}}
   }
  }
  AllFingers_l {
   widgetClassPath=&EUW_Picker_Button
   bIsVariable=true
   parent=CanvasPanel_0
   widget=AllFingers_l
   properties {
    label Font Size=8
    label Text=Fingers
    label Color {colorUseRule=UseColor_Specified, specifiedColor=<rgba 0 0 0 1>}
    normal Tint {colorUseRule=UseColor_Specified, specifiedColor=<rgba 0.1144 0.2623 0.1119 1>}
   }
   slot {
    class=CanvasPanelSlot
    properties {layoutData {alignment=<0.5 0.5>, anchors {maximum=<0.5 0.5>, minimum=<0.5 0.5>}, offsets=<ltrb 215 -85 55 25>}}
   }
  }
  Fingers_L_End {
   widgetClassPath=&EUW_Picker_Button
   bIsVariable=true
   parent=CanvasPanel_0
   widget=Fingers_L_End
   properties {
    label Font Size=5
    label Color {colorUseRule=UseColor_Specified, specifiedColor=<rgba 0 0 0 1>}
    normal Tint {colorUseRule=UseColor_Specified, specifiedColor=<rgba 0.1144 0.2623 0.1119 1>}
   }
   slot {
    class=CanvasPanelSlot
    properties {layoutData {alignment=<0.5 0.5>, anchors {maximum=<0.5 0.5>, minimum=<0.5 0.5>}, offsets=<ltrb 284 35 15 15>}}
   }
  }
  Fingers_L_Mid {
   widgetClassPath=&EUW_Picker_Button
   bIsVariable=true
   parent=CanvasPanel_0
   widget=Fingers_L_Mid
   properties {
    label Font Size=5
    label Color {colorUseRule=UseColor_Specified, specifiedColor=<rgba 0 0 0 1>}
    normal Tint {colorUseRule=UseColor_Specified, specifiedColor=<rgba 0.1144 0.2623 0.1119 1>}
   }
   slot {
    class=CanvasPanelSlot
    properties {layoutData {alignment=<0.5 0.5>, anchors {maximum=<0.5 0.5>, minimum=<0.5 0.5>}, offsets=<ltrb 284 15 15 15>}}
   }
  }
  Index_L {
   widgetClassPath=&EUW_Picker_Button
   bIsVariable=true
   parent=CanvasPanel_0
   widget=Index_L
   properties {
    label Font Size=5
    label Color {colorUseRule=UseColor_Specified, specifiedColor=<rgba 0 0 0 1>}
    normal Tint {colorUseRule=UseColor_Specified, specifiedColor=<rgba 0.1144 0.2623 0.1119 1>}
   }
   slot {
    class=CanvasPanelSlot
    properties {layoutData {alignment=<0.5 0.5>, anchors {maximum=<0.5 0.5>, minimum=<0.5 0.5>}, offsets=<ltrb 196 -45 15 15>}}
   }
  }
  Meta_L {
   widgetClassPath=&EUW_Picker_Button
   bIsVariable=true
   parent=CanvasPanel_0
   widget=Meta_L
   properties {
    label Font Size=5
    label Color {colorUseRule=UseColor_Specified, specifiedColor=<rgba 0 0 0 1>}
    normal Tint {colorUseRule=UseColor_Specified, specifiedColor=<rgba 0.1144 0.2623 0.1119 1>}
   }
   slot {
    class=CanvasPanelSlot
    properties {layoutData {alignment=<0.5 0.5>, anchors {maximum=<0.5 0.5>, minimum=<0.5 0.5>}, offsets=<ltrb 284 -25 15 15>}}
   }
  }
  Fingers_L_Start {
   widgetClassPath=&EUW_Picker_Button
   bIsVariable=true
   parent=CanvasPanel_0
   widget=Fingers_L_Start
   properties {
    label Font Size=5
    label Color {colorUseRule=UseColor_Specified, specifiedColor=<rgba 0 0 0 1>}
    normal Tint {colorUseRule=UseColor_Specified, specifiedColor=<rgba 0.1144 0.2623 0.1119 1>}
   }
   slot {
    class=CanvasPanelSlot
    properties {layoutData {alignment=<0.5 0.5>, anchors {maximum=<0.5 0.5>, minimum=<0.5 0.5>}, offsets=<ltrb 284 -5 15 15>}}
   }
  }
  Pinky_L {
   widgetClassPath=&EUW_Picker_Button
   bIsVariable=true
   parent=CanvasPanel_0
   widget=Pinky_L
   properties {
    label Font Size=5
    label Color {colorUseRule=UseColor_Specified, specifiedColor=<rgba 0 0 0 1>}
    normal Tint {colorUseRule=UseColor_Specified, specifiedColor=<rgba 0.1144 0.2623 0.1119 1>}
   }
   slot {
    class=CanvasPanelSlot
    properties {layoutData {alignment=<0.5 0.5>, anchors {maximum=<0.5 0.5>, minimum=<0.5 0.5>}, offsets=<ltrb 262 -45 15 15>}}
   }
  }
  Ring_L {
   widgetClassPath=&EUW_Picker_Button
   bIsVariable=true
   parent=CanvasPanel_0
   widget=Ring_L
   properties {
    label Font Size=5
    label Color {colorUseRule=UseColor_Specified, specifiedColor=<rgba 0 0 0 1>}
    normal Tint {colorUseRule=UseColor_Specified, specifiedColor=<rgba 0.1144 0.2623 0.1119 1>}
   }
   slot {
    class=CanvasPanelSlot
    properties {layoutData {alignment=<0.5 0.5>, anchors {maximum=<0.5 0.5>, minimum=<0.5 0.5>}, offsets=<ltrb 240 -45 15 15>}}
   }
  }
  Middle_L {
   widgetClassPath=&EUW_Picker_Button
   bIsVariable=true
   parent=CanvasPanel_0
   widget=Middle_L
   properties {
    label Font Size=5
    label Color {colorUseRule=UseColor_Specified, specifiedColor=<rgba 0 0 0 1>}
    normal Tint {colorUseRule=UseColor_Specified, specifiedColor=<rgba 0.1144 0.2623 0.1119 1>}
   }
   slot {
    class=CanvasPanelSlot
    properties {layoutData {alignment=<0.5 0.5>, anchors {maximum=<0.5 0.5>, minimum=<0.5 0.5>}, offsets=<ltrb 218 -45 15 15>}}
   }
  }
  Thumb_L {
   widgetClassPath=&EUW_Picker_Button
   bIsVariable=true
   parent=CanvasPanel_0
   widget=Thumb_L
   properties {
    label Font Size=5
    label Color {colorUseRule=UseColor_Specified, specifiedColor=<rgba 0 0 0 1>}
    normal Tint {colorUseRule=UseColor_Specified, specifiedColor=<rgba 0.1144 0.2623 0.1119 1>}
   }
   slot {
    class=CanvasPanelSlot
    properties {layoutData {alignment=<0.5 0.5>, anchors {maximum=<0.5 0.5>, minimum=<0.5 0.5>}, offsets=<ltrb 174 -50 15 15>}}
   }
  }
  Fingers_R_End {
   widgetClassPath=&EUW_Picker_Button
   bIsVariable=true
   parent=CanvasPanel_0
   widget=Fingers_R_End
   properties {
    label Font Size=5
    label Color {colorUseRule=UseColor_Specified, specifiedColor=<rgba 0 0 0 1>}
    normal Tint {colorUseRule=UseColor_Specified, specifiedColor=<rgba 0.1144 0.2623 0.1119 1>}
   }
   slot {
    class=CanvasPanelSlot
    properties {layoutData {alignment=<0.5 0.5>, anchors {maximum=<0.5 0.5>, minimum=<0.5 0.5>}, offsets=<ltrb -278 35 15 15>}}
   }
  }
  Fingers_R_Mid {
   widgetClassPath=&EUW_Picker_Button
   bIsVariable=true
   parent=CanvasPanel_0
   widget=Fingers_R_Mid
   properties {
    label Font Size=5
    label Color {colorUseRule=UseColor_Specified, specifiedColor=<rgba 0 0 0 1>}
    normal Tint {colorUseRule=UseColor_Specified, specifiedColor=<rgba 0.1144 0.2623 0.1119 1>}
   }
   slot {
    class=CanvasPanelSlot
    properties {layoutData {alignment=<0.5 0.5>, anchors {maximum=<0.5 0.5>, minimum=<0.5 0.5>}, offsets=<ltrb -278 15 15 15>}}
   }
  }
  Fingers_R_Start {
   widgetClassPath=&EUW_Picker_Button
   bIsVariable=true
   parent=CanvasPanel_0
   widget=Fingers_R_Start
   properties {
    label Font Size=5
    label Color {colorUseRule=UseColor_Specified, specifiedColor=<rgba 0 0 0 1>}
    normal Tint {colorUseRule=UseColor_Specified, specifiedColor=<rgba 0.1144 0.2623 0.1119 1>}
   }
   slot {
    class=CanvasPanelSlot
    properties {layoutData {alignment=<0.5 0.5>, anchors {maximum=<0.5 0.5>, minimum=<0.5 0.5>}, offsets=<ltrb -278 -5 15 15>}}
   }
  }
  Meta_R {
   widgetClassPath=&EUW_Picker_Button
   bIsVariable=true
   parent=CanvasPanel_0
   widget=Meta_R
   properties {
    label Font Size=5
    label Color {colorUseRule=UseColor_Specified, specifiedColor=<rgba 0 0 0 1>}
    normal Tint {colorUseRule=UseColor_Specified, specifiedColor=<rgba 0.1144 0.2623 0.1119 1>}
   }
   slot {
    class=CanvasPanelSlot
    properties {layoutData {alignment=<0.5 0.5>, anchors {maximum=<0.5 0.5>, minimum=<0.5 0.5>}, offsets=<ltrb -278 -25 15 15>}}
   }
  }
  Pinky_R {
   widgetClassPath=&EUW_Picker_Button
   bIsVariable=true
   parent=CanvasPanel_0
   widget=Pinky_R
   properties {
    label Font Size=5
    label Color {colorUseRule=UseColor_Specified, specifiedColor=<rgba 0 0 0 1>}
    normal Tint {colorUseRule=UseColor_Specified, specifiedColor=<rgba 0.1144 0.2623 0.1119 1>}
   }
   slot {
    class=CanvasPanelSlot
    properties {layoutData {alignment=<0.5 0.5>, anchors {maximum=<0.5 0.5>, minimum=<0.5 0.5>}, offsets=<ltrb -256 -45 15 15>}}
   }
  }
  Ring_R {
   widgetClassPath=&EUW_Picker_Button
   bIsVariable=true
   parent=CanvasPanel_0
   widget=Ring_R
   properties {
    label Font Size=5
    label Color {colorUseRule=UseColor_Specified, specifiedColor=<rgba 0 0 0 1>}
    normal Tint {colorUseRule=UseColor_Specified, specifiedColor=<rgba 0.1144 0.2623 0.1119 1>}
   }
   slot {
    class=CanvasPanelSlot
    properties {layoutData {alignment=<0.5 0.5>, anchors {maximum=<0.5 0.5>, minimum=<0.5 0.5>}, offsets=<ltrb -234 -45 15 15>}}
   }
  }
  Middle_R {
   widgetClassPath=&EUW_Picker_Button
   bIsVariable=true
   parent=CanvasPanel_0
   widget=Middle_R
   properties {
    label Font Size=5
    label Color {colorUseRule=UseColor_Specified, specifiedColor=<rgba 0 0 0 1>}
    normal Tint {colorUseRule=UseColor_Specified, specifiedColor=<rgba 0.1144 0.2623 0.1119 1>}
   }
   slot {
    class=CanvasPanelSlot
    properties {layoutData {alignment=<0.5 0.5>, anchors {maximum=<0.5 0.5>, minimum=<0.5 0.5>}, offsets=<ltrb -212 -45 15 15>}}
   }
  }
  Index_R {
   widgetClassPath=&EUW_Picker_Button
   bIsVariable=true
   parent=CanvasPanel_0
   widget=Index_R
   properties {
    label Font Size=5
    label Color {colorUseRule=UseColor_Specified, specifiedColor=<rgba 0 0 0 1>}
    normal Tint {colorUseRule=UseColor_Specified, specifiedColor=<rgba 0.1144 0.2623 0.1119 1>}
   }
   slot {
    class=CanvasPanelSlot
    properties {layoutData {alignment=<0.5 0.5>, anchors {maximum=<0.5 0.5>, minimum=<0.5 0.5>}, offsets=<ltrb -190 -45 15 15>}}
   }
  }
  Thumb_R {
   widgetClassPath=&EUW_Picker_Button
   bIsVariable=true
   parent=CanvasPanel_0
   widget=Thumb_R
   properties {
    label Font Size=5
    label Color {colorUseRule=UseColor_Specified, specifiedColor=<rgba 0 0 0 1>}
    normal Tint {colorUseRule=UseColor_Specified, specifiedColor=<rgba 0.1144 0.2623 0.1119 1>}
   }
   slot {
    class=CanvasPanelSlot
    properties {layoutData {alignment=<0.5 0.5>, anchors {maximum=<0.5 0.5>, minimum=<0.5 0.5>}, offsets=<ltrb -168 -50 15 15>}}
   }
  }
  ArmsFingers_l {
   widgetClassPath=&EUW_Picker_Button
   bIsVariable=true
   parent=CanvasPanel_0
   widget=ArmsFingers_l
   properties {
    label Font Size=9
    label Text="Arm && Fingers"
    label Color {colorUseRule=UseColor_Specified, specifiedColor=<rgba 0 0 0 1>}
    normal Tint {colorUseRule=UseColor_Specified, specifiedColor=<rgba 0.1144 0.2623 0.1119 1>}
   }
   slot {
    class=CanvasPanelSlot
    properties {layoutData {alignment=<0.5 0.5>, anchors {maximum=<0.5 0.5>, minimum=<0.5 0.5>}, offsets=<ltrb 230 -120 60 30>}}
   }
  }
  ArmsFingers_r {
   widgetClassPath=&EUW_Picker_Button
   bIsVariable=true
   parent=CanvasPanel_0
   widget=ArmsFingers_r
   properties {
    label Font Size=9
    label Text="Arm && Fingers"
    label Color {colorUseRule=UseColor_Specified, specifiedColor=<rgba 0 0 0 1>}
    normal Tint {colorUseRule=UseColor_Specified, specifiedColor=<rgba 0.1144 0.2623 0.1119 1>}
   }
   slot {
    class=CanvasPanelSlot
    properties {layoutData {alignment=<0.5 0.5>, anchors {maximum=<0.5 0.5>, minimum=<0.5 0.5>}, offsets=<ltrb -230 -120 60 30>}}
   }
  }
  AllArm_r {
   widgetClassPath=&EUW_Picker_Button
   bIsVariable=true
   parent=CanvasPanel_0
   widget=AllArm_r
   properties {
    label Font Size=7
    label Text=Arm R
    label Color {colorUseRule=UseColor_Specified, specifiedColor=<rgba 0 0 0 1>}
    normal Tint {colorUseRule=UseColor_Specified, specifiedColor=<rgba 0.1144 0.2623 0.1119 1>}
   }
   slot {
    class=CanvasPanelSlot
    properties {layoutData {alignment=<0.5 0.5>, anchors {maximum=<0.5 0.5>, minimum=<0.5 0.5>}, offsets=<ltrb -310 -266 35 22>}}
   }
  }
  AllArm_l {
   widgetClassPath=&EUW_Picker_Button
   bIsVariable=true
   parent=CanvasPanel_0
   widget=AllArm_l
   properties {
    label Font Size=7
    label Text=Arm L
    label Color {colorUseRule=UseColor_Specified, specifiedColor=<rgba 0 0 0 1>}
    normal Tint {colorUseRule=UseColor_Specified, specifiedColor=<rgba 0.1144 0.2623 0.1119 1>}
   }
   slot {
    class=CanvasPanelSlot
    properties {layoutData {alignment=<0.5 0.5>, anchors {maximum=<0.5 0.5>, minimum=<0.5 0.5>}, offsets=<ltrb 310 -266 35 22>}}
   }
  }
  allFingers_r {
   widgetClassPath=&EUW_Picker_Button
   bIsVariable=true
   parent=CanvasPanel_0
   widget=allFingers_r
   properties {
    label Font Size=8
    label Text=Fingers
    label Color {colorUseRule=UseColor_Specified, specifiedColor=<rgba 0 0 0 1>}
    normal Tint {colorUseRule=UseColor_Specified, specifiedColor=<rgba 0.1144 0.2623 0.1119 1>}
   }
   slot {
    class=CanvasPanelSlot
    properties {layoutData {alignment=<0.5 0.5>, anchors {maximum=<0.5 0.5>, minimum=<0.5 0.5>}, offsets=<ltrb -245 -85 55 25>}}
   }
  }
  4Fingers_r {
   widgetClassPath=&EUW_Picker_Button
   bIsVariable=true
   parent=CanvasPanel_0
   widget=4Fingers_r
   properties {
    label Font Size=8
    label Text=4
    label Color {colorUseRule=UseColor_Specified, specifiedColor=<rgba 0 0 0 1>}
    normal Tint {colorUseRule=UseColor_Specified, specifiedColor=<rgba 0.1144 0.2623 0.1119 1>}
   }
   slot {
    class=CanvasPanelSlot
    properties {layoutData {alignment=<0.5 0.5>, anchors {maximum=<0.5 0.5>, minimum=<0.5 0.5>}, offsets=<ltrb -200 -85 25 25>}}
   }
  }
  AllArms {
   widgetClassPath=&EUW_Picker_Button
   bIsVariable=true
   parent=CanvasPanel_0
   widget=AllArms
   properties {
    label Text=All Arms
    label Color {colorUseRule=UseColor_Specified, specifiedColor=<rgba 0 0 0 1>}
    normal Tint {colorUseRule=UseColor_Specified, specifiedColor=<rgba 0.1144 0.2623 0.1119 1>}
   }
   slot {
    class=CanvasPanelSlot
    properties {layoutData {alignment=<0 0>, anchors {maximum=<0.5 0.5>, minimum=<0.5 0.5>}, offsets=<ltrb -465 -195 70 25>}}
   }
  }
  AllClavicles {
   widgetClassPath=&EUW_Picker_Button
   bIsVariable=true
   parent=CanvasPanel_0
   widget=AllClavicles
   properties {
    label Font Size=9
    label Text=All Clavicles
    label Color {colorUseRule=UseColor_Specified, specifiedColor=<rgba 0 0 0 1>}
    normal Tint {colorUseRule=UseColor_Specified, specifiedColor=<rgba 0.1144 0.2623 0.1119 1>}
   }
   slot {
    class=CanvasPanelSlot
    properties {layoutData {alignment=<0 0>, anchors {maximum=<0.5 0.5>, minimum=<0.5 0.5>}, offsets=<ltrb -465 -165 70 25>}}
   }
  }
  AllFingers {
   widgetClassPath=&EUW_Picker_Button
   bIsVariable=true
   parent=CanvasPanel_0
   widget=AllFingers
   properties {
    label Text=All Fingers
    label Color {colorUseRule=UseColor_Specified, specifiedColor=<rgba 0 0 0 1>}
    normal Tint {colorUseRule=UseColor_Specified, specifiedColor=<rgba 0.1144 0.2623 0.1119 1>}
   }
   slot {
    class=CanvasPanelSlot
    properties {layoutData {alignment=<0 0>, anchors {maximum=<0.5 0.5>, minimum=<0.5 0.5>}, offsets=<ltrb -465 -135 70 25>}}
   }
  }
  IKElbows {
   widgetClassPath=&EUW_Picker_Button
   bIsVariable=true
   parent=CanvasPanel_0
   widget=IKElbows
   properties {
    label Text=IK Elbows
    label Color {colorUseRule=UseColor_Specified, specifiedColor=<rgba 0 0 0 1>}
    normal Tint {colorUseRule=UseColor_Specified, specifiedColor=<rgba 0.2582 0.5841 0.2542 1>}
   }
   slot {
    class=CanvasPanelSlot
    properties {layoutData {alignment=<0 0>, anchors {maximum=<0.5 0.5>, minimum=<0.5 0.5>}, offsets=<ltrb -425 -45 70 25>}}
   }
  }
  IKArms {
   widgetClassPath=&EUW_Picker_Button
   bIsVariable=true
   parent=CanvasPanel_0
   widget=IKArms
   properties {
    label Text=IK Arms
    label Color {colorUseRule=UseColor_Specified, specifiedColor=<rgba 0 0 0 1>}
    normal Tint {colorUseRule=UseColor_Specified, specifiedColor=<rgba 0.2582 0.5841 0.2542 1>}
   }
   slot {
    class=CanvasPanelSlot
    properties {layoutData {alignment=<0 0>, anchors {maximum=<0.5 0.5>, minimum=<0.5 0.5>}, offsets=<ltrb -425 -105 70 25>}}
   }
  }
  FKArms {
   widgetClassPath=&EUW_Picker_Button
   bIsVariable=true
   parent=CanvasPanel_0
   widget=FKArms
   properties {
    label Text=FK Arms
    label Color {colorUseRule=UseColor_Specified, specifiedColor=<rgba 0 0 0 1>}
    normal Tint {colorUseRule=UseColor_Specified, specifiedColor=<rgba 0.1242 0.5677 0.1153 1>}
   }
   slot {
    class=CanvasPanelSlot
    properties {layoutData {alignment=<0 0>, anchors {maximum=<0.5 0.5>, minimum=<0.5 0.5>}, offsets=<ltrb -500 -105 70 25>}}
   }
  }
  FKBiceps {
   widgetClassPath=&EUW_Picker_Button
   bIsVariable=true
   parent=CanvasPanel_0
   widget=FKBiceps
   properties {
    label Text=FK Biceps
    label Color {colorUseRule=UseColor_Specified, specifiedColor=<rgba 0 0 0 1>}
    normal Tint {colorUseRule=UseColor_Specified, specifiedColor=<rgba 0.1242 0.5677 0.1153 1>}
   }
   slot {
    class=CanvasPanelSlot
    properties {layoutData {alignment=<0 0>, anchors {maximum=<0.5 0.5>, minimum=<0.5 0.5>}, offsets=<ltrb -500 -75 70 25>}}
   }
  }
  FKElbows {
   widgetClassPath=&EUW_Picker_Button
   bIsVariable=true
   parent=CanvasPanel_0
   widget=FKElbows
   properties {
    label Text=FK Elbows
    label Color {colorUseRule=UseColor_Specified, specifiedColor=<rgba 0 0 0 1>}
    normal Tint {colorUseRule=UseColor_Specified, specifiedColor=<rgba 0.1242 0.5677 0.1153 1>}
   }
   slot {
    class=CanvasPanelSlot
    properties {layoutData {alignment=<0 0>, anchors {maximum=<0.5 0.5>, minimum=<0.5 0.5>}, offsets=<ltrb -500 -45 70 25>}}
   }
  }
  FKHands {
   widgetClassPath=&EUW_Picker_Button
   bIsVariable=true
   parent=CanvasPanel_0
   widget=FKHands
   properties {
    label Text=FK Hands
    label Color {colorUseRule=UseColor_Specified, specifiedColor=<rgba 0 0 0 1>}
    normal Tint {colorUseRule=UseColor_Specified, specifiedColor=<rgba 0.1242 0.5677 0.1153 1>}
   }
   slot {
    class=CanvasPanelSlot
    properties {layoutData {alignment=<0 0>, anchors {maximum=<0.5 0.5>, minimum=<0.5 0.5>}, offsets=<ltrb -500 -15 70 25>}}
   }
  }
  IKHands {
   widgetClassPath=&EUW_Picker_Button
   bIsVariable=true
   parent=CanvasPanel_0
   widget=IKHands
   properties {
    label Text=IK Hands
    label Color {colorUseRule=UseColor_Specified, specifiedColor=<rgba 0 0 0 1>}
    normal Tint {colorUseRule=UseColor_Specified, specifiedColor=<rgba 0.2582 0.5841 0.2542 1>}
   }
   slot {
    class=CanvasPanelSlot
    properties {layoutData {alignment=<0 0>, anchors {maximum=<0.5 0.5>, minimum=<0.5 0.5>}, offsets=<ltrb -425 -15 70 25>}}
   }
  }
  AllFKRArm {
   widgetClassPath=&EUW_Picker_Button
   bIsVariable=true
   parent=CanvasPanel_0
   widget=AllFKRArm
   properties {
    label Font Size=8
    label Text=FK
    label Color {colorUseRule=UseColor_Specified, specifiedColor=<rgba 0 0 0 1>}
    normal Tint {colorUseRule=UseColor_Specified, specifiedColor=<rgba 0.1242 0.5677 0.1153 1>}
   }
   slot {
    class=CanvasPanelSlot
    properties {layoutData {alignment=<0.5 0.5>, anchors {maximum=<0.5 0.5>, minimum=<0.5 0.5>}, offsets=<ltrb -310 -235 25 25>}}
   }
  }
  AllIKRArm {
   widgetClassPath=&EUW_Picker_Button
   bIsVariable=true
   parent=CanvasPanel_0
   widget=AllIKRArm
   properties {
    label Font Size=8
    label Text=IK
    label Color {colorUseRule=UseColor_Specified, specifiedColor=<rgba 0 0 0 1>}
    normal Tint {colorUseRule=UseColor_Specified, specifiedColor=<rgba 0.2598 0.5833 0.2522 1>}
   }
   slot {
    class=CanvasPanelSlot
    properties {layoutData {alignment=<0.5 0.5>, anchors {maximum=<0.5 0.5>, minimum=<0.5 0.5>}, offsets=<ltrb -310 -205 25 25>}}
   }
  }
  AllFKLArm {
   widgetClassPath=&EUW_Picker_Button
   bIsVariable=true
   parent=CanvasPanel_0
   widget=AllFKLArm
   properties {
    label Font Size=8
    label Text=FK
    label Color {colorUseRule=UseColor_Specified, specifiedColor=<rgba 0 0 0 1>}
    normal Tint {colorUseRule=UseColor_Specified, specifiedColor=<rgba 0.1242 0.5677 0.1153 1>}
   }
   slot {
    class=CanvasPanelSlot
    properties {layoutData {alignment=<0.5 0.5>, anchors {maximum=<0.5 0.5>, minimum=<0.5 0.5>}, offsets=<ltrb 310 -235 25 25>}}
   }
  }
  AllIKLArm {
   widgetClassPath=&EUW_Picker_Button
   bIsVariable=true
   parent=CanvasPanel_0
   widget=AllIKLArm
   properties {
    label Font Size=8
    label Text=IK
    label Color {colorUseRule=UseColor_Specified, specifiedColor=<rgba 0 0 0 1>}
    normal Tint {colorUseRule=UseColor_Specified, specifiedColor=<rgba 0.2598 0.5833 0.2522 1>}
   }
   slot {
    class=CanvasPanelSlot
    properties {layoutData {alignment=<0.5 0.5>, anchors {maximum=<0.5 0.5>, minimum=<0.5 0.5>}, offsets=<ltrb 310 -205 25 25>}}
   }
  }
  AllLegs {
   widgetClassPath=&EUW_Picker_Button
   bIsVariable=true
   parent=CanvasPanel_0
   widget=AllLegs
   properties {
    label Text=All Legs
    label Color {colorUseRule=UseColor_Specified, specifiedColor=<rgba 0 0 0 1>}
    normal Tint {colorUseRule=UseColor_Specified, specifiedColor=<rgba 0.1144 0.2623 0.1119 1>}
   }
   slot {
    class=CanvasPanelSlot
    properties {layoutData {alignment=<0 0>, anchors {maximum=<0.5 0.5>, minimum=<0.5 0.5>}, offsets=<ltrb -465 110 70 25>}}
   }
  }
  AllFKLegs {
   widgetClassPath=&EUW_Picker_Button
   bIsVariable=true
   parent=CanvasPanel_0
   widget=AllFKLegs
   properties {
    label Text=All FK Legs
    label Color {colorUseRule=UseColor_Specified, specifiedColor=<rgba 0 0 0 1>}
    normal Tint {colorUseRule=UseColor_Specified, specifiedColor=<rgba 0.1242 0.5677 0.1153 1>}
   }
   slot {
    class=CanvasPanelSlot
    properties {layoutData {alignment=<0 0>, anchors {maximum=<0.5 0.5>, minimum=<0.5 0.5>}, offsets=<ltrb -500 140 70 25>}}
   }
  }
  FKThighs {
   widgetClassPath=&EUW_Picker_Button
   bIsVariable=true
   parent=CanvasPanel_0
   widget=FKThighs
   properties {
    label Text=FK Thighs
    label Color {colorUseRule=UseColor_Specified, specifiedColor=<rgba 0 0 0 1>}
    normal Tint {colorUseRule=UseColor_Specified, specifiedColor=<rgba 0.1242 0.5677 0.1153 1>}
   }
   slot {
    class=CanvasPanelSlot
    properties {layoutData {alignment=<0 0>, anchors {maximum=<0.5 0.5>, minimum=<0.5 0.5>}, offsets=<ltrb -500 170 70 25>}}
   }
  }
  AllIKLegs {
   widgetClassPath=&EUW_Picker_Button
   bIsVariable=true
   parent=CanvasPanel_0
   widget=AllIKLegs
   properties {
    label Text=All IK Legs
    label Color {colorUseRule=UseColor_Specified, specifiedColor=<rgba 0 0 0 1>}
    normal Tint {colorUseRule=UseColor_Specified, specifiedColor=<rgba 0.2598 0.5833 0.2522 1>}
   }
   slot {
    class=CanvasPanelSlot
    properties {layoutData {alignment=<0 0>, anchors {maximum=<0.5 0.5>, minimum=<0.5 0.5>}, offsets=<ltrb -425 140 70 25>}}
   }
  }
  FKKnees {
   widgetClassPath=&EUW_Picker_Button
   bIsVariable=true
   parent=CanvasPanel_0
   widget=FKKnees
   properties {
    label Text=FK Knees
    label Color {colorUseRule=UseColor_Specified, specifiedColor=<rgba 0 0 0 1>}
    normal Tint {colorUseRule=UseColor_Specified, specifiedColor=<rgba 0.1242 0.5677 0.1153 1>}
   }
   slot {
    class=CanvasPanelSlot
    properties {layoutData {alignment=<0 0>, anchors {maximum=<0.5 0.5>, minimum=<0.5 0.5>}, offsets=<ltrb -500 200 70 25>}}
   }
  }
  FKFeets {
   widgetClassPath=&EUW_Picker_Button
   bIsVariable=true
   parent=CanvasPanel_0
   widget=FKFeets
   properties {
    label Text=FK Feets
    label Color {colorUseRule=UseColor_Specified, specifiedColor=<rgba 0 0 0 1>}
    normal Tint {colorUseRule=UseColor_Specified, specifiedColor=<rgba 0.1242 0.5677 0.1153 1>}
   }
   slot {
    class=CanvasPanelSlot
    properties {layoutData {alignment=<0 0>, anchors {maximum=<0.5 0.5>, minimum=<0.5 0.5>}, offsets=<ltrb -500 230 70 25>}}
   }
  }
  FKToes {
   widgetClassPath=&EUW_Picker_Button
   bIsVariable=true
   parent=CanvasPanel_0
   widget=FKToes
   properties {
    label Text=FK Toes
    label Color {colorUseRule=UseColor_Specified, specifiedColor=<rgba 0 0 0 1>}
    normal Tint {colorUseRule=UseColor_Specified, specifiedColor=<rgba 0.1242 0.5677 0.1153 1>}
   }
   slot {
    class=CanvasPanelSlot
    properties {layoutData {alignment=<0 0>, anchors {maximum=<0.5 0.5>, minimum=<0.5 0.5>}, offsets=<ltrb -500 260 70 25>}}
   }
  }
  IKToes {
   widgetClassPath=&EUW_Picker_Button
   bIsVariable=true
   parent=CanvasPanel_0
   widget=IKToes
   properties {
    label Text=IK Toes
    label Color {colorUseRule=UseColor_Specified, specifiedColor=<rgba 0 0 0 1>}
    normal Tint {colorUseRule=UseColor_Specified, specifiedColor=<rgba 0.2598 0.5833 0.2522 1>}
   }
   slot {
    class=CanvasPanelSlot
    properties {layoutData {alignment=<0 0>, anchors {maximum=<0.5 0.5>, minimum=<0.5 0.5>}, offsets=<ltrb -425 260 70 25>}}
   }
  }
  IKFeet {
   widgetClassPath=&EUW_Picker_Button
   bIsVariable=true
   parent=CanvasPanel_0
   widget=IKFeet
   properties {
    label Text=IK Feets
    label Color {colorUseRule=UseColor_Specified, specifiedColor=<rgba 0 0 0 1>}
    normal Tint {colorUseRule=UseColor_Specified, specifiedColor=<rgba 0.2598 0.5833 0.2522 1>}
   }
   slot {
    class=CanvasPanelSlot
    properties {layoutData {alignment=<0 0>, anchors {maximum=<0.5 0.5>, minimum=<0.5 0.5>}, offsets=<ltrb -425 230 70 25>}}
   }
  }
  IKKnees {
   widgetClassPath=&EUW_Picker_Button
   bIsVariable=true
   parent=CanvasPanel_0
   widget=IKKnees
   properties {
    label Text=IK Knees
    label Color {colorUseRule=UseColor_Specified, specifiedColor=<rgba 0 0 0 1>}
    normal Tint {colorUseRule=UseColor_Specified, specifiedColor=<rgba 0.2598 0.5833 0.2522 1>}
   }
   slot {
    class=CanvasPanelSlot
    properties {layoutData {alignment=<0 0>, anchors {maximum=<0.5 0.5>, minimum=<0.5 0.5>}, offsets=<ltrb -425 200 70 25>}}
   }
  }
  IKBases {
   widgetClassPath=&EUW_Picker_Button
   bIsVariable=true
   parent=CanvasPanel_0
   widget=IKBases
   properties {
    label Text=IK Bases
    label Color {colorUseRule=UseColor_Specified, specifiedColor=<rgba 0 0 0 1>}
    normal Tint {colorUseRule=UseColor_Specified, specifiedColor=<rgba 0.2598 0.5833 0.2522 1>}
   }
   slot {
    class=CanvasPanelSlot
    properties {layoutData {alignment=<0 0>, anchors {maximum=<0.5 0.5>, minimum=<0.5 0.5>}, offsets=<ltrb -425 170 70 25>}}
   }
  }
  IKFeetExtras {
   widgetClassPath=&EUW_Picker_Button
   bIsVariable=true
   parent=CanvasPanel_0
   widget=IKFeetExtras
   properties {
    label Font Size=8
    label Text=IK Feet Extras
    label Color {colorUseRule=UseColor_Specified, specifiedColor=<rgba 0 0 0 1>}
    normal Tint {colorUseRule=UseColor_Specified, specifiedColor=<rgba 0.2598 0.5833 0.2522 1>}
   }
   slot {
    class=CanvasPanelSlot
    properties {layoutData {alignment=<0 0>, anchors {maximum=<0.5 0.5>, minimum=<0.5 0.5>}, offsets=<ltrb -425 290 70 25>}}
   }
  }
  AllLeg_r {
   widgetClassPath=&EUW_Picker_Button
   bIsVariable=true
   parent=CanvasPanel_0
   widget=AllLeg_r
   properties {
    label Font Size=8
    label Text=Leg R
    label Color {colorUseRule=UseColor_Specified, specifiedColor=<rgba 0 0 0 1>}
    normal Tint {colorUseRule=UseColor_Specified, specifiedColor=<rgba 0.1144 0.2623 0.1119 1>}
   }
   slot {
    class=CanvasPanelSlot
    properties {layoutData {alignment=<0.5 0.5>, anchors {maximum=<0.5 0.5>, minimum=<0.5 0.5>}, offsets=<ltrb -215 271 55 25>}}
   }
  }
  AllIKLegExtras_r {
   widgetClassPath=&EUW_Picker_Button
   bIsVariable=true
   parent=CanvasPanel_0
   widget=AllIKLegExtras_r
   properties {
    label Font Size=8
    label Text=IK Extras
    label Color {colorUseRule=UseColor_Specified, specifiedColor=<rgba 0 0 0 1>}
    normal Tint {colorUseRule=UseColor_Specified, specifiedColor=<rgba 0.2598 0.5833 0.2522 1>}
   }
   slot {
    class=CanvasPanelSlot
    properties {layoutData {alignment=<0.5 0.5>, anchors {maximum=<0.5 0.5>, minimum=<0.5 0.5>}, offsets=<ltrb -215 300 55 25>}}
   }
  }
  AllIKLeg_r {
   widgetClassPath=&EUW_Picker_Button
   bIsVariable=true
   parent=CanvasPanel_0
   widget=AllIKLeg_r
   properties {
    label Font Size=8
    label Text=IK
    label Color {colorUseRule=UseColor_Specified, specifiedColor=<rgba 0 0 0 1>}
    normal Tint {colorUseRule=UseColor_Specified, specifiedColor=<rgba 0.2598 0.5833 0.2522 1>}
   }
   slot {
    class=CanvasPanelSlot
    properties {layoutData {alignment=<0.5 0.5>, anchors {maximum=<0.5 0.5>, minimum=<0.5 0.5>}, offsets=<ltrb -170 300 25 25>}}
   }
  }
  AllFKLeg_r {
   widgetClassPath=&EUW_Picker_Button
   bIsVariable=true
   parent=CanvasPanel_0
   widget=AllFKLeg_r
   properties {
    label Font Size=8
    label Text=FK
    label Color {colorUseRule=UseColor_Specified, specifiedColor=<rgba 0 0 0 1>}
    normal Tint {colorUseRule=UseColor_Specified, specifiedColor=<rgba 0.1242 0.5677 0.1153 1>}
   }
   slot {
    class=CanvasPanelSlot
    properties {layoutData {alignment=<0.5 0.5>, anchors {maximum=<0.5 0.5>, minimum=<0.5 0.5>}, offsets=<ltrb -170 271 25 25>}}
   }
  }
  AllFKLeg_l {
   widgetClassPath=&EUW_Picker_Button
   bIsVariable=true
   parent=CanvasPanel_0
   widget=AllFKLeg_l
   properties {
    label Font Size=8
    label Text=FK
    label Color {colorUseRule=UseColor_Specified, specifiedColor=<rgba 0 0 0 1>}
    normal Tint {colorUseRule=UseColor_Specified, specifiedColor=<rgba 0.1242 0.5677 0.1153 1>}
   }
   slot {
    class=CanvasPanelSlot
    properties {layoutData {alignment=<0.5 0.5>, anchors {maximum=<0.5 0.5>, minimum=<0.5 0.5>}, offsets=<ltrb 170 271 25 25>}}
   }
  }
  AllLeg_l {
   widgetClassPath=&EUW_Picker_Button
   bIsVariable=true
   parent=CanvasPanel_0
   widget=AllLeg_l
   properties {
    label Font Size=8
    label Text=Leg L
    label Color {colorUseRule=UseColor_Specified, specifiedColor=<rgba 0 0 0 1>}
    normal Tint {colorUseRule=UseColor_Specified, specifiedColor=<rgba 0.1144 0.2623 0.1119 1>}
   }
   slot {
    class=CanvasPanelSlot
    properties {layoutData {alignment=<0.5 0.5>, anchors {maximum=<0.5 0.5>, minimum=<0.5 0.5>}, offsets=<ltrb 215 271 55 25>}}
   }
  }
  AllIKLegExtra_l {
   widgetClassPath=&EUW_Picker_Button
   bIsVariable=true
   parent=CanvasPanel_0
   widget=AllIKLegExtra_l
   properties {
    label Font Size=8
    label Text=IK Extras
    label Color {colorUseRule=UseColor_Specified, specifiedColor=<rgba 0 0 0 1>}
    normal Tint {colorUseRule=UseColor_Specified, specifiedColor=<rgba 0.2598 0.5833 0.2522 1>}
   }
   slot {
    class=CanvasPanelSlot
    properties {layoutData {alignment=<0.5 0.5>, anchors {maximum=<0.5 0.5>, minimum=<0.5 0.5>}, offsets=<ltrb 215 300 55 25>}}
   }
  }
  AllIKLeg_l {
   widgetClassPath=&EUW_Picker_Button
   bIsVariable=true
   parent=CanvasPanel_0
   widget=AllIKLeg_l
   properties {
    label Font Size=8
    label Text=IK
    label Color {colorUseRule=UseColor_Specified, specifiedColor=<rgba 0 0 0 1>}
    normal Tint {colorUseRule=UseColor_Specified, specifiedColor=<rgba 0.2598 0.5833 0.2522 1>}
   }
   slot {
    class=CanvasPanelSlot
    properties {layoutData {alignment=<0.5 0.5>, anchors {maximum=<0.5 0.5>, minimum=<0.5 0.5>}, offsets=<ltrb 170 300 25 25>}}
   }
  }
  AllFKHead {
   widgetClassPath=&EUW_Picker_Button
   bIsVariable=true
   parent=CanvasPanel_0
   widget=AllFKHead
   properties {
    label Font Size=6
    label Text=FK
    label Color {colorUseRule=UseColor_Specified, specifiedColor=<rgba 0 0 0 1>}
    normal Tint {colorUseRule=UseColor_Specified, specifiedColor=<rgba 0.1248 0.5647 0.1144 1>}
   }
   slot {
    class=CanvasPanelSlot
    properties {layoutData {alignment=<0.5 0.5>, anchors {maximum=<0.5 0.5>, minimum=<0.5 0.5>}, offsets=<ltrb 0.0391 -367.5405 18 18>}}
   }
  }
  AllIKHead {
   widgetClassPath=&EUW_Picker_Button
   bIsVariable=true
   parent=CanvasPanel_0
   widget=AllIKHead
   properties {
    label Font Size=6
    label Text=IK
    label Color {colorUseRule=UseColor_Specified, specifiedColor=<rgba 0 0 0 1>}
    normal Tint {colorUseRule=UseColor_Specified, specifiedColor=<rgba 0.2582 0.5841 0.2542 1>}
   }
   slot {
    class=CanvasPanelSlot
    properties {layoutData {alignment=<0.5 0.5>, anchors {maximum=<0.5 0.5>, minimum=<0.5 0.5>}, offsets=<ltrb -39.9609 -343.5405 18 18>}}
   }
  }
  AllIKSpine {
   widgetClassPath=&EUW_Picker_Button
   bIsVariable=true
   parent=CanvasPanel_0
   widget=AllIKSpine
   properties {
    label Font Size=8
    label Text=IK
    label Color {colorUseRule=UseColor_Specified, specifiedColor=<rgba 0 0 0 1>}
    normal Tint {colorUseRule=UseColor_Specified, specifiedColor=<rgba 0.2582 0.5841 0.2542 1>}
   }
   slot {
    class=CanvasPanelSlot
    properties {layoutData {alignment=<0.5 0.5>, anchors {maximum=<0.5 0.5>, minimum=<0.5 0.5>}, offsets=<ltrb 60 -162 25 20>}}
   }
  }
  AllFKSpine {
   widgetClassPath=&EUW_Picker_Button
   bIsVariable=true
   parent=CanvasPanel_0
   widget=AllFKSpine
   properties {
    label Font Size=8
    label Text=FK
    label Color {colorUseRule=UseColor_Specified, specifiedColor=<rgba 0 0 0 1>}
    normal Tint {colorUseRule=UseColor_Specified, specifiedColor=<rgba 0.1248 0.5647 0.1144 1>}
   }
   slot {
    class=CanvasPanelSlot
    properties {layoutData {alignment=<0.5 0.5>, anchors {maximum=<0.5 0.5>, minimum=<0.5 0.5>}, offsets=<ltrb 91.5391 -162 25 20>}}
   }
  }
  root_Global {
   widgetClassPath=&EUW_Picker_Button
   bIsVariable=true
   parent=CanvasPanel_0
   widget=root_Global
   properties {
    draw Style=Square
    label Font Size=15
    label Text=G
    label Color {colorUseRule=UseColor_Specified, specifiedColor=<rgba 0 0 0 1>}
    normal Tint {colorUseRule=UseColor_Specified, specifiedColor=<rgba 0.9911 0.4852 0.0452 1>}
   }
   slot {
    class=CanvasPanelSlot
    properties {layoutData {alignment=<0.5 0.5>, anchors {maximum=<0.5 0.5>, minimum=<0.5 0.5>}, offsets=<ltrb -320 -385.5405 30 30>}}
   }
  }
  root_Local {
   widgetClassPath=&EUW_Picker_Button
   bIsVariable=true
   parent=CanvasPanel_0
   widget=root_Local
   properties {
    draw Style=Square
    label Font Size=15
    label Text=L
    label Color {colorUseRule=UseColor_Specified, specifiedColor=<rgba 0 0 0 1>}
    normal Tint {colorUseRule=UseColor_Specified, specifiedColor=<rgba 0.9911 0.4852 0.0452 1>}
   }
   slot {
    class=CanvasPanelSlot
    properties {layoutData {alignment=<0.5 0.5>, anchors {maximum=<0.5 0.5>, minimum=<0.5 0.5>}, offsets=<ltrb -285 -385.5405 30 30>}}
   }
  }
  root_Root {
   widgetClassPath=&EUW_Picker_Button
   bIsVariable=true
   parent=CanvasPanel_0
   widget=root_Root
   properties {
    draw Style=Square
    label Font Size=15
    label Text=R
    label Color {colorUseRule=UseColor_Specified, specifiedColor=<rgba 0 0 0 1>}
    normal Tint {colorUseRule=UseColor_Specified, specifiedColor=<rgba 0.9911 0.4852 0.0452 1>}
   }
   slot {
    class=CanvasPanelSlot
    properties {layoutData {alignment=<0.5 0.5>, anchors {maximum=<0.5 0.5>, minimum=<0.5 0.5>}, offsets=<ltrb -250 -385.5405 30 30>}}
   }
  }
  ZeroOutbutton {
   widgetClassPath=&EUW_Picker_Button
   bIsVariable=true
   parent=CanvasPanel_0
   widget=ZeroOutbutton
   properties {
    draw Style=Square
    label Text=Zero All
    toolTipText=Zero Out All Controls
    normal Tint {colorUseRule=UseColor_Specified, specifiedColor=<rgba 0.2423 0.0595 0.2122 1>}
   }
   slot {
    class=CanvasPanelSlot
    properties {layoutData {alignment=<0 0>, anchors {maximum=<0.5 0.5>, minimum=<0.5 0.5>}, offsets=<ltrb 355 270 100 25>}}
   }
  }
  ZeroOutselected {
   widgetClassPath=&EUW_Picker_Button
   bIsVariable=true
   parent=CanvasPanel_0
   widget=ZeroOutselected
   properties {
    draw Style=Square
    label Text=Zero Selected
    toolTipText=Zero Out All Controls
    normal Tint {colorUseRule=UseColor_Specified, specifiedColor=<rgba 0.2423 0.0595 0.2122 1>}
   }
   slot {
    class=CanvasPanelSlot
    properties {layoutData {alignment=<0 0>, anchors {maximum=<0.5 0.5>, minimum=<0.5 0.5>}, offsets=<ltrb 355 300 100 25>}}
   }
  }
  ToggleCapsule_Button {
   widgetClassPath=&EUW_Picker_Button
   bIsVariable=true
   parent=CanvasPanel_0
   widget=ToggleCapsule_Button
   properties {label Text=Capsule Visibility, normal Tint {colorUseRule=UseColor_Specified, specifiedColor=<rgba 0.2423 0.0595 0.2122 1>}}
   slot {
    class=CanvasPanelSlot
    properties {layoutData {alignment=<0 0>, anchors {maximum=<0.5 0.5>, minimum=<0.5 0.5>}, offsets=<ltrb 355 -340 150 30>}}
   }
  }
  BlockingMode {
   widgetClassPath=&EUW_Picker_Button
   bIsVariable=true
   parent=CanvasPanel_0
   widget=BlockingMode
   properties {label Text=Blocking Mode, normal Tint {colorUseRule=UseColor_Specified, specifiedColor=<rgba 0.2423 0.0595 0.2122 1>}}
   slot {
    class=CanvasPanelSlot
    properties {layoutData {alignment=<0 0>, anchors {maximum=<0.5 0.5>, minimum=<0.5 0.5>}, offsets=<ltrb 355 -400 150 30>}}
   }
  }
  ShowExtraButtons {
   widgetClassPath=&EUW_Picker_Button
   bIsVariable=true
   parent=CanvasPanel_0
   widget=ShowExtraButtons
   properties {label Text=Sets Buttons Visibility, normal Tint {colorUseRule=UseColor_Specified, specifiedColor=<rgba 0.2435 0.0593 0.2134 1>}}
   slot {
    class=CanvasPanelSlot
    properties {layoutData {alignment=<0 0>, anchors {maximum=<0.5 0.5>, minimum=<0.5 0.5>}, offsets=<ltrb 355 -305 150 30>}}
   }
  }
  TwistsVisibility {
   widgetClassPath=&EUW_Picker_Button
   bIsVariable=true
   parent=CanvasPanel_0
   widget=TwistsVisibility
   properties {label Text=Twists Visibility, normal Tint {colorUseRule=UseColor_Specified, specifiedColor=<rgba 0.2435 0.0593 0.2134 1>}}
   slot {
    class=CanvasPanelSlot
    properties {layoutData {alignment=<0 0>, anchors {maximum=<0.5 0.5>, minimum=<0.5 0.5>}, offsets=<ltrb 355 -235 150 30>}}
   }
  }
  ArmIKVisibility {
   widgetClassPath=&EUW_Picker_Button
   bIsVariable=true
   parent=CanvasPanel_0
   widget=ArmIKVisibility
   properties {label Text=Arm IK Visibility, normal Tint {colorUseRule=UseColor_Specified, specifiedColor=<rgba 0.2435 0.0593 0.2134 1>}}
   slot {
    class=CanvasPanelSlot
    properties {layoutData {alignment=<0 0>, anchors {maximum=<0.5 0.5>, minimum=<0.5 0.5>}, offsets=<ltrb 355 -165 150 30>}}
   }
  }
  SpineSecVisibility {
   widgetClassPath=&EUW_Picker_Button
   bIsVariable=true
   parent=CanvasPanel_0
   widget=SpineSecVisibility
   properties {label Text=Spine Sec Visibility, normal Tint {colorUseRule=UseColor_Specified, specifiedColor=<rgba 0.2435 0.0593 0.2134 1>}}
   slot {
    class=CanvasPanelSlot
    properties {layoutData {alignment=<0 0>, anchors {maximum=<0.5 0.5>, minimum=<0.5 0.5>}, offsets=<ltrb 355.0391 -130 150 30>}}
   }
  }
  ArmFKVisibility {
   widgetClassPath=&EUW_Picker_Button
   bIsVariable=true
   parent=CanvasPanel_0
   widget=ArmFKVisibility
   properties {label Text=Arm FK Visibility, normal Tint {colorUseRule=UseColor_Specified, specifiedColor=<rgba 0.2435 0.0593 0.2134 1>}}
   slot {
    class=CanvasPanelSlot
    properties {layoutData {alignment=<0 0>, anchors {maximum=<0.5 0.5>, minimum=<0.5 0.5>}, offsets=<ltrb 355 -200 150 30>}}
   }
  }
  LegIKVisibility {
   widgetClassPath=&EUW_Picker_Button
   bIsVariable=true
   parent=CanvasPanel_0
   widget=LegIKVisibility
   properties {label Text=Leg IK Visibility, normal Tint {colorUseRule=UseColor_Specified, specifiedColor=<rgba 0.2435 0.0593 0.2134 1>}}
   slot {
    class=CanvasPanelSlot
    properties {layoutData {alignment=<0 0>, anchors {maximum=<0.5 0.5>, minimum=<0.5 0.5>}, offsets=<ltrb 355 -60 150 30>}}
   }
  }
  LegFKVisibility {
   widgetClassPath=&EUW_Picker_Button
   bIsVariable=true
   parent=CanvasPanel_0
   widget=LegFKVisibility
   properties {label Text=Leg FK Visibility, normal Tint {colorUseRule=UseColor_Specified, specifiedColor=<rgba 0.2435 0.0593 0.2134 1>}}
   slot {
    class=CanvasPanelSlot
    properties {layoutData {alignment=<0 0>, anchors {maximum=<0.5 0.5>, minimum=<0.5 0.5>}, offsets=<ltrb 355 -95 150 30>}}
   }
  }
  FacePickerButton {
   widgetClassPath=&EUW_Picker_Button
   bIsVariable=true
   parent=CanvasPanel_0
   widget=FacePickerButton
   properties {label Text=Face Picker Visibility, normal Tint {colorUseRule=UseColor_Specified, specifiedColor=<rgba 0.2435 0.0593 0.2134 1>}}
   slot {
    class=CanvasPanelSlot
    properties {layoutData {alignment=<0 0>, anchors {maximum=<0.5 0.5>, minimum=<0.5 0.5>}, offsets=<ltrb 355.0391 59.4595 150 30>}}
   }
  }
  FaceBoardButton {
   widgetClassPath=&EUW_Picker_Button
   bIsVariable=true
   parent=CanvasPanel_0
   widget=FaceBoardButton
   properties {label Text=Face Board Camera, normal Tint {colorUseRule=UseColor_Specified, specifiedColor=<rgba 0.2435 0.0593 0.2134 1>}}
   slot {
    class=CanvasPanelSlot
    properties {layoutData {alignment=<0 0>, anchors {maximum=<0.5 0.5>, minimum=<0.5 0.5>}, offsets=<ltrb 355.0391 23.4595 150 30>}}
   }
  }
  BakeIKFKTool {
   widgetClassPath=&EUW_Picker_Button
   bIsVariable=true
   parent=CanvasPanel_0
   widget=BakeIKFKTool
   properties {label Text=IK/FK Tool Visibility, normal Tint {colorUseRule=UseColor_Specified, specifiedColor=<rgba 0.2423 0.0595 0.2122 1>}}
   slot {
    class=CanvasPanelSlot
    properties {layoutData {alignment=<0 0>, anchors {maximum=<0.5 0.5>, minimum=<0.5 0.5>}, offsets=<ltrb 355.0391 95.4595 150 30>}}
   }
  }
  BG {
   widgetClassPath=&Image
   bIsVariable=true
   parent=CanvasPanel_0
   widget=BG
   properties {brush {
     bIsDynamicallyLoaded=false
     drawAs=Image
     imageType=FullColor
     mirroring=NoMirror
     resourceName=None
     resourceObject=/Game/Characters/Player/Common/SharedRigs/Graphics/Picker_Biped_Jonesy_Body_BW_Mid
     tiling=NoTile
     imageSize=<690 736>
     margin=<ltrb 0 0 0 0>
     outlineSettings {
      bUseBrushTransparency=false
      roundingType=HalfHeightRadius
      width=0
      color {colorUseRule=UseColor_Specified, specifiedColor=<rgba 0 0 0 0>}
      cornerRadii=<0 0 0 0>
     }
     tintColor {colorUseRule=UseColor_Foreground, specifiedColor=<rgba 1 1 1 0>}
     uVRegion {bIsValid=false, max=<0 0>, min=<0 0>}
    }}
   slot {
    class=CanvasPanelSlot
    properties {layoutData {alignment=<0.5 0.5>, anchors {maximum=<0.5 0.5>, minimum=<0.5 0.5>}, offsets=<ltrb 0 -29.5168 690 756.9664>}}
   }
  }
  PropVis_Button {
   widgetClassPath=&EUW_Picker_Button
   bIsVariable=true
   parent=CanvasPanel_0
   widget=PropVis_Button
   properties {label Text=Prop Visibility, normal Tint {colorUseRule=UseColor_Specified, specifiedColor=<rgba 0.2423 0.0595 0.2122 1>}}
   slot {
    class=CanvasPanelSlot
    properties {layoutData {alignment=<0 0>, anchors {maximum=<0.5 0.5>, minimum=<0.5 0.5>}, offsets=<ltrb 355 -270 150 30>}}
   }
  }
  Match_Leg_L {
   widgetClassPath=&EUW_Picker_Button
   bIsVariable=true
   parent=CanvasPanel_0
   widget=Match_Leg_L
   properties {
    draw Style=Square
    label Font Size=7
    label Text=Match
    label Color {colorUseRule=UseColor_Specified, specifiedColor=<rgba 0 0 0 1>}
    normal Tint {colorUseRule=UseColor_Specified, specifiedColor=<rgba 0.9473 0.314 0.0545 1>}
   }
   slot {
    class=CanvasPanelSlot
    properties {layoutData {alignment=<0.5 0.5>, anchors {maximum=<0.5 0.5>, minimum=<0.5 0.5>}, offsets=<ltrb 180.4947 225 34.9112 17>}}
   }
  }
  Snap_Leg_L {
   widgetClassPath=&EUW_Picker_Button
   bIsVariable=true
   parent=CanvasPanel_0
   widget=Snap_Leg_L
   properties {
    draw Style=Square
    label Font Size=7
    label Text=Snap
    label Color {colorUseRule=UseColor_Specified, specifiedColor=<rgba 0 0 0 1>}
    normal Tint {colorUseRule=UseColor_Specified, specifiedColor=<rgba 0.5472 0.2366 0.8009 1>}
   }
   slot {
    class=CanvasPanelSlot
    properties {layoutData {alignment=<0.5 0.5>, anchors {maximum=<0.5 0.5>, minimum=<0.5 0.5>}, offsets=<ltrb 219.7927 225 33.5072 17>}}
   }
  }
  Snap_Leg_R {
   widgetClassPath=&EUW_Picker_Button
   bIsVariable=true
   parent=CanvasPanel_0
   widget=Snap_Leg_R
   properties {
    draw Style=Square
    label Font Size=7
    label Text=Snap
    label Color {colorUseRule=UseColor_Specified, specifiedColor=<rgba 0 0 0 1>}
    normal Tint {colorUseRule=UseColor_Specified, specifiedColor=<rgba 0.5472 0.2366 0.8009 1>}
   }
   slot {
    class=CanvasPanelSlot
    properties {layoutData {alignment=<0.5 0.5>, anchors {maximum=<0.5 0.5>, minimum=<0.5 0.5>}, offsets=<ltrb -148.2073 225 33.5072 17>}}
   }
  }
  Match_Leg_R {
   widgetClassPath=&EUW_Picker_Button
   bIsVariable=true
   parent=CanvasPanel_0
   widget=Match_Leg_R
   properties {
    draw Style=Square
    label Font Size=7
    label Text=Match
    label Color {colorUseRule=UseColor_Specified, specifiedColor=<rgba 0 0 0 1>}
    normal Tint {colorUseRule=UseColor_Specified, specifiedColor=<rgba 0.9473 0.314 0.0545 1>}
   }
   slot {
    class=CanvasPanelSlot
    properties {layoutData {alignment=<0.5 0.5>, anchors {maximum=<0.5 0.5>, minimum=<0.5 0.5>}, offsets=<ltrb -187.5053 225 34.9112 17>}}
   }
  }
  Match_Arm_R {
   widgetClassPath=&EUW_Picker_Button
   bIsVariable=true
   parent=CanvasPanel_0
   widget=Match_Arm_R
   properties {
    draw Style=Square
    label Font Size=7
    label Text=Match
    label Color {colorUseRule=UseColor_Specified, specifiedColor=<rgba 0 0 0 1>}
    normal Tint {colorUseRule=UseColor_Specified, specifiedColor=<rgba 0.9473 0.314 0.0545 1>}
   }
   slot {
    class=CanvasPanelSlot
    properties {layoutData {alignment=<0.5 0.5>, anchors {maximum=<0.5 0.5>, minimum=<0.5 0.5>}, offsets=<ltrb -179.5053 -292.0405 34.9112 17>}}
   }
  }
  Snap_Arm_R {
   widgetClassPath=&EUW_Picker_Button
   bIsVariable=true
   parent=CanvasPanel_0
   widget=Snap_Arm_R
   properties {
    draw Style=Square
    label Font Size=7
    label Text=Snap
    label Color {colorUseRule=UseColor_Specified, specifiedColor=<rgba 0 0 0 1>}
    normal Tint {colorUseRule=UseColor_Specified, specifiedColor=<rgba 0.5472 0.2366 0.8009 1>}
   }
   slot {
    class=CanvasPanelSlot
    properties {layoutData {alignment=<0.5 0.5>, anchors {maximum=<0.5 0.5>, minimum=<0.5 0.5>}, offsets=<ltrb -140.2073 -292.0405 33.5072 17>}}
   }
  }
  Match_Arm_L {
   widgetClassPath=&EUW_Picker_Button
   bIsVariable=true
   parent=CanvasPanel_0
   widget=Match_Arm_L
   properties {
    draw Style=Square
    label Font Size=7
    label Text=Match
    label Color {colorUseRule=UseColor_Specified, specifiedColor=<rgba 0 0 0 1>}
    normal Tint {colorUseRule=UseColor_Specified, specifiedColor=<rgba 0.9473 0.314 0.0545 1>}
   }
   slot {
    class=CanvasPanelSlot
    properties {layoutData {alignment=<0.5 0.5>, anchors {maximum=<0.5 0.5>, minimum=<0.5 0.5>}, offsets=<ltrb 140.4947 -292.0405 34.9112 17>}}
   }
  }
  Snap_Arm_L {
   widgetClassPath=&EUW_Picker_Button
   bIsVariable=true
   parent=CanvasPanel_0
   widget=Snap_Arm_L
   properties {
    draw Style=Square
    label Font Size=7
    label Text=Snap
    label Color {colorUseRule=UseColor_Specified, specifiedColor=<rgba 0 0 0 1>}
    normal Tint {colorUseRule=UseColor_Specified, specifiedColor=<rgba 0.5472 0.2366 0.8009 1>}
   }
   slot {
    class=CanvasPanelSlot
    properties {layoutData {alignment=<0.5 0.5>, anchors {maximum=<0.5 0.5>, minimum=<0.5 0.5>}, offsets=<ltrb 179.7927 -292.0405 33.5072 17>}}
   }
  }
 ]
}
